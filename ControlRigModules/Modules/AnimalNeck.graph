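# AnimalNeck is a modular Control Rig module that rigs an animal neck with IK spline and FK controls.
# Construction creates rig bones, head, neck and tangent controls, an FK/IK switch, and child sockets.
# Spline logic fits bones to a control-driven curve, optionally stretched, using an offset secondary spline as each bone's up-axis aim target.
# Optional dynamics read strength, damping and blend values from animation channels found through module metadata.
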ControlRigBlueprint /ControlRigModules/Modules/AnimalNeck
graphs [
 CollapseNode_ContainedGraph {nodes [
   RigVMFunctionEntryNode()
   DISPATCH_RigDispatch_GetModuleMetadata = RigVMDispatchNode(Default=(Type=Bone,Name="None"), Name=Dynamics Channel, NameSpace=Self) {template="DISPATCH_RigDispatch_GetModuleMetadata(in Name,in NameSpace,in Default,out Value,out Found)"}
   n1 = RigVMDispatchNode(Default=(Type=Bone,Name="None"), Name=Channel Parent, NameSpace=Self) {template="DISPATCH_RigDispatch_GetModuleMetadata(in Name,in NameSpace,in Default,out Value,out Found)"}
   GetAnimationChannel = /Script/ControlRig.RigUnit_GetBoolAnimationChannel(bInitial=False, Channel=@DISPATCH_RigDispatch_GetModuleMetadata.Value.Name, Control=@n1.Value.Name) {is=struct RigVMUnitNode, template="GetAnimationChannel::Execute(out Value,in Control,in Channel,in bInitial)"}
   RigVMFunctionReturnNode(
    Value=@GetAnimationChannel.Value
    Value_1=/Script/ControlRig.RigUnit_GetFloatAnimationChannel(
     bInitial=False
     Channel=RigVMDispatchNode(Default=(Type=Bone,Name="None"), Name=Strength Channel, NameSpace=Self) {template="DISPATCH_RigDispatch_GetModuleMetadata(in Name,in NameSpace,in Default,out Value,out Found)"}.Value.Name
     Control=@n1.Value.Name
    ) {is=struct RigVMUnitNode, template="GetAnimationChannel::Execute(out Value,in Control,in Channel,in bInitial)"}.Value
    Value_2=&RigUnit_GetFloatAnimationChannel(
     bInitial=False
     Channel=RigVMDispatchNode(Default=(Type=Bone,Name="None"), Name=Damping Channel, NameSpace=Self) {template="DISPATCH_RigDispatch_GetModuleMetadata(in Name,in NameSpace,in Default,out Value,out Found)"}.Value.Name
     Control=@n1.Value.Name
    ) {template="GetAnimationChannel::Execute(out Value,in Control,in Channel,in bInitial)"}.Value
    Value_3=&RigUnit_GetFloatAnimationChannel(
     bInitial=False
     Channel=RigVMDispatchNode(Default=(Type=Bone,Name="None"), Name=Blend Start Channel, NameSpace=Self) {template="DISPATCH_RigDispatch_GetModuleMetadata(in Name,in NameSpace,in Default,out Value,out Found)"}.Value.Name
     Control=@n1.Value.Name
    ) {template="GetAnimationChannel::Execute(out Value,in Control,in Channel,in bInitial)"}.Value
    Value_4=&RigUnit_GetFloatAnimationChannel(
     bInitial=False
     Channel=RigVMDispatchNode(Default=(Type=Bone,Name="None"), Name=Blend End Channel, NameSpace=Self) {template="DISPATCH_RigDispatch_GetModuleMetadata(in Name,in NameSpace,in Default,out Value,out Found)"}.Value.Name
     Control=@n1.Value.Name
    ) {template="GetAnimationChannel::Execute(out Value,in Control,in Channel,in bInitial)"}.Value
    Value_5=&RigUnit_GetFloatAnimationChannel(
     bInitial=False
     Channel=RigVMDispatchNode(Default=(Type=Bone,Name="None"), Name=Blend Twist Channel, NameSpace=Self) {template="DISPATCH_RigDispatch_GetModuleMetadata(in Name,in NameSpace,in Default,out Value,out Found)"}.Value.Name
     Control=@n1.Value.Name
    ) {template="GetAnimationChannel::Execute(out Value,in Control,in Channel,in bInitial)"}.Value
   )
  ]}
 CollapseNode_ContainedGraph {nodes [
   n1 = RigVMFunctionEntryNode()
   RigVMFunction_StringSplit = /Script/RigVM.RigVMFunction_StringSplit(Separator=_, Value=RigVMDispatchNode(Value=@n1.Value) {template="DISPATCH_RigDispatch_ToString(in Value,out Result)"}.Result) {is=struct RigVMUnitNode}
   > RigVMFunctionReturnNode(Result=RigVMDispatchNode(String=RigVMDispatchNode(Index=0, Array=@RigVMFunction_StringSplit.Result) {template="DISPATCH_RigVMDispatch_ArrayGetAtIndex(in Array,in Index,out Element)"}.Element) {template="DISPATCH_RigDispatch_FromString(in String,out Result)"}.Result)
  ]}
 CollapseNode_ContainedGraph {nodes [
   n1 = RigVMFunctionEntryNode()
   > RigVMFunction_Sequence = RigVMAggregateNode()
   DISPATCH_RigVMDispatch_ArrayGetNum = RigVMDispatchNode(Array=@n1.Array) {template="DISPATCH_RigVMDispatch_ArrayGetNum(in Array,out Num)"}
   LessEqual = /Script/RigVM.RigVMFunction_MathIntLessEqual(B=3, A=@DISPATCH_RigVMDispatch_ArrayGetNum.Num) {is=struct RigVMUnitNode, template="LessEqual::Execute(in A,in B,out Result)"}
   DISPATCH_RigVMDispatch_If = RigVMDispatchNode(True=4, False=@DISPATCH_RigVMDispatch_ArrayGetNum.Num, Condition=@LessEqual.Result) {template="DISPATCH_RigVMDispatch_If(in Condition,in True,in False,out Result)"}
   DISPATCH_RigVMDispatch_ArraySetNum = RigVMDispatchNode(Num=@DISPATCH_RigVMDispatch_If.Result, ExecuteContext=@RigVMFunction_Sequence.A, Array=RigVMVariableNode() {variable=Positions}.Value) {template="DISPATCH_RigVMDispatch_ArraySetNum(io Array,in Num)"}
   > DISPATCH_RigVMDispatch_ArrayIterator = RigVMDispatchNode(Array=@n1.Array) {template="DISPATCH_RigVMDispatch_ArrayIterator(in Array,out Element,out Index,out Count,out Ratio)"}
   > RigVMDispatchNode(
    Index=@DISPATCH_RigVMDispatch_ArrayIterator.Index
    Element=&RigUnit_GetTransform(bInitial=False, Item=@DISPATCH_RigVMDispatch_ArrayIterator.Element).Transform.Translation
    Array=RigVMVariableNode() {variable=Positions}.Value
   ) {template="DISPATCH_RigVMDispatch_ArraySetAtIndex(io Array,in Index,in Element)"}
   n2 = RigVMDispatchNode(Array=@n1.Array) {template="DISPATCH_RigVMDispatch_ArrayGetNum(in Array,out Num)"}
   DISPATCH_RigVMDispatch_CoreEquals = RigVMDispatchNode(B=3, A=@n2.Num) {template="DISPATCH_RigVMDispatch_CoreEquals(in A,in B,out Result)"}
   RigVMFunction_ControlFlowBranch = &RigVMFunction_ControlFlowBranch(Condition=@DISPATCH_RigVMDispatch_CoreEquals.Result, ExecuteContext=@RigVMFunction_Sequence.B)
   n3 = &RigVMFunction_ControlFlowBranch(
    Condition=RigVMDispatchNode(B=2, A=@n2.Num) {template="DISPATCH_RigVMDispatch_CoreEquals(in A,in B,out Result)"}.Result
    ExecuteContext=@RigVMFunction_ControlFlowBranch.Completed
   )
   Greater = /Script/RigVM.RigVMFunction_MathIntGreater(B=4, A=@n2.Num) {is=struct RigVMUnitNode, template="Greater::Execute(in A,in B,out Result)"}
   n4 = &RigVMFunction_ControlFlowBranch(Condition=@Greater.Result, ExecuteContext=@n3.Completed)
   RigVMDispatchNode(
    Index=0
    Element=RigVMDispatchNode(Index=0, Array=RigVMVariableNode() {variable=Positions}.Value) {template="DISPATCH_RigVMDispatch_ArrayGetAtIndex(in Array,in Index,out Element)"}.Element
    ExecuteContext=@n4.True
    Array=RigVMVariableNode() {variable=Positions}.Value
   ) {template="DISPATCH_RigVMDispatch_ArraySetAtIndex(io Array,in Index,in Element)"}
   > RigVMDispatchNode(
    Index=1
    Element=RigVMDispatchNode(Index=1, Array=RigVMVariableNode() {variable=Positions}.Value) {template="DISPATCH_RigVMDispatch_ArrayGetAtIndex(in Array,in Index,out Element)"}.Element
    Array=RigVMVariableNode() {variable=Positions}.Value
   ) {template="DISPATCH_RigVMDispatch_ArraySetAtIndex(io Array,in Index,in Element)"}
   > RigVMDispatchNode(
    Index=2
    Element=RigVMDispatchNode(Index=2, Array=RigVMVariableNode() {variable=Positions}.Value) {template="DISPATCH_RigVMDispatch_ArrayGetAtIndex(in Array,in Index,out Element)"}.Element
    Array=RigVMVariableNode() {variable=Positions}.Value
   ) {template="DISPATCH_RigVMDispatch_ArraySetAtIndex(io Array,in Index,in Element)"}
   > RigVMDispatchNode(
    Index=3
    Element=RigVMDispatchNode(Index=-1, Array=RigVMVariableNode() {variable=Positions}.Value) {template="DISPATCH_RigVMDispatch_ArrayGetAtIndex(in Array,in Index,out Element)"}.Element
    Array=RigVMVariableNode() {variable=Positions}.Value
   ) {template="DISPATCH_RigVMDispatch_ArraySetAtIndex(io Array,in Index,in Element)"}
   RigVMDispatchNode(
    Index=0
    Element=RigVMDispatchNode(Index=0, Array=RigVMVariableNode() {variable=Positions}.Value) {template="DISPATCH_RigVMDispatch_ArrayGetAtIndex(in Array,in Index,out Element)"}.Element
    ExecuteContext=@n3.True
    Array=RigVMVariableNode() {variable=Positions}.Value
   ) {template="DISPATCH_RigVMDispatch_ArraySetAtIndex(io Array,in Index,in Element)"}
   > RigVMDispatchNode(
    Index=1
    Element=RigVMDispatchNode(Index=0, Array=RigVMVariableNode() {variable=Positions}.Value) {template="DISPATCH_RigVMDispatch_ArrayGetAtIndex(in Array,in Index,out Element)"}.Element
    Array=RigVMVariableNode() {variable=Positions}.Value
   ) {template="DISPATCH_RigVMDispatch_ArraySetAtIndex(io Array,in Index,in Element)"}
   > RigVMDispatchNode(
    Index=2
    Element=RigVMDispatchNode(Index=0, Array=RigVMVariableNode() {variable=Positions}.Value) {template="DISPATCH_RigVMDispatch_ArrayGetAtIndex(in Array,in Index,out Element)"}.Element
    Array=RigVMVariableNode() {variable=Positions}.Value
   ) {template="DISPATCH_RigVMDispatch_ArraySetAtIndex(io Array,in Index,in Element)"}
   > RigVMDispatchNode(
    Index=3
    Element=RigVMDispatchNode(Index=1, Array=RigVMVariableNode() {variable=Positions}.Value) {template="DISPATCH_RigVMDispatch_ArrayGetAtIndex(in Array,in Index,out Element)"}.Element
    Array=RigVMVariableNode() {variable=Positions}.Value
   ) {template="DISPATCH_RigVMDispatch_ArraySetAtIndex(io Array,in Index,in Element)"}
   DISPATCH_RigVMDispatch_ArrayGetAtIndex = RigVMDispatchNode(Index=0, Array=RigVMVariableNode() {variable=Positions}.Value) {template="DISPATCH_RigVMDispatch_ArrayGetAtIndex(in Array,in Index,out Element)"}
   DISPATCH_RigVMDispatch_ArraySetAtIndex = RigVMDispatchNode(
    Index=0
    Element=@DISPATCH_RigVMDispatch_ArrayGetAtIndex.Element
    ExecuteContext=@RigVMFunction_ControlFlowBranch.True
    Array=RigVMVariableNode() {variable=Positions}.Value
   ) {template="DISPATCH_RigVMDispatch_ArraySetAtIndex(io Array,in Index,in Element)"}
   > RigVMDispatchNode(
    Index=1
    Element=RigVMDispatchNode(Index=1, Array=RigVMVariableNode() {variable=Positions}.Value) {template="DISPATCH_RigVMDispatch_ArrayGetAtIndex(in Array,in Index,out Element)"}.Element
    Array=RigVMVariableNode() {variable=Positions}.Value
   ) {template="DISPATCH_RigVMDispatch_ArraySetAtIndex(io Array,in Index,in Element)"}
   > RigVMDispatchNode(
    Index=2
    Element=RigVMDispatchNode(Index=2, Array=RigVMVariableNode() {variable=Positions}.Value) {template="DISPATCH_RigVMDispatch_ArrayGetAtIndex(in Array,in Index,out Element)"}.Element
    Array=RigVMVariableNode() {variable=Positions}.Value
   ) {template="DISPATCH_RigVMDispatch_ArraySetAtIndex(io Array,in Index,in Element)"}
   > RigVMDispatchNode(
    Index=3
    Element=RigVMDispatchNode(Index=3, Array=RigVMVariableNode() {variable=Positions}.Value) {template="DISPATCH_RigVMDispatch_ArrayGetAtIndex(in Array,in Index,out Element)"}.Element
    Array=RigVMVariableNode() {variable=Positions}.Value
   ) {template="DISPATCH_RigVMDispatch_ArraySetAtIndex(io Array,in Index,in Element)"}
   n5 = &RigVMFunction_ControlFlowBranch(Condition=@n1.Debug, ExecuteContext=@RigVMFunction_Sequence.C)
   n6 = RigVMVariableNode() {variable=Positions}
   RigVMFunctionReturnNode(ExecuteContext=@n5.Completed, Value=@n6.Value)
   n7 = RigVMDispatchNode(ExecuteContext=@n5.True, Array=@n6.Value) {template="DISPATCH_RigVMDispatch_ArrayIterator(in Array,out Element,out Index,out Count,out Ratio)"}
   > RigVMFunction_DebugTransformMutableNoSpace = /Script/RigVM.RigVMFunction_DebugTransformMutableNoSpace(
    Color=(R=1,G=0,B=0,A=1)
    DebugDrawSettings=(DepthPriority=SDPG_Foreground,Lifetime=-1)
    Mode=Axes
    Scale=10
    Thickness=1
    Transform=(Rotation=(X=0,Y=0,Z=0,W=1),Translation=(X=0,Y=0,Z=0),Scale3D=(X=1,Y=1,Z=1))
    WorldOffset=(Rotation=(X=0,Y=0,Z=0,W=1),Translation=(X=0,Y=0,Z=0),Scale3D=(X=1,Y=1,Z=1))
    bEnabled=True
    Transform.Translation=@n7.Element
   ) {is=struct RigVMUnitNode}
  ]}
 Concat_1_1_ContainedGraph {nodes [
   n1 = RigVMFunctionEntryNode()
   n2 = /Script/RigVM.RigVMFunction_NameConcat(A=@n1.A, B=@n1.B) {is=struct RigVMUnitNode, template="Concat::Execute(in A,in B,out Result)"}
   RigVMFunctionReturnNode(Result=&RigVMFunction_NameConcat(A=&RigVMFunction_NameConcat(A=@n2.Result, B=@n1.C) {template="Concat::Execute(in A,in B,out Result)"}.Result, B=@n1.D) {template="Concat::Execute(in A,in B,out Result)"}.Result)
  ]}
 Concat_1_1_ContainedGraph {nodes [
   n1 = RigVMFunctionEntryNode()
   n2 = &RigVMFunction_NameConcat(A=@n1.A, B=@n1.B) {template="Concat::Execute(in A,in B,out Result)"}
   RigVMFunctionReturnNode(Result=&RigVMFunction_NameConcat(A=&RigVMFunction_NameConcat(A=@n2.Result, B=@n1.C) {template="Concat::Execute(in A,in B,out Result)"}.Result, B=@n1.D) {template="Concat::Execute(in A,in B,out Result)"}.Result)
  ]}
 Concat_1_1_ContainedGraph {nodes [
   n1 = RigVMFunctionEntryNode()
   n2 = &RigVMFunction_NameConcat(A=@n1.A, B=@n1.B) {template="Concat::Execute(in A,in B,out Result)"}
   RigVMFunctionReturnNode(Result=&RigVMFunction_NameConcat(A=&RigVMFunction_NameConcat(A=@n2.Result, B=@n1.C) {template="Concat::Execute(in A,in B,out Result)"}.Result, B=@n1.D) {template="Concat::Execute(in A,in B,out Result)"}.Result)
  ]}
 Concat_1_ContainedGraph {nodes [
   n1 = RigVMFunctionEntryNode()
   Concat = &RigVMFunction_NameConcat(A=@n1.A, B=@n1.B) {template="Concat::Execute(in A,in B,out Result)"}
   n2 = &RigVMFunction_NameConcat(A=@Concat.Result, B=@n1.C) {template="Concat::Execute(in A,in B,out Result)"}
   RigVMFunctionReturnNode(Result=&RigVMFunction_NameConcat(A=&RigVMFunction_NameConcat(A=@n2.Result, B=@n1.D) {template="Concat::Execute(in A,in B,out Result)"}.Result, B=@n1.E) {template="Concat::Execute(in A,in B,out Result)"}.Result)
  ]}
 Concat_1_ContainedGraph {nodes [
   n1 = RigVMFunctionEntryNode()
   Concat = &RigVMFunction_NameConcat(A=@n1.A, B=@n1.B) {template="Concat::Execute(in A,in B,out Result)"}
   n2 = &RigVMFunction_NameConcat(A=@Concat.Result, B=@n1.C) {template="Concat::Execute(in A,in B,out Result)"}
   RigVMFunctionReturnNode(Result=&RigVMFunction_NameConcat(A=&RigVMFunction_NameConcat(A=@n2.Result, B=@n1.D) {template="Concat::Execute(in A,in B,out Result)"}.Result, B=@n1.E) {template="Concat::Execute(in A,in B,out Result)"}.Result)
  ]}
 Concat_4_1_ContainedGraph {nodes [
   n1 = RigVMFunctionEntryNode()
   n2 = &RigVMFunction_NameConcat(A=@n1.A, B=@n1.B) {template="Concat::Execute(in A,in B,out Result)"}
   RigVMFunctionReturnNode(Result=&RigVMFunction_NameConcat(A=&RigVMFunction_NameConcat(A=@n2.Result, B=@n1.C) {template="Concat::Execute(in A,in B,out Result)"}.Result, B=@n1.D) {template="Concat::Execute(in A,in B,out Result)"}.Result)
  ]}
 Concat_4_1_ContainedGraph {nodes [
   n1 = RigVMFunctionEntryNode()
   n2 = &RigVMFunction_NameConcat(A=@n1.A, B=@n1.B) {template="Concat::Execute(in A,in B,out Result)"}
   RigVMFunctionReturnNode(Result=&RigVMFunction_NameConcat(A=&RigVMFunction_NameConcat(A=@n2.Result, B=@n1.C) {template="Concat::Execute(in A,in B,out Result)"}.Result, B=@n1.D) {template="Concat::Execute(in A,in B,out Result)"}.Result)
  ]}
 ConstructionGraph {nodes [
   RigUnit_PrepareForExecution = &RigUnit_PrepareForExecution()
   Create Neck Rig Bones = RigVMFunctionReferenceNode(ExecuteContext=@RigUnit_PrepareForExecution.ExecutePin, Enable Orient Bones=RigVMVariableNode() {variable=Re-Orient Controls}.Value)
   n1 = RigVMVariableNode() {variable=Module Settings}
   > CONSTR IK Animal Neck = RigVMFunctionReferenceNode(
    Axis=(X=-1,Y=0,Z=0)
    Dynamics=RigVMVariableNode() {variable=Dynamics}.Value
    Create Control Per Bone=RigVMVariableNode() {variable=Create Control per Bone}.Value
    Up Axis=RigVMVariableNode() {variable=Secondary Axis}.Value
    Parent Controls to each other=RigVMVariableNode() {variable=Main Controls as FK}.Value
    Add Stretch Option=RigVMVariableNode() {variable=Stretch}.Value
    Head Control Settings=RigVMVariableNode() {variable=Head Settings}.Value
    Headneck Position=RigVMVariableNode() {variable=Headneck Control Pivot Position}.Value
    Debug=RigVMVariableNode() {variable=Debug}.Value
    Dynamics Strength=RigVMVariableNode() {variable=Default Strength}.Value
    Suffix=@n1.Value.ControlSuffix_38_#1
    Prefix=@n1.Value.Prefix_20_#2
    Primary Axis=@n1.Value.PrimaryBoneAxis_32_#3
    Dynamics Damping=RigVMVariableNode() {variable=Default Damping}.Value
    Number of Controls=RigVMVariableNode() {variable=Number of Controls}.Value
    Enable Orient Bones=RigVMVariableNode() {variable=Re-Orient Controls}.Value
    Dynamics Blend Start=RigVMVariableNode() {variable=Default Blend Start}.Value
    Dynamics Blend End=RigVMVariableNode() {variable=Default Blend End}.Value
    Neck Tangent Control Settings=RigVMVariableNode() {variable=Tangent Settings}.Value
    Number of Main Controls=RigVMVariableNode() {variable=Number of Main Controls}.Value
    Dynamics Blend Twist=RigVMVariableNode() {variable=Default Blend Twist}.Value
    Secondary Spline Direction=RigVMVariableNode() {variable=Secondary Spline Direction}.Value
   )
   > Create FK IK Switch = RigVMFunctionReferenceNode(name=neck, prefix=None, side=None, Initial Value=RigVMVariableNode() {variable=Default IK or FK}.Value)
   Get_Module_Metadata = RigVMDispatchNode(Default=(), Name=Rig Bones, NameSpace=Self) {template="DISPATCH_RigDispatch_GetModuleMetadata(in Name,in NameSpace,in Default,out Value,out Found)"}
   > Create FK Tail Controls = RigVMFunctionReferenceNode(
    Start=RigVMDispatchNode(Index=0, Array=@Get_Module_Metadata.Value) {template="DISPATCH_RigVMDispatch_ArrayGetAtIndex(in Array,in Index,out Element)"}.Element
    End=RigVMDispatchNode(Index=-1, Array=@Get_Module_Metadata.Value) {template="DISPATCH_RigVMDispatch_ArrayGetAtIndex(in Array,in Index,out Element)"}.Element
    Settings=RigVMVariableNode() {variable=FK Control Settings}.Value
   )
   > Set Controls Scale and Color = RigVMFunctionReferenceNode(Module Settings=RigVMVariableNode() {variable=Module Settings}.Value)
   > Sockets on Direct Children = RigVMFunctionReferenceNode(Parent=/Script/ControlRig.RigUnit_ResolveConnector(Connector=(Type=Connector,Name="End Bone"), SkipSocket=true) {is=struct RigVMUnitNode}.Result)
  ]}
 ContainedGraph {nodes [
   n1 = RigVMFunctionEntryNode()
   RigUnit_SequenceAggregate_1 = &RigVMFunction_Sequence(ExecuteContext=@n1.ExecuteContext)
   RigUnit_SequenceAggregate = RigVMAggregateNode(ExecuteContext=@RigUnit_SequenceAggregate_1.A)
   BranchNode_4 = &RigVMFunction_ControlFlowBranch(Condition=@n1.X Plane, ExecuteContext=@RigUnit_SequenceAggregate.A)
   n2 = RigVMVariableNode() {variable=vector pos}
   RigVMVariableNode(
    Value=(X=0,Y=0,Z=0)
    ExecuteContext=@BranchNode_4.True
    Value.X=&RigUnit_GetTransform(bInitial=true, Item=@n1.Item).Transform.Translation.X
    Value.Y=@n2.Value.Y
    Value.Z=@n2.Value.Z
   ) {variable=vector pos}
   BranchNode_1_1 = &RigVMFunction_ControlFlowBranch(Condition=@n1.Y Plane, ExecuteContext=@RigUnit_SequenceAggregate.B)
   n3 = RigVMVariableNode() {variable=vector pos}
   RigVMVariableNode(
    Value=(X=0,Y=0,Z=0)
    ExecuteContext=@BranchNode_1_1.True
    Value.Y=&RigUnit_GetTransform(bInitial=true, Item=@n1.Item).Transform.Translation.Y
    Value.X=@n3.Value.X
    Value.Z=@n3.Value.Z
   ) {variable=vector pos}
   BranchNode_2_1 = &RigVMFunction_ControlFlowBranch(Condition=@n1.Z Plane, ExecuteContext=@RigUnit_SequenceAggregate.C)
   n4 = RigVMVariableNode() {variable=vector pos}
   RigVMVariableNode(
    Value=(X=0,Y=0,Z=0)
    ExecuteContext=@BranchNode_2_1.True
    Value.Z=&RigUnit_GetTransform(bInitial=true, Item=@n1.Item).Transform.Translation.Z
    Value.X=@n4.Value.X
    Value.Y=@n4.Value.Y
   ) {variable=vector pos}
   Cast = &RigVMFunction_MathBoolToFloat(Value=@n1.X Plane)
   Dot__ = /Script/RigVM.RigVMFunction_MathVectorDot(
    B=(X=1,Y=1,Z=0)
    B.X=@Cast.Result
    B.Y=&RigVMFunction_MathBoolToFloat(Value=@n1.Y Plane).Result
    B.Z=&RigVMFunction_MathBoolToFloat(Value=@n1.Z Plane).Result
    A=RigVMVariableNode() {variable=vector pos}.Value
   ) {is=struct RigVMUnitNode, template="Dot,|::Execute(in A,in B,out Result)"}
   Less = /Script/RigVM.RigVMFunction_MathDoubleLess(B=0, A=@Dot__.Result) {is=struct RigVMUnitNode, template="Less::Execute(in A,in B,out Result)"}
   RigVMVariableNode(Value=@Less.Result, ExecuteContext=@RigUnit_SequenceAggregate.D) {variable=is on right}
   RigVMFunctionReturnNode(ExecuteContext=@RigUnit_SequenceAggregate_1.B, Right Side=RigVMVariableNode() {variable=is on right}.Value)
  ]}
 ContainedGraph {nodes [
   n1 = RigVMFunctionEntryNode()
   RigUnit_SequenceAggregate = &RigVMFunction_Sequence(ExecuteContext=@n1.ExecuteContext)
   RigUnit_SequenceAggregate_1 = &RigVMFunction_Sequence(ExecuteContext=@RigUnit_SequenceAggregate.B)
   RigUnit_SequenceAggregate_2 = &RigVMFunction_Sequence(ExecuteContext=@RigUnit_SequenceAggregate_1.B)
   RigVMFunctionReturnNode(A=@RigUnit_SequenceAggregate.A, B=@RigUnit_SequenceAggregate_1.A, C=@RigUnit_SequenceAggregate_2.A, D=@RigUnit_SequenceAggregate_2.B)
  ]}
 ContainedGraph {nodes [
   BranchNode_RigVMUnitNode = &RigVMFunction_ControlFlowBranch(ExecuteContext=(), Condition=RigVMVariableNode() {variable=debug}.Value)
   /Script/ControlRigSpline.RigUnit_DrawControlRigSpline(
    Color=(R=1,G=0,B=0,A=1)
    Detail=16
    Thickness=0.75
    ExecutePin=@BranchNode_RigVMUnitNode.True
    Spline=RigVMVariableNode() {variable=template_spline}.Value
   ) {is=struct RigVMUnitNode}
   n1 = RigVMFunctionEntryNode()
   RigVMFunction_Sequence = &RigVMFunction_Sequence(ExecuteContext=@n1.ExecuteContext)
   n2 = RigVMVariableNode() {variable=Controls}
   BranchNode_4_RigVMUnitNode = &RigVMFunction_ControlFlowBranch(
    Condition=/Script/RigVM.RigVMFunction_MathIntLess(B=4, A=RigVMDispatchNode(Array=@n2.Value) {template="DISPATCH_RigVMDispatch_ArrayGetNum(in Array,out Num)"}.Num) {is=struct RigVMUnitNode, template="Less::Execute(in A,in B,out Result)"}.Result
    ExecuteContext=@RigVMFunction_Sequence.B
   )
   RigVMFunctionReturnNode(ExecuteContext=@BranchNode_4_RigVMUnitNode.Completed, Output Spline=RigVMVariableNode() {variable=local_spline}.Value)
   RigUnit_SequenceExecution_1 = RigVMAggregateNode(ExecuteContext=@BranchNode_4_RigVMUnitNode.False)
   RigVMVariableNode(Value=@n1.Template Spline, ExecuteContext=@RigUnit_SequenceExecution_1.A) {variable=template_spline}
   n3 = /Script/ControlRig.RigUnit_GetTransformItemArray(Space=GlobalSpace, bInitial=False, Items=RigVMVariableNode() {variable=Controls}.Value) {is=struct RigVMUnitNode}
   > n4 = /Script/ControlRigSpline.RigUnit_SetSplineTransforms(Spline=@n1.Template Spline, Transforms=@n3.Transforms) {is=struct RigVMUnitNode}
   > RigVMVariableNode(Value=@n4.Spline) {variable=local_spline}
   > RigVMVariableNode(Value=@n3.Transforms) {variable=template_transforms}
   > RigUnit_SequenceExecution = RigVMAggregateNode()
   n5 = RigVMVariableNode() {variable=template_transforms}
   RigVMDispatchNode(
    Num=RigVMDispatchNode(Array=@n5.Value) {template="DISPATCH_RigVMDispatch_ArrayGetNum(in Array,out Num)"}.Num
    ExecuteContext=@RigUnit_SequenceExecution.A
    Array=RigVMVariableNode() {variable=offset_positions}.Value
   ) {template="DISPATCH_RigVMDispatch_ArraySetNum(io Array,in Num)"}
   > n6 = RigVMDispatchNode(Array=@n5.Value) {template="DISPATCH_RigVMDispatch_ArrayIterator(in Array,out Element,out Index,out Count,out Ratio)"}
   > RigVMDispatchNode(
    Index=@n6.Index
    Element=/Script/RigVM.RigVMFunction_MathTransformTransformVector(Transform=@n6.Element, Location=RigVMVariableNode() {variable=Secondary Spline Direction}.Value) {is=struct RigVMUnitNode}.Result
    Array=RigVMVariableNode() {variable=offset_positions}.Value
   ) {template="DISPATCH_RigVMDispatch_ArraySetAtIndex(io Array,in Index,in Element)"}
   RigUnit_SequenceExecution_4 = RigVMAggregateNode(ExecuteContext=@RigUnit_SequenceExecution.C)
   Branch_RigVMUnitNode = &RigVMFunction_ControlFlowBranch(ExecuteContext=@RigUnit_SequenceExecution_4.A, Condition=RigVMVariableNode() {variable=Stretch}.Value)
   n7 = RigVMVariableNode() {variable=template_spline}
   n8 = RigVMVariableNode() {variable=Primary Axis}
   n9 = RigVMVariableNode() {variable=Bones}
   /Script/ControlRigSpline.RigUnit_FitChainToSplineCurveItemArray(
    Alignment=Front
    DebugSettings=(Scale=1,CurveColor=(R=1,G=1,B=0,A=1),SegmentsColor=(R=1,G=0,B=0,A=1),WorldOffset=(Rotation=(X=0,Y=0,Z=0,W=1),Translation=(X=0,Y=0,Z=0),Scale3D=(X=1,Y=1,Z=1)))
    Maximum=1
    PoleVectorPosition=(X=0,Y=0,Z=0)
    Rotations=()
    SamplingPrecision=12
    SecondaryAxis=(X=0,Y=0,Z=0)
    Weight=1
    bPropagateToChildren=True
    ExecutePin=@Branch_RigVMUnitNode.False
    Spline=@n7.Value
    PrimaryAxis=@n8.Value
    Items=@n9.Value
   ) {is=struct RigVMUnitNode}
   &RigUnit_FitChainToSplineCurveItemArray(
    Alignment=Stretched
    DebugSettings=(Scale=1,CurveColor=(R=1,G=1,B=0,A=1),SegmentsColor=(R=1,G=0,B=0,A=1),WorldOffset=(Rotation=(X=0,Y=0,Z=0,W=1),Translation=(X=0,Y=0,Z=0),Scale3D=(X=1,Y=1,Z=1)))
    Maximum=1
    PoleVectorPosition=(X=0,Y=0,Z=0)
    Rotations=()
    SamplingPrecision=12
    SecondaryAxis=(X=0,Y=0,Z=0)
    Weight=1
    bPropagateToChildren=True
    ExecutePin=@Branch_RigVMUnitNode.True
    Spline=@n7.Value
    PrimaryAxis=@n8.Value
    Items=@n9.Value
   )
   RigUnit_SequenceExecution_2 = RigVMAggregateNode(ExecuteContext=@RigUnit_SequenceExecution_4.B)
   BranchNode_2_RigVMUnitNode = &RigVMFunction_ControlFlowBranch(ExecuteContext=@RigUnit_SequenceExecution_2.A, Condition=RigVMVariableNode() {variable=debug}.Value)
   &RigUnit_DrawControlRigSpline(
    Color=(R=1,G=0,B=0,A=1)
    Detail=16
    Thickness=1
    ExecutePin=@BranchNode_2_RigVMUnitNode.True
    Spline=RigVMVariableNode() {variable=local_spline}.Value
   )
   n10 = /Script/ControlRigSpline.RigUnit_SetSplinePoints(
    ExecutePin=@RigUnit_SequenceExecution_2.B
    Points=RigVMVariableNode() {variable=offset_positions}.Value
    Spline=RigVMVariableNode() {variable=local_spline}.Value
   ) {is=struct RigVMUnitNode}
   > RigVMVariableNode(Value=@n10.Spline) {variable=local_spline}
   n11 = RigVMVariableNode() {variable=Bones}
   DISPATCH_RigVMDispatch_ArrayGetNum = RigVMDispatchNode(Array=@n11.Value) {template="DISPATCH_RigVMDispatch_ArrayGetNum(in Array,out Num)"}
   > DISPATCH_RigVMDispatch_ArraySetNum = RigVMDispatchNode(Num=@DISPATCH_RigVMDispatch_ArrayGetNum.Num, Array=RigVMVariableNode() {variable=aim_transforms}.Value) {template="DISPATCH_RigVMDispatch_ArraySetNum(io Array,in Num)"}
   > n12 = RigVMDispatchNode(Array=@n11.Value) {template="DISPATCH_RigVMDispatch_ArrayIterator(in Array,out Element,out Index,out Count,out Ratio)"}
   n13 = &RigUnit_GetTransform(Item=@n12.Element)
   n14 = RigVMVariableNode() {variable=local_spline}
   > DISPATCH_RigVMDispatch_ArraySetAtIndex = RigVMDispatchNode(
    Index=@n12.Index
    Element=/Script/ControlRigSpline.RigUnit_PositionFromControlRigSpline(
     U=/Script/ControlRigSpline.RigUnit_ClosestParameterFromControlRigSpline(Position=@n13.Transform.Translation, Spline=@n14.Value) {is=struct RigVMUnitNode}.U
     Spline=@n14.Value
    ) {is=struct RigVMUnitNode}.Position
    Array=RigVMVariableNode() {variable=aim_transforms}.Value
   ) {template="DISPATCH_RigVMDispatch_ArraySetAtIndex(io Array,in Index,in Element)"}
   n15 = RigVMDispatchNode(ExecuteContext=@RigUnit_SequenceExecution.D, Array=RigVMVariableNode() {variable=Bones}.Value) {template="DISPATCH_RigVMDispatch_ArrayIterator(in Array,out Element,out Index,out Count,out Ratio)"}
   DISPATCH_RigVMDispatch_ArrayGetAtIndex = RigVMDispatchNode(Index=@n15.Index, Array=RigVMVariableNode() {variable=aim_transforms}.Value) {template="DISPATCH_RigVMDispatch_ArrayGetAtIndex(in Array,in Index,out Element)"}
   n16 = /Script/ControlRig.RigUnit_AimBoneMath(
    DebugSettings=(Scale=10,WorldOffset=(Rotation=(X=0,Y=0,Z=0,W=1),Translation=(X=0,Y=0,Z=0),Scale3D=(X=1,Y=1,Z=1)))
    Primary=(Weight=1,Axis=(X=0,Y=1,Z=0),Target=(X=1,Y=0,Z=0),Kind=Location,Space=(Type=Bone))
    Secondary=(Weight=0,Axis=(X=0,Y=0,Z=0),Target=(X=0,Y=0,Z=0),Kind=Location,Space=(Type=Bone))
    Weight=1
    Primary.Target=@DISPATCH_RigVMDispatch_ArrayGetAtIndex.Element
    InputTransform=&RigUnit_GetTransform(Item=@n15.Element).Transform
    Primary.Axis=RigVMVariableNode() {variable=Up Axis}.Value
   ) {is=struct RigVMUnitNode}
   MathIntEquals = RigVMDispatchNode(
    A=@n15.Index
    B=/Script/RigVM.RigVMFunction_MathIntSub(B=1, A=@n15.Count) {is=struct RigVMUnitNode, template="Subtract::Execute(in A,in B,out Result)"}.Result
   ) {template="DISPATCH_RigVMDispatch_CoreEquals(in A,in B,out Result)"}
   > n17 = &RigUnit_SetTransform(Weight=1, bInitial=False, Value=@n16.Result, Item=@n15.Element, bPropagateToChildren=@MathIntEquals.Result)
   BranchNode_3_RigVMUnitNode = &RigVMFunction_ControlFlowBranch(ExecuteContext=@n17.ExecutePin, Condition=RigVMVariableNode() {variable=debug}.Value)
   /Script/ControlRig.RigUnit_DebugTransformMutableItemSpace(
    Color=(R=1,G=0,B=0,A=1)
    DebugDrawSettings=(DepthPriority=SDPG_Foreground,Lifetime=-1)
    Mode=Axes
    Scale=10
    Space=(Type=Bone)
    Thickness=0.15
    WorldOffset=(Rotation=(X=0,Y=0,Z=0,W=1),Translation=(X=0,Y=0,Z=0),Scale3D=(X=1,Y=1,Z=1))
    bEnabled=True
    Transform=@n16.Result
    ExecutePin=@BranchNode_3_RigVMUnitNode.True
   ) {is=struct RigVMUnitNode}
   Branch_1_RigVMUnitNode = &RigVMFunction_ControlFlowBranch(ExecuteContext=@RigUnit_SequenceExecution_1.B, Condition=RigVMVariableNode() {variable=Stretch}.Value)
   n18 = RigVMDispatchNode(Index=-1, Array=RigVMVariableNode() {variable=Controls}.Value) {template="DISPATCH_RigVMDispatch_ArrayGetAtIndex(in Array,in Index,out Element)"}
   n19 = RigVMDispatchNode(Index=-1, Array=RigVMVariableNode() {variable=Bones}.Value) {template="DISPATCH_RigVMDispatch_ArrayGetAtIndex(in Array,in Index,out Element)"}
   &RigUnit_ParentConstraint(
    AdvancedSettings=(RotationOrderForFilter=XZY)
    Filter=(TranslationFilter=(bX=false,bY=false,bZ=false),RotationFilter=(bX=True,bY=True,bZ=True),ScaleFilter=(bX=True,bY=True,bZ=True))
    Parents=((Item=(Type=Bone),Weight=1))
    Weight=1
    Parents.0.Item=@n18.Element
    Child=@n19.Element
    ExecutePin=@Branch_1_RigVMUnitNode.False
   )
   &RigUnit_ParentConstraint(
    AdvancedSettings=(RotationOrderForFilter=XZY)
    Parents=((Item=(Type=Bone),Weight=1))
    Weight=1
    Parents.0.Item=@n18.Element
    Child=@n19.Element
    ExecutePin=@Branch_1_RigVMUnitNode.True
   )
   RigVMVariableNode() {variable=Spline Mode}
   RigVMVariableNode() {variable=Samples Per Segment}
  ]}
 ContainedGraph {nodes [
   n1 = RigVMFunctionEntryNode()
   RigUnit_SequenceExecution_2 = &RigVMFunction_Sequence(ExecuteContext=@n1.ExecuteContext)
   RigUnit_SequenceExecution_2_1 = &RigVMFunction_Sequence(ExecuteContext=@RigUnit_SequenceExecution_2.B)
   RigUnit_SequenceExecution_2_2 = &RigVMFunction_Sequence(ExecuteContext=@RigUnit_SequenceExecution_2_1.B)
   RigVMFunctionReturnNode(A=@RigUnit_SequenceExecution_2.A, B=@RigUnit_SequenceExecution_2_1.A, C=@RigUnit_SequenceExecution_2_2.A, D=@RigUnit_SequenceExecution_2_2.B)
  ]}
 ContainedGraph {nodes [
   n1 = RigVMFunctionEntryNode()
   RigUnit_SequenceExecution_4 = &RigVMFunction_Sequence(ExecuteContext=@n1.ExecuteContext)
   RigUnit_SequenceExecution_4_1 = &RigVMFunction_Sequence(ExecuteContext=@RigUnit_SequenceExecution_4.B)
   RigUnit_SequenceExecution_4_2 = &RigVMFunction_Sequence(ExecuteContext=@RigUnit_SequenceExecution_4_1.B)
   RigVMFunctionReturnNode(A=@RigUnit_SequenceExecution_4.A, B=@RigUnit_SequenceExecution_4_1.A, C=@RigUnit_SequenceExecution_4_2.A, D=@RigUnit_SequenceExecution_4_2.B)
  ]}
 ContainedGraph {nodes [
   n1 = RigVMFunctionEntryNode()
   RigUnit_SequenceExecution = &RigVMFunction_Sequence(ExecuteContext=@n1.ExecuteContext)
   RigUnit_SequenceExecution_1 = &RigVMFunction_Sequence(ExecuteContext=@RigUnit_SequenceExecution.B)
   RigUnit_SequenceExecution_2 = &RigVMFunction_Sequence(ExecuteContext=@RigUnit_SequenceExecution_1.B)
   RigVMFunctionReturnNode(A=@RigUnit_SequenceExecution.A, B=@RigUnit_SequenceExecution_1.A, C=@RigUnit_SequenceExecution_2.A, D=@RigUnit_SequenceExecution_2.B)
  ]}
 ContainedGraph {nodes [
   n1 = RigVMFunctionEntryNode()
   RigUnit_SequenceExecution_1 = &RigVMFunction_Sequence(ExecuteContext=@n1.ExecuteContext)
   RigUnit_SequenceExecution_1_1 = &RigVMFunction_Sequence(ExecuteContext=@RigUnit_SequenceExecution_1.B)
   RigUnit_SequenceExecution_1_2 = &RigVMFunction_Sequence(ExecuteContext=@RigUnit_SequenceExecution_1_1.B)
   RigVMFunctionReturnNode(A=@RigUnit_SequenceExecution_1.A, B=@RigUnit_SequenceExecution_1_1.A, C=@RigUnit_SequenceExecution_1_2.A, D=@RigUnit_SequenceExecution_1_2.B)
  ]}
 ContainedGraph {nodes [
   n1 = RigVMFunctionEntryNode()
   n2 = RigVMVariableNode() {variable=Controls}
   BranchNode_4_RigVMUnitNode = &RigVMFunction_ControlFlowBranch(
    ExecuteContext=@n1.ExecuteContext
    Condition=&RigVMFunction_MathIntLess(B=4, A=RigVMDispatchNode(Array=@n2.Value) {template="DISPATCH_RigVMDispatch_ArrayGetNum(in Array,out Num)"}.Num) {template="Less::Execute(in A,in B,out Result)"}.Result
   )
   RigVMFunctionReturnNode(
    ExecuteContext=@BranchNode_4_RigVMUnitNode.Completed
    Local Spline=RigVMVariableNode() {variable=local_spline}.Value
    Template Spline=RigVMVariableNode() {variable=template_spline}.Value
   )
   RigUnit_SequenceExecution_1 = RigVMAggregateNode(ExecuteContext=@BranchNode_4_RigVMUnitNode.False)
   RigVMVariableNode(
    Value=&RigUnit_GetTransformItemArray(Space=GlobalSpace, bInitial=true, Items=@n1.Bones).Transforms
    ExecuteContext=@RigUnit_SequenceExecution_1.A
   ) {variable=Original Positions}
   n3 = &RigUnit_GetTransformItemArray(Space=GlobalSpace, bInitial=False, Items=RigVMVariableNode() {variable=Controls}.Value)
   RigVMFunction_MathTransformArrayToSRT = /Script/RigVM.RigVMFunction_MathTransformArrayToSRT(Transforms=@n3.Transforms) {is=struct RigVMUnitNode}
   RigVMVariableNode(
    Value=/Script/ControlRigSpline.RigUnit_ControlRigSplineFromPoints(
     bClosed=False
     Points=@RigVMFunction_MathTransformArrayToSRT.Translations
     SplineMode=RigVMVariableNode() {variable=Spline Mode}.Value
     SamplesPerSegment=RigVMVariableNode() {variable=Samples Per Segment}.Value
    ) {is=struct RigVMUnitNode}.Spline
    ExecuteContext=@RigUnit_SequenceExecution_1.B
   ) {variable=template_spline}
   > RigVMVariableNode(Value=@n3.Transforms) {variable=template_transforms}
   > RigUnit_SequenceExecution = RigVMAggregateNode()
   n4 = RigVMVariableNode() {variable=template_transforms}
   RigVMDispatchNode(
    Num=RigVMDispatchNode(Array=@n4.Value) {template="DISPATCH_RigVMDispatch_ArrayGetNum(in Array,out Num)"}.Num
    ExecuteContext=@RigUnit_SequenceExecution.A
    Array=RigVMVariableNode() {variable=offset_positions}.Value
   ) {template="DISPATCH_RigVMDispatch_ArraySetNum(io Array,in Num)"}
   > n5 = RigVMDispatchNode(Array=@n4.Value) {template="DISPATCH_RigVMDispatch_ArrayIterator(in Array,out Element,out Index,out Count,out Ratio)"}
   > RigVMDispatchNode(
    Index=@n5.Index
    Element=&RigVMFunction_MathTransformTransformVector(Transform=@n5.Element, Location=RigVMVariableNode() {variable=Secondary Spline Direction}.Value).Result
    Array=RigVMVariableNode() {variable=offset_positions}.Value
   ) {template="DISPATCH_RigVMDispatch_ArraySetAtIndex(io Array,in Index,in Element)"}
   RigUnit_SequenceExecution_4 = RigVMAggregateNode(ExecuteContext=@RigUnit_SequenceExecution.C)
   Branch_RigVMUnitNode = &RigVMFunction_ControlFlowBranch(ExecuteContext=@RigUnit_SequenceExecution_4.A, Condition=RigVMVariableNode() {variable=Stretch}.Value)
   n6 = RigVMVariableNode() {variable=template_spline}
   n7 = RigVMVariableNode() {variable=Primary Axis}
   n8 = RigVMVariableNode() {variable=Bones}
   &RigUnit_FitChainToSplineCurveItemArray(
    Alignment=Front
    DebugSettings=(Scale=1,CurveColor=(R=1,G=1,B=0,A=1),SegmentsColor=(R=1,G=0,B=0,A=1),WorldOffset=(Rotation=(X=0,Y=0,Z=0,W=1),Translation=(X=0,Y=0,Z=0),Scale3D=(X=1,Y=1,Z=1)))
    Maximum=1
    PoleVectorPosition=(X=0,Y=0,Z=0)
    Rotations=()
    SamplingPrecision=12
    SecondaryAxis=(X=0,Y=0,Z=0)
    Weight=1
    bPropagateToChildren=True
    ExecutePin=@Branch_RigVMUnitNode.False
    Spline=@n6.Value
    PrimaryAxis=@n7.Value
    Items=@n8.Value
   )
   &RigUnit_FitChainToSplineCurveItemArray(
    Alignment=Stretched
    DebugSettings=(Scale=1,CurveColor=(R=1,G=1,B=0,A=1),SegmentsColor=(R=1,G=0,B=0,A=1),WorldOffset=(Rotation=(X=0,Y=0,Z=0,W=1),Translation=(X=0,Y=0,Z=0),Scale3D=(X=1,Y=1,Z=1)))
    Maximum=1
    PoleVectorPosition=(X=0,Y=0,Z=0)
    Rotations=()
    SamplingPrecision=12
    SecondaryAxis=(X=0,Y=0,Z=0)
    Weight=1
    bPropagateToChildren=True
    ExecutePin=@Branch_RigVMUnitNode.True
    Spline=@n6.Value
    PrimaryAxis=@n7.Value
    Items=@n8.Value
   )
   RigUnit_SequenceExecution_2 = RigVMAggregateNode(ExecuteContext=@RigUnit_SequenceExecution_4.B)
   n9 = RigVMVariableNode(
    Value=&RigUnit_ControlRigSplineFromPoints(
     Stretch=0
     bClosed=False
     SplineMode=RigVMVariableNode() {variable=Spline Mode}.Value
     SamplesPerSegment=RigVMVariableNode() {variable=Samples Per Segment}.Value
     Points=RigVMVariableNode() {variable=offset_positions}.Value
    ).Spline
    ExecuteContext=@RigUnit_SequenceExecution_2.A
   ) {variable=local_spline}
   BranchNode_2_RigVMUnitNode = &RigVMFunction_ControlFlowBranch(ExecuteContext=@n9.ExecuteContext, Condition=RigVMVariableNode() {variable=debug}.Value)
   &RigUnit_DrawControlRigSpline(
    Color=(R=1,G=0,B=0,A=1)
    Detail=16
    Thickness=1
    ExecutePin=@BranchNode_2_RigVMUnitNode.True
    Spline=RigVMVariableNode() {variable=local_spline}.Value
   )
   n10 = RigVMVariableNode() {variable=Bones}
   DISPATCH_RigVMDispatch_ArrayGetNum = RigVMDispatchNode(Array=@n10.Value) {template="DISPATCH_RigVMDispatch_ArrayGetNum(in Array,out Num)"}
   DISPATCH_RigVMDispatch_ArraySetNum = RigVMDispatchNode(
    Num=@DISPATCH_RigVMDispatch_ArrayGetNum.Num
    ExecuteContext=@RigUnit_SequenceExecution_2.B
    Array=RigVMVariableNode() {variable=aim_transforms}.Value
   ) {template="DISPATCH_RigVMDispatch_ArraySetNum(io Array,in Num)"}
   > n11 = RigVMDispatchNode(Array=@n10.Value) {template="DISPATCH_RigVMDispatch_ArrayIterator(in Array,out Element,out Index,out Count,out Ratio)"}
   n12 = &RigUnit_GetTransform(bInitial=false, Item=@n11.Element)
   n13 = RigVMVariableNode() {variable=local_spline}
   > DISPATCH_RigVMDispatch_ArraySetAtIndex = RigVMDispatchNode(
    Index=@n11.Index
    Element=&RigUnit_PositionFromControlRigSpline(
     U=&RigUnit_ClosestParameterFromControlRigSpline(Position=@n12.Transform.Translation, Spline=@n13.Value).U
     Spline=@n13.Value
    ).Position
    Array=RigVMVariableNode() {variable=aim_transforms}.Value
   ) {template="DISPATCH_RigVMDispatch_ArraySetAtIndex(io Array,in Index,in Element)"}
   n14 = RigVMDispatchNode(ExecuteContext=@RigUnit_SequenceExecution.D, Array=RigVMVariableNode() {variable=Bones}.Value) {template="DISPATCH_RigVMDispatch_ArrayIterator(in Array,out Element,out Index,out Count,out Ratio)"}
   DISPATCH_RigVMDispatch_ArrayGetAtIndex = RigVMDispatchNode(Index=@n14.Index, Array=RigVMVariableNode() {variable=aim_transforms}.Value) {template="DISPATCH_RigVMDispatch_ArrayGetAtIndex(in Array,in Index,out Element)"}
   n15 = &RigUnit_AimBoneMath(
    DebugSettings=(Scale=10,WorldOffset=(Rotation=(X=0,Y=0,Z=0,W=1),Translation=(X=0,Y=0,Z=0),Scale3D=(X=1,Y=1,Z=1)))
    Primary=(Weight=1,Axis=(X=0,Y=1,Z=0),Target=(X=1,Y=0,Z=0),Kind=Location,Space=(Type=Bone))
    Secondary=(Weight=0,Axis=(X=0,Y=0,Z=0),Target=(X=0,Y=0,Z=0),Kind=Location,Space=(Type=Bone))
    Weight=1
    Primary.Target=@DISPATCH_RigVMDispatch_ArrayGetAtIndex.Element
    InputTransform=&RigUnit_GetTransform(Item=@n14.Element).Transform
    Primary.Axis=RigVMVariableNode() {variable=Up Axis}.Value
   )
   MathIntEquals = RigVMDispatchNode(A=@n14.Index, B=&RigVMFunction_MathIntSub(B=1, A=@n14.Count) {template="Subtract::Execute(in A,in B,out Result)"}.Result) {template="DISPATCH_RigVMDispatch_CoreEquals(in A,in B,out Result)"}
   > n16 = &RigUnit_SetTransform(Weight=1, bInitial=true, Value=@n15.Result, Item=@n14.Element, bPropagateToChildren=@MathIntEquals.Result)
   BranchNode_3_RigVMUnitNode = &RigVMFunction_ControlFlowBranch(ExecuteContext=@n16.ExecutePin, Condition=RigVMVariableNode() {variable=debug}.Value)
   &RigUnit_DebugTransformMutableItemSpace(
    Color=(R=1,G=0,B=0,A=1)
    DebugDrawSettings=(DepthPriority=SDPG_Foreground,Lifetime=-1)
    Mode=Axes
    Scale=10
    Space=(Type=Bone)
    Thickness=0.15
    WorldOffset=(Rotation=(X=0,Y=0,Z=0,W=1),Translation=(X=0,Y=0,Z=0),Scale3D=(X=1,Y=1,Z=1))
    bEnabled=True
    Transform=@n15.Result
    ExecutePin=@BranchNode_3_RigVMUnitNode.True
   )
   Branch_1_RigVMUnitNode = &RigVMFunction_ControlFlowBranch(ExecuteContext=@RigUnit_SequenceExecution_1.C, Condition=RigVMVariableNode() {variable=Stretch}.Value)
   n17 = RigVMDispatchNode(Index=-1, Array=RigVMVariableNode() {variable=Controls}.Value) {template="DISPATCH_RigVMDispatch_ArrayGetAtIndex(in Array,in Index,out Element)"}
   n18 = RigVMDispatchNode(Index=-1, Array=RigVMVariableNode() {variable=Bones}.Value) {template="DISPATCH_RigVMDispatch_ArrayGetAtIndex(in Array,in Index,out Element)"}
   &RigUnit_ParentConstraint(
    AdvancedSettings=(RotationOrderForFilter=XZY)
    Filter=(TranslationFilter=(bX=false,bY=false,bZ=false),RotationFilter=(bX=True,bY=True,bZ=True),ScaleFilter=(bX=True,bY=True,bZ=True))
    Parents=((Item=(Type=Bone),Weight=1))
    Weight=1
    Parents.0.Item=@n17.Element
    Child=@n18.Element
    ExecutePin=@Branch_1_RigVMUnitNode.False
   )
   &RigUnit_ParentConstraint(
    AdvancedSettings=(RotationOrderForFilter=XZY)
    Parents=((Item=(Type=Bone),Weight=1))
    Weight=1
    Parents.0.Item=@n17.Element
    Child=@n18.Element
    ExecutePin=@Branch_1_RigVMUnitNode.True
   )
   /Script/ControlRig.RigUnit_SetTransformItemArray(
    ExecutePin=()
    Space=GlobalSpace
    Weight=1
    bInitial=true
    bPropagateToChildren=True
    Items=@n1.Bones
    Transforms=RigVMVariableNode() {variable=Original Positions}.Value
   ) {is=struct RigVMUnitNode}
  ]}
 ContainedGraph {nodes [
   n1 = RigVMFunctionEntryNode()
   RigUnit_SequenceExecution_2 = &RigVMFunction_Sequence(ExecuteContext=@n1.ExecuteContext)
   RigUnit_SequenceExecution_2_1 = &RigVMFunction_Sequence(ExecuteContext=@RigUnit_SequenceExecution_2.B)
   RigUnit_SequenceExecution_2_2 = &RigVMFunction_Sequence(ExecuteContext=@RigUnit_SequenceExecution_2_1.B)
   RigVMFunctionReturnNode(A=@RigUnit_SequenceExecution_2.A, B=@RigUnit_SequenceExecution_2_1.A, C=@RigUnit_SequenceExecution_2_2.A, D=@RigUnit_SequenceExecution_2_2.B)
  ]}
 ContainedGraph {nodes [
   n1 = RigVMFunctionEntryNode()
   RigUnit_SequenceExecution_4 = &RigVMFunction_Sequence(ExecuteContext=@n1.ExecuteContext)
   RigUnit_SequenceExecution_4_1 = &RigVMFunction_Sequence(ExecuteContext=@RigUnit_SequenceExecution_4.B)
   RigUnit_SequenceExecution_4_2 = &RigVMFunction_Sequence(ExecuteContext=@RigUnit_SequenceExecution_4_1.B)
   RigVMFunctionReturnNode(A=@RigUnit_SequenceExecution_4.A, B=@RigUnit_SequenceExecution_4_1.A, C=@RigUnit_SequenceExecution_4_2.A, D=@RigUnit_SequenceExecution_4_2.B)
  ]}
 ContainedGraph {nodes [
   n1 = RigVMFunctionEntryNode()
   RigUnit_SequenceExecution = &RigVMFunction_Sequence(ExecuteContext=@n1.ExecuteContext)
   RigUnit_SequenceExecution_1 = &RigVMFunction_Sequence(ExecuteContext=@RigUnit_SequenceExecution.B)
   RigUnit_SequenceExecution_2 = &RigVMFunction_Sequence(ExecuteContext=@RigUnit_SequenceExecution_1.B)
   RigVMFunctionReturnNode(A=@RigUnit_SequenceExecution.A, B=@RigUnit_SequenceExecution_1.A, C=@RigUnit_SequenceExecution_2.A, D=@RigUnit_SequenceExecution_2.B)
  ]}
 ContainedGraph {nodes [
   n1 = RigVMFunctionEntryNode()
   RigUnit_SequenceExecution_1 = &RigVMFunction_Sequence(ExecuteContext=@n1.ExecuteContext)
   RigUnit_SequenceExecution_1_1 = &RigVMFunction_Sequence(ExecuteContext=@RigUnit_SequenceExecution_1.B)
   RigUnit_SequenceExecution_1_2 = &RigVMFunction_Sequence(ExecuteContext=@RigUnit_SequenceExecution_1_1.B)
   RigVMFunctionReturnNode(A=@RigUnit_SequenceExecution_1.A, B=@RigUnit_SequenceExecution_1_1.A, C=@RigUnit_SequenceExecution_1_2.A, D=@RigUnit_SequenceExecution_1_2.B)
  ]}
 ContainedGraph {nodes [
   n1 = RigVMFunctionEntryNode()
   RigUnit_SequenceAggregate = &RigVMFunction_Sequence(ExecuteContext=@n1.ExecuteContext)
   RigVMVariableNode(Value=1000000000, ExecuteContext=@RigUnit_SequenceAggregate.A) {variable=closestDistanceSoFar}
   > n2 = RigVMDispatchNode(Array=@n1.Bones) {template="DISPATCH_RigVMDispatch_ArrayIterator(in Array,out Element,out Index,out Count,out Ratio)"}
   Item = /Script/ControlRig.RigUnit_Item(Item=@n1.Item) {is=struct RigVMUnitNode}
   n3 = &RigVMFunction_MathVectorDistance(
    A=&RigUnit_GetTransform(bInitial=@n1.Initial, Item=@n2.Element).Transform.Translation
    B=&RigUnit_GetTransform(bInitial=@n1.Initial, Item=@Item.Item).Transform.Translation
   )
   n4 = &RigVMFunction_ControlFlowBranch(
    ExecuteContext=@n2.ExecuteContext
    Condition=/Script/RigVM.RigVMFunction_MathFloatLessEqual(A=@n3.Result, B=RigVMVariableNode() {variable=closestDistanceSoFar}.Value) {is=struct RigVMUnitNode, template="LessEqual::Execute(in A,in B,out Result)"}.Result
   )
   RigVMVariableNode(ExecuteContext=@n4.True, Value=@n3.Result) {variable=closestDistanceSoFar}
   > RigVMVariableNode(Value=@n2.Element) {variable=closestBone}
   RigVMFunctionReturnNode(ExecuteContext=@RigUnit_SequenceAggregate.B, Value=RigVMVariableNode() {variable=closestBone}.Value)
  ]}
 ContainedGraph {nodes [
   n1 = RigVMFunctionEntryNode()
   RigUnit_SequenceAggregate = &RigVMFunction_Sequence(ExecuteContext=@n1.ExecuteContext)
   RigVMVariableNode(Value=1000000000, ExecuteContext=@RigUnit_SequenceAggregate.A) {variable=closestDistanceSoFar}
   > n2 = RigVMDispatchNode(Array=@n1.Bones) {template="DISPATCH_RigVMDispatch_ArrayIterator(in Array,out Element,out Index,out Count,out Ratio)"}
   n3 = &RigVMFunction_MathVectorDistance(
    A=&RigUnit_GetTransform(bInitial=@n1.Initial, Item=@n2.Element).Transform.Translation
    B=&RigUnit_GetTransform(bInitial=@n1.Initial, Item=@n1.Item).Transform.Translation
   )
   n4 = &RigVMFunction_ControlFlowBranch(
    ExecuteContext=@n2.ExecuteContext
    Condition=&RigVMFunction_MathFloatLessEqual(A=@n3.Result, B=RigVMVariableNode() {variable=closestDistanceSoFar}.Value) {template="LessEqual::Execute(in A,in B,out Result)"}.Result
   )
   RigVMVariableNode(ExecuteContext=@n4.True, Value=@n3.Result) {variable=closestDistanceSoFar}
   > RigVMVariableNode(Value=@n2.Element) {variable=closestBone}
   RigVMFunctionReturnNode(
    ExecuteContext=@RigUnit_SequenceAggregate.B
    Value=RigVMVariableNode() {variable=closestBone}.Value
    Distance=RigVMVariableNode() {variable=closestDistanceSoFar}.Value
   )
   Item = &RigUnit_Item(Item=@n1.Item)
  ]}
 ContainedGraph {nodes [
   n1 = RigVMVariableNode() {variable=Point 3}
   n2 = RigVMFunctionEntryNode()
   n3 = &RigVMFunction_MathVectorDistance(B=@n2.Point 0, A=@n2.Point 1)
   n4 = &RigVMFunction_MathFloatAdd(A=@n3.Result, B=&RigVMFunction_MathVectorDistance(A=@n2.Point 1, B=@n2.Point 2).Result) {template="Add::Execute(in A,in B,out Result)"}
   n5 = &RigVMFunction_MathFloatAdd(A=@n4.Result, B=&RigVMFunction_MathVectorDistance(A=@n2.Point 2, B=@n2.Point 3).Result) {template="Add::Execute(in A,in B,out Result)"}
   n6 = /Script/RigVM.RigVMFunction_MathFloatDiv(B=@n5.Result, A=@n3.Result) {is=struct RigVMUnitNode, template="Divide::Execute(in A,in B,out Result)"}
   b3t1 = RigVMCollapseNode(t=@n6.Result)
   n7 = RigVMVariableNode() {variable=Point 0}
   b0t1 = RigVMCollapseNode(t=@n6.Result)
   n8 = /Script/RigVM.RigVMFunction_MathVectorScale(Value=@n7.Value, Factor=@b0t1.Result) {is=struct RigVMUnitNode, template="Scale::Execute(in Value,in Factor,out Result)"}
   n9 = RigVMVariableNode() {variable=Point 1}
   n10 = &RigVMFunction_MathFloatDiv(A=@n4.Result, B=@n5.Result) {template="Divide::Execute(in A,in B,out Result)"}
   b3t2 = RigVMCollapseNode(t=@n10.Result)
   b0t2 = RigVMCollapseNode(t=@n10.Result)
   n11 = &RigVMFunction_MathVectorScale(Value=@n7.Value, Factor=@b0t2.Result) {template="Scale::Execute(in Value,in Factor,out Result)"}
   n12 = RigVMVariableNode() {variable=Point 2}
   b1t1 = RigVMCollapseNode(t=@n6.Result)
   b1t2 = RigVMCollapseNode(t=@n10.Result)
   b2t1 = RigVMCollapseNode(t=@n6.Result)
   b2t2 = RigVMCollapseNode(t=@n10.Result)
   Solve = RigVMCollapseNode(
    c=/Script/RigVM.RigVMFunction_MathVectorSub(
     B=&RigVMFunction_MathVectorScale(Value=@n1.Value, Factor=@b3t1.Result) {template="Scale::Execute(in Value,in Factor,out Result)"}.Result
     A=&RigVMFunction_MathVectorSub(B=@n8.Result, A=@n9.Value) {template="Subtract::Execute(in A,in B,out Result)"}.Result
    ) {is=struct RigVMUnitNode, template="Subtract::Execute(in A,in B,out Result)"}.Result
    f=&RigVMFunction_MathVectorSub(
     B=&RigVMFunction_MathVectorScale(Value=@n1.Value, Factor=@b3t2.Result) {template="Scale::Execute(in Value,in Factor,out Result)"}.Result
     A=&RigVMFunction_MathVectorSub(B=@n11.Result, A=@n12.Value) {template="Subtract::Execute(in A,in B,out Result)"}.Result
    ) {template="Subtract::Execute(in A,in B,out Result)"}.Result
    a=@b1t1.Result
    d=@b1t2.Result
    b=@b2t1.Result
    e=@b2t2.Result
   )
   n13 = RigVMVariableNode() {variable=Point 0}
   n14 = RigVMVariableNode() {variable=Point 3}
   RigVMFunctionReturnNode(
    Spline=&RigUnit_ControlRigSplineFromPoints(
     Points=((X=0,Y=0,Z=0),(X=0,Y=0,Z=0),(X=0,Y=0,Z=0),(X=0,Y=0,Z=0))
     SamplesPerSegment=16
     SplineMode=BSpline
     bClosed=False
     Points.1=@Solve.i
     Points.2=@Solve.j
     Points.0=@n13.Value
     Points.3=@n14.Value
    ).Spline
    Out Point 2=@Solve.i
    Out Point 3=@Solve.j
    Out Point 1=@n13.Value
    Out Point 4=@n14.Value
    Execute=@n2.Execute
   )
  ]}
 ContainedGraph {nodes [
   RigVMFunctionEntryNode()
   n1 = &RigVMFunction_MathFloatDiv(
    A=&RigVMFunction_MathFloatMul(A=RigVMVariableNode() {variable=a}.Value, B=RigVMVariableNode() {variable=e}.Value) {template="Multiply::Execute(in A,in B,out Result)"}.Result
    B=RigVMVariableNode() {variable=d}.Value
   ) {template="Divide::Execute(in A,in B,out Result)"}
   n2 = RigVMVariableNode() {variable=b}
   n3 = &RigVMFunction_MathFloatDiv(A=RigVMVariableNode() {variable=a}.Value, B=RigVMVariableNode() {variable=d}.Value) {template="Divide::Execute(in A,in B,out Result)"}
   n4 = RigVMVariableNode() {variable=f}
   n5 = &RigVMFunction_MathVectorScale(
    Factor=&RigVMFunction_MathFloatDiv(
     A=1
     B=/Script/RigVM.RigVMFunction_MathFloatSub(B=@n1.Result, A=@n2.Value) {is=struct RigVMUnitNode, template="Subtract::Execute(in A,in B,out Result)"}.Result
    ) {template="Divide::Execute(in A,in B,out Result)"}.Result
    Value=&RigVMFunction_MathVectorSub(
     B=&RigVMFunction_MathVectorScale(Factor=@n3.Result, Value=@n4.Value) {template="Scale::Execute(in Value,in Factor,out Result)"}.Result
     A=RigVMVariableNode() {variable=c}.Value
    ) {template="Subtract::Execute(in A,in B,out Result)"}.Result
   ) {template="Scale::Execute(in Value,in Factor,out Result)"}
   n6 = RigVMVariableNode() {variable=a}
   n7 = &RigVMFunction_MathVectorScale(Value=@n5.Result, Factor=RigVMVariableNode() {variable=b}.Value) {template="Scale::Execute(in Value,in Factor,out Result)"}
   n8 = RigVMVariableNode() {variable=c}
   RigVMFunctionReturnNode(
    j=@n5.Result
    i=&RigVMFunction_MathVectorScale(
     Factor=&RigVMFunction_MathFloatDiv(A=1, B=@n6.Value) {template="Divide::Execute(in A,in B,out Result)"}.Result
     Value=&RigVMFunction_MathVectorSub(B=@n7.Result, A=@n8.Value) {template="Subtract::Execute(in A,in B,out Result)"}.Result
    ) {template="Scale::Execute(in Value,in Factor,out Result)"}.Result
   )
  ]}
 ContainedGraph {nodes [
   n1 = RigVMFunctionEntryNode()
   n2 = &RigVMFunction_MathFloatSub(A=1, B=@n1.t) {template="Subtract::Execute(in A,in B,out Result)"}
   n3 = &RigVMFunction_MathFloatMul(A=@n2.Result, B=@n2.Result) {template="Multiply::Execute(in A,in B,out Result)"}
   RigVMFunctionReturnNode(Result=&RigVMFunction_MathFloatMul(B=3, A=&RigVMFunction_MathFloatMul(A=@n1.t, B=@n3.Result) {template="Multiply::Execute(in A,in B,out Result)"}.Result) {template="Multiply::Execute(in A,in B,out Result)"}.Result)
  ]}
 ContainedGraph {nodes [
   n1 = RigVMFunctionEntryNode()
   n2 = &RigVMFunction_MathFloatMul(A=@n1.t, B=@n1.t) {template="Multiply::Execute(in A,in B,out Result)"}
   n3 = &RigVMFunction_MathFloatSub(A=1, B=@n1.t) {template="Subtract::Execute(in A,in B,out Result)"}
   RigVMFunctionReturnNode(Result=&RigVMFunction_MathFloatMul(B=3, A=&RigVMFunction_MathFloatMul(B=@n2.Result, A=@n3.Result) {template="Multiply::Execute(in A,in B,out Result)"}.Result) {template="Multiply::Execute(in A,in B,out Result)"}.Result)
  ]}
 ContainedGraph {nodes [
   n1 = RigVMFunctionEntryNode()
   n2 = &RigVMFunction_MathFloatSub(A=1, B=@n1.t) {template="Subtract::Execute(in A,in B,out Result)"}
   n3 = &RigVMFunction_MathFloatMul(A=@n2.Result, B=@n2.Result) {template="Multiply::Execute(in A,in B,out Result)"}
   RigVMFunctionReturnNode(Result=&RigVMFunction_MathFloatMul(B=3, A=&RigVMFunction_MathFloatMul(A=@n1.t, B=@n3.Result) {template="Multiply::Execute(in A,in B,out Result)"}.Result) {template="Multiply::Execute(in A,in B,out Result)"}.Result)
  ]}
 ContainedGraph {nodes [
   n1 = RigVMFunctionEntryNode()
   n2 = &RigVMFunction_MathFloatMul(A=@n1.t, B=@n1.t) {template="Multiply::Execute(in A,in B,out Result)"}
   n3 = &RigVMFunction_MathFloatSub(A=1, B=@n1.t) {template="Subtract::Execute(in A,in B,out Result)"}
   RigVMFunctionReturnNode(Result=&RigVMFunction_MathFloatMul(B=3, A=&RigVMFunction_MathFloatMul(B=@n2.Result, A=@n3.Result) {template="Multiply::Execute(in A,in B,out Result)"}.Result) {template="Multiply::Execute(in A,in B,out Result)"}.Result)
  ]}
 ContainedGraph {nodes [
   n1 = RigVMFunctionEntryNode()
   RigVMFunctionReturnNode(Result=/Script/RigVM.RigVMFunction_MathFloatPow(B=3, A=&RigVMFunction_MathFloatSub(A=1, B=@n1.t) {template="Subtract::Execute(in A,in B,out Result)"}.Result) {is=struct RigVMUnitNode, template="Power::Execute(in A,in B,out Result)"}.Result)
  ]}
 ContainedGraph {nodes [RigVMFunctionReturnNode(Result=&RigVMFunction_MathFloatPow(B=3, A=RigVMFunctionEntryNode().t) {template="Power::Execute(in A,in B,out Result)"}.Result)]}
 ContainedGraph {nodes [
   n1 = RigVMFunctionEntryNode()
   RigVMFunctionReturnNode(Result=&RigVMFunction_MathFloatPow(B=3, A=&RigVMFunction_MathFloatSub(A=1, B=@n1.t) {template="Subtract::Execute(in A,in B,out Result)"}.Result) {template="Power::Execute(in A,in B,out Result)"}.Result)
  ]}
 ContainedGraph {nodes [RigVMFunctionReturnNode(Result=&RigVMFunction_MathFloatPow(B=3, A=RigVMFunctionEntryNode().t) {template="Power::Execute(in A,in B,out Result)"}.Result)]}
 ContainedGraph {nodes [
   n1 = RigVMFunctionEntryNode()
   n2 = RigVMDispatchNode(Num=4, Array=RigVMVariableNode() {variable=Output Points}.Value, ExecuteContext=@n1.Execute) {template="DISPATCH_RigVMDispatch_ArraySetNum(io Array,in Num)"}
   n3 = RigVMVariableNode() {variable=Point 0}
   > n4 = RigVMDispatchNode(Index=0, Array=@n2.Array, Element=@n3.Value) {template="DISPATCH_RigVMDispatch_ArraySetAtIndex(io Array,in Index,in Element)"}
   n5 = RigVMVariableNode() {variable=t}
   n6 = RigVMVariableNode() {variable=t}
   n7 = &RigVMFunction_MathFloatSub(A=1, B=@n6.Value) {template="Subtract::Execute(in A,in B,out Result)"}
   Get A and C = RigVMCollapseNode(order=3, Start Point=@n1.Point 0, Middle Point=@n1.Point 1, End Point=@n1.Point 2, t=@n1.t)
   n8 = RigVMDispatchNode(
    Condition=true
    False=3
    True=&RigVMFunction_MathVectorDistance(B=@"Get A and C".C, A=RigVMVariableNode() {variable=Point 1}.Value).Result
   ) {template="DISPATCH_RigVMDispatch_If(in Condition,in True,in False,out Result)"}
   n9 = /Script/RigVM.RigVMFunction_MathVectorUnit(Value=&RigVMFunction_MathVectorSub(B=RigVMVariableNode() {variable=Point 0}.Value, A=RigVMVariableNode() {variable=Point 2}.Value) {template="Subtract::Execute(in A,in B,out Result)"}.Result) {is=struct RigVMUnitNode, template="Unit::Execute(in Value,out Result)"}
   n10 = RigVMVariableNode() {variable=Point 1}
   n11 = &RigVMFunction_MathVectorSub(
    B=&RigVMFunction_MathVectorScale(Factor=@n8.Result, Value=@n9.Result) {template="Scale::Execute(in Value,in Factor,out Result)"}.Result
    A=@n10.Value
   ) {template="Subtract::Execute(in A,in B,out Result)"}
   n12 = /Script/RigVM.RigVMFunction_MathVectorAdd(
    B=&RigVMFunction_MathVectorScale(
     Factor=&RigVMFunction_MathFloatDiv(A=1, B=@n7.Result) {template="Divide::Execute(in A,in B,out Result)"}.Result
     Value=&RigVMFunction_MathVectorSub(A=@n11.Result, B=@"Get A and C".A) {template="Subtract::Execute(in A,in B,out Result)"}.Result
    ) {template="Scale::Execute(in Value,in Factor,out Result)"}.Result
    A=@"Get A and C".A
   ) {is=struct RigVMUnitNode, template="Add::Execute(in A,in B,out Result)"}
   n13 = RigVMVariableNode() {variable=Point 0}
   n14 = &RigVMFunction_MathVectorAdd(
    B=&RigVMFunction_MathVectorScale(
     Factor=&RigVMFunction_MathFloatDiv(A=1, B=@n5.Value) {template="Divide::Execute(in A,in B,out Result)"}.Result
     Value=&RigVMFunction_MathVectorSub(A=@n12.Result, B=@n13.Value) {template="Subtract::Execute(in A,in B,out Result)"}.Result
    ) {template="Scale::Execute(in Value,in Factor,out Result)"}.Result
    A=@n13.Value
   ) {template="Add::Execute(in A,in B,out Result)"}
   > n15 = RigVMDispatchNode(Index=1, Array=@n4.Array, Element=@n14.Result) {template="DISPATCH_RigVMDispatch_ArraySetAtIndex(io Array,in Index,in Element)"}
   n16 = &RigVMFunction_MathFloatSub(A=1, B=@n5.Value) {template="Subtract::Execute(in A,in B,out Result)"}
   n17 = RigVMVariableNode() {variable=t}
   n18 = &RigVMFunction_MathFloatMul(A=@n8.Result, B=&RigVMFunction_MathFloatSub(A=1, B=@n17.Value) {template="Subtract::Execute(in A,in B,out Result)"}.Result) {template="Multiply::Execute(in A,in B,out Result)"}
   n19 = &RigVMFunction_MathVectorAdd(
    B=&RigVMFunction_MathVectorScale(
     Factor=&RigVMFunction_MathFloatDiv(A=@n18.Result, B=@n17.Value) {template="Divide::Execute(in A,in B,out Result)"}.Result
     Value=@n9.Result
    ) {template="Scale::Execute(in Value,in Factor,out Result)"}.Result
    A=@n10.Value
   ) {template="Add::Execute(in A,in B,out Result)"}
   n20 = &RigVMFunction_MathVectorAdd(
    B=&RigVMFunction_MathVectorScale(
     Factor=&RigVMFunction_MathFloatDiv(A=1, B=@n6.Value) {template="Divide::Execute(in A,in B,out Result)"}.Result
     Value=&RigVMFunction_MathVectorSub(A=@n19.Result, B=@"Get A and C".A) {template="Subtract::Execute(in A,in B,out Result)"}.Result
    ) {template="Scale::Execute(in Value,in Factor,out Result)"}.Result
    A=@"Get A and C".A
   ) {template="Add::Execute(in A,in B,out Result)"}
   n21 = RigVMVariableNode() {variable=Point 2}
   n22 = &RigVMFunction_MathVectorAdd(
    B=&RigVMFunction_MathVectorScale(
     Factor=&RigVMFunction_MathFloatDiv(A=1, B=@n16.Result) {template="Divide::Execute(in A,in B,out Result)"}.Result
     Value=&RigVMFunction_MathVectorSub(A=@n20.Result, B=@n21.Value) {template="Subtract::Execute(in A,in B,out Result)"}.Result
    ) {template="Scale::Execute(in Value,in Factor,out Result)"}.Result
    A=@n21.Value
   ) {template="Add::Execute(in A,in B,out Result)"}
   > n23 = RigVMDispatchNode(Index=2, Array=@n15.Array, Element=@n22.Result) {template="DISPATCH_RigVMDispatch_ArraySetAtIndex(io Array,in Index,in Element)"}
   n24 = RigVMVariableNode() {variable=Point 2}
   > n25 = RigVMDispatchNode(Index=3, Array=@n23.Array, Element=@n24.Value) {template="DISPATCH_RigVMDispatch_ArraySetAtIndex(io Array,in Index,in Element)"}
   RerouteNode = RigVMDispatchNode(
    Values=((X=0,Y=0,Z=0),((X=0,Y=0,Z=0),X=0,Y=0,Z=0),((X=0,Y=0,Z=0),X=0,Y=0,Z=0),((X=0,Y=0,Z=0),X=0,Y=0,Z=0))
    Values.1=@n14.Result
    Values.2=@n22.Result
    Values.3=@n24.Value
    Values.0=@n3.Value
   ) {template="DISPATCH_RigVMDispatch_ArrayMake(in Values,out Array)"}
   RigVMFunctionReturnNode(
    Control Points=@n25.Array
    Execute=@n25.ExecuteContext
    Spline=&RigUnit_ControlRigSplineFromPoints(SamplesPerSegment=16, SplineMode=BSpline, bClosed=False, Points=@RerouteNode.Array).Spline
    Out Point 4=@n24.Value
    Out Point 1=@n3.Value
    Out Point 2=@n22.Result
    Out Point 3=@n14.Result
   )
  ]}
 ContainedGraph {nodes [
   n1 = RigVMFunctionEntryNode()
   Projection Ratio = RigVMCollapseNode(order=@n1.order, t=@n1.t)
   n2 = &RigVMFunction_MathVectorAdd(
    A=&RigVMFunction_MathVectorScale(Factor=@"Projection Ratio".Result, Value=RigVMVariableNode() {variable=Start Point}.Value) {template="Scale::Execute(in Value,in Factor,out Result)"}.Result
    B=&RigVMFunction_MathVectorScale(
     Factor=&RigVMFunction_MathFloatSub(A=1, B=@"Projection Ratio".Result) {template="Subtract::Execute(in A,in B,out Result)"}.Result
     Value=RigVMVariableNode() {variable=End Point}.Value
    ) {template="Scale::Execute(in Value,in Factor,out Result)"}.Result
   ) {template="Add::Execute(in A,in B,out Result)"}
   ABC Ratio = RigVMCollapseNode(order=@n1.order, t=@n1.t)
   n3 = &RigVMFunction_MathFloatDiv(A=1, B=@"ABC Ratio".Result) {template="Divide::Execute(in A,in B,out Result)"}
   n4 = RigVMVariableNode() {variable=Middle Point}
   n5 = &RigVMFunction_MathVectorSub(B=@n2.Result, A=@n4.Value) {template="Subtract::Execute(in A,in B,out Result)"}
   RigVMFunctionReturnNode(
    C=@n2.Result
    A=&RigVMFunction_MathVectorAdd(
     B=&RigVMFunction_MathVectorScale(Factor=@n3.Result, Value=@n5.Result) {template="Scale::Execute(in Value,in Factor,out Result)"}.Result
     A=@n4.Value
    ) {template="Add::Execute(in A,in B,out Result)"}.Result
   )
  ]}
 ContainedGraph {nodes [
   n1 = RigVMFunctionEntryNode()
   n2 = /Script/RigVM.RigVMFunction_MathIntToFloat(Value=@n1.order) {is=struct RigVMUnitNode, template="Cast::Execute(in Value,out Result)"}
   n3 = &RigVMFunction_MathFloatPow(A=&RigVMFunction_MathFloatSub(A=1, B=@n1.t) {template="Subtract::Execute(in A,in B,out Result)"}.Result, B=@n2.Result) {template="Power::Execute(in A,in B,out Result)"}
   n4 = &RigVMFunction_MathFloatPow(A=@n1.t, B=@n2.Result) {template="Power::Execute(in A,in B,out Result)"}
   RigVMFunctionReturnNode(Result=&RigVMFunction_MathFloatDiv(B=&RigVMFunction_MathFloatAdd(A=@n3.Result, B=@n4.Result) {template="Add::Execute(in A,in B,out Result)"}.Result, A=@n3.Result) {template="Divide::Execute(in A,in B,out Result)"}.Result)
  ]}
 ContainedGraph {nodes [
   n1 = RigVMFunctionEntryNode()
   n2 = &RigVMFunction_MathIntToFloat(Value=@n1.order) {template="Cast::Execute(in Value,out Result)"}
   n3 = &RigVMFunction_MathFloatAdd(
    B=&RigVMFunction_MathFloatPow(A=@n1.t, B=@n2.Result) {template="Power::Execute(in A,in B,out Result)"}.Result
    A=&RigVMFunction_MathFloatPow(A=&RigVMFunction_MathFloatSub(A=1, B=@n1.t) {template="Subtract::Execute(in A,in B,out Result)"}.Result, B=@n2.Result) {template="Power::Execute(in A,in B,out Result)"}.Result
   ) {template="Add::Execute(in A,in B,out Result)"}
   n4 = &RigVMFunction_MathFloatSub(B=1, A=@n3.Result) {template="Subtract::Execute(in A,in B,out Result)"}
   RigVMFunctionReturnNode(Result=/Script/RigVM.RigVMFunction_MathFloatAbs(Value=&RigVMFunction_MathFloatDiv(B=@n3.Result, A=@n4.Result) {template="Divide::Execute(in A,in B,out Result)"}.Result) {is=struct RigVMUnitNode, template="Absolute::Execute(in Value,out Result)"}.Result)
  ]}
 ContainedGraph {nodes [
   n1 = RigVMFunctionEntryNode()
   Replace = /Script/RigVM.RigVMFunction_NameReplace(New=None, Name=@n1.Item.Name, Old=@n1.Replace Suffix) {is=struct RigVMUnitNode, template="Replace::Execute(in Name,in Old,in New,out Result)"}
   Concat = &RigVMFunction_NameConcat(B=@n1.New Suffix, A=@Replace.Result) {template="Concat::Execute(in A,in B,out Result)"}
   > n2 = &RigUnit_HierarchyAddNull(
    Space=GlobalSpace
    Name=@Concat.Result
    Transform=&RigUnit_GetTransform(bInitial=true, Item=@n1.Item).Transform
    Parent=/Script/ControlRig.RigUnit_HierarchyGetParent(bDefaultParent=True, Child=@n1.Item) {is=struct RigVMUnitNode}.Parent
   )
   > /Script/ControlRig.RigUnit_SetDefaultParent(Child=@n1.Item, Parent=@n2.Item) {is=struct RigVMUnitNode}
   > RigVMFunctionReturnNode(Null=@n2.Item)
  ]}
 ContainedGraph {nodes [
   n1 = RigVMFunctionEntryNode()
   > SequenceAggregate = RigVMAggregateNode()
   RigVMFunction_Sequence = RigVMAggregateNode(ExecuteContext=@SequenceAggregate.A)
   Concat = &RigVMFunction_NameConcat(B=@n1.Control Suffix, A=@n1.Name) {template="Concat::Execute(in A,in B,out Result)"}
   SpawnControl = /Script/ControlRig.RigUnit_HierarchyAddControlTransform(
    OffsetSpace=LocalSpace
    Parent=(Type=Bone,Name="None")
    Name=@Concat.Result
    OffsetTransform=@n1.Global Position
    Settings=@n1.Settings
    InitialValue=&RigVMFunction_MathTransformMakeRelative(Global=@n1.Global Position, Parent=@n1.Global Position).Local
    ExecutePin=@RigVMFunction_Sequence.A
   ) {
    is=struct RigVMUnitNode
    template="SpawnControl::Execute(in InitialValue,in Settings,in OffsetTransform,in OffsetSpace,in Parent,in Name,out Item)"
   }
   > &RigUnit_SetDefaultParent(Parent=@n1.Parent, Child=@SpawnControl.Item)
   > RigVMVariableNode(Value=@SpawnControl.Item) {variable=the control}
   n2 = &RigVMFunction_ControlFlowBranch(Condition=@n1.Orient to World, ExecuteContext=@RigVMFunction_Sequence.C)
   n3 = RigVMVariableNode() {variable=the control}
   n4 = &RigUnit_GetTransform(bInitial=true, Item=@n3.Value)
   &RigUnit_SetTransform(
    Value=(Rotation=(X=0,Y=0,Z=0,W=1),Translation=(X=0,Y=0,Z=0),Scale3D=(X=1,Y=1,Z=1))
    Weight=1
    bInitial=true
    Value.Scale3D=@n4.Transform.Scale3D
    Value.Translation=@n4.Transform.Translation
    ExecutePin=@n2.True
    Item=@n3.Value
   )
   n5 = &RigVMFunction_ControlFlowBranch(Condition=@n1.Zero Controls, ExecuteContext=@RigVMFunction_Sequence.B)
   &RigUnit_SetTransform(
    Weight=1
    bInitial=true
    Value=&RigUnit_GetTransform(bInitial=true, Item=RigVMVariableNode() {variable=the control}.Value).Transform
    ExecutePin=@n5.True
    Item=RigVMVariableNode() {variable=the control}.Value
   )
   BranchNode_2 = &RigVMFunction_ControlFlowBranch(Condition=@n1.Add Secondary Control, ExecuteContext=@SequenceAggregate.B)
   n6 = &RigUnit_HierarchyAddControlTransform(
    InitialValue=(Rotation=(X=0,Y=0,Z=0,W=1),Translation=(X=0,Y=0,Z=0),Scale3D=(X=1,Y=1,Z=1))
    OffsetSpace=LocalSpace
    OffsetTransform=(Rotation=(X=0,Y=0,Z=0,W=1),Translation=(X=0,Y=0,Z=0),Scale3D=(X=1,Y=1,Z=1))
    ExecutePin=@BranchNode_2.True
    Name=&RigVMFunction_NameConcat(A=&RigVMFunction_NameConcat(B=_sec, A=@n1.Name) {template="Concat::Execute(in A,in B,out Result)"}.Result, B=@n1.Control Suffix) {template="Concat::Execute(in A,in B,out Result)"}.Result
    Settings=@n1.Secondary Settings
    Parent=RigVMVariableNode() {variable=the control}.Value
   ) {template="SpawnControl::Execute(in InitialValue,in Settings,in OffsetTransform,in OffsetSpace,in Parent,in Name,out Item)"}
   > RigVMVariableNode(Value=@n6.Item) {variable=the secondary ctrl}
   n7 = &RigVMFunction_ControlFlowBranch(Condition=@n1.Add Null, ExecuteContext=@SequenceAggregate.C)
   n8 = RigVMVariableNode() {variable=the control}
   Add Null Above = RigVMFunctionReferenceNode(ExecuteContext=@n7.True, Replace Suffix=@n1.Control Suffix, New Suffix=@n1.Null Suffix, Item=@n8.Value)
   > DISPATCH_RigDispatch_SetMetadata = RigVMDispatchNode(Name=Null, NameSpace=Self, Value=@"Add Null Above".Null, Item=@n8.Value) {template="DISPATCH_RigDispatch_SetMetadata(in Item,in Name,in NameSpace,in Value,out Success)"}
   > n9 = RigVMVariableNode(Value=@"Add Null Above".Null) {variable=the null}
   RigVMFunction_ControlFlowBranch = &RigVMFunction_ControlFlowBranch(Condition=@n1.Zero Controls, ExecuteContext=@n9.ExecuteContext)
   &RigUnit_SetTransform(
    Weight=1
    bInitial=true
    Value=&RigUnit_GetTransform(bInitial=true, Item=RigVMVariableNode() {variable=the null}.Value).Transform
    ExecutePin=@RigVMFunction_ControlFlowBranch.True
    Item=RigVMVariableNode() {variable=the control}.Value
   )
   BranchNode_1 = &RigVMFunction_ControlFlowBranch(Condition=@n1.Add Bottom Null, ExecuteContext=@SequenceAggregate.D)
   n10 = &RigUnit_HierarchyAddNull(
    Space=LocalSpace
    Transform=(Rotation=(X=0,Y=0,Z=0,W=1),Translation=(X=0,Y=0,Z=0),Scale3D=(X=1,Y=1,Z=1))
    ExecutePin=@BranchNode_1.True
    Name=&RigVMFunction_NameConcat(B=_ctrl_null, A=@n1.Name) {template="Concat::Execute(in A,in B,out Result)"}.Result
    Parent=RigVMDispatchNode(
     Condition=@n1.Add Secondary Control
     False=RigVMVariableNode() {variable=the control}.Value
     True=RigVMVariableNode() {variable=the secondary ctrl}.Value
    ) {template="DISPATCH_RigVMDispatch_If(in Condition,in True,in False,out Result)"}.Result
   )
   > RigVMDispatchNode(Name=Bottom Null, NameSpace=Self, Value=@n10.Item, Item=RigVMVariableNode() {variable=the control}.Value) {template="DISPATCH_RigDispatch_SetMetadata(in Item,in Name,in NameSpace,in Value,out Success)"}
   > RigVMVariableNode(Value=@n10.Item) {variable=the bottom null}
   RigVMDispatchNode(
    Name=Secondary
    NameSpace=Self
    ExecuteContext=@SequenceAggregate.E
    Item=RigVMVariableNode() {variable=the control}.Value
    Value=RigVMVariableNode() {variable=the secondary ctrl}.Value
   ) {template="DISPATCH_RigDispatch_SetMetadata(in Item,in Name,in NameSpace,in Value,out Success)"}
   > RigVMFunctionReturnNode(
    Control=RigVMVariableNode() {variable=the control}.Value
    Secondary=RigVMVariableNode() {variable=the secondary ctrl}.Value
    Null=RigVMVariableNode() {variable=the null}.Value
    Bottom Null=RigVMVariableNode() {variable=the bottom null}.Value
   )
   n11 = &RigVMFunction_ControlFlowBranch(Condition=@n1.Tag Controls, ExecuteContext=@SequenceAggregate.F)
   /Script/ControlRig.RigUnit_SetMetadataTag(NameSpace=Self, Tag=Controls, ExecutePin=@n11.True, Item=RigVMVariableNode() {variable=the control}.Value) {is=struct RigVMUnitNode}
  ]}
 ContainedGraph {nodes [
   n1 = RigVMFunctionEntryNode()
   SequenceAggregate = &RigVMFunction_Sequence(ExecuteContext=@n1.ExecuteContext)
   SequenceAggregate_1 = &RigVMFunction_Sequence(ExecuteContext=@SequenceAggregate.B)
   SequenceAggregate_2 = &RigVMFunction_Sequence(ExecuteContext=@SequenceAggregate_1.B)
   SequenceAggregate_3 = &RigVMFunction_Sequence(ExecuteContext=@SequenceAggregate_2.B)
   SequenceAggregate_4 = &RigVMFunction_Sequence(ExecuteContext=@SequenceAggregate_3.B)
   RigVMFunctionReturnNode(
    A=@SequenceAggregate.A
    B=@SequenceAggregate_1.A
    C=@SequenceAggregate_2.A
    D=@SequenceAggregate_3.A
    E=@SequenceAggregate_4.A
    F=@SequenceAggregate_4.B
   )
  ]}
 ContainedGraph {nodes [
   n1 = RigVMFunctionEntryNode()
   RigUnit_SequenceAggregate = &RigVMFunction_Sequence(ExecuteContext=@n1.ExecuteContext)
   DISPATCH_RigDispatch_CoreNotEquals = RigVMDispatchNode(B=None, A=@n1.Name) {template="DISPATCH_RigVMDispatch_CoreNotEquals(in A,in B,out Result)"}
   Control Stack at Position = RigVMFunctionReferenceNode(
    Add Bottom Null=@n1.Add Bottom Null
    Add Null=@n1.Add Null
    Add Secondary Control=@n1.Add Secondary Control
    Control Suffix=@n1.Control Suffix
    Null Suffix=@n1.Null Suffix
    Orient to World=@n1.Orient to World
    Parent=@n1.Parent
    Secondary Settings=@n1.Secondary Settings
    Settings=@n1.Settings
    Tag Controls=@n1.Tag Controls
    Zero Controls=@n1.Zero Controls
    Global Position=&RigUnit_GetTransform(bInitial=true, Item=@n1.Item).Transform
    Name=RigVMDispatchNode(Condition=@DISPATCH_RigDispatch_CoreNotEquals.Result, False=@n1.Item.Name, True=@n1.Name) {template="DISPATCH_RigVMDispatch_If(in Condition,in True,in False,out Result)"}.Result
    ExecuteContext=@RigUnit_SequenceAggregate.A
   )
   > RigVMDispatchNode(Name=Item, NameSpace=Self, Item=@"Control Stack at Position".Control, Value=@n1.Item) {template="DISPATCH_RigDispatch_SetMetadata(in Item,in Name,in NameSpace,in Value,out Success)"}
   RigVMFunctionReturnNode(
    Bottom Null=@"Control Stack at Position".Bottom Null
    Control=@"Control Stack at Position".Control
    Null=@"Control Stack at Position".Null
    Secondary=@"Control Stack at Position".Secondary
    ExecuteContext=@RigUnit_SequenceAggregate.B
   )
  ]}
 ContainedGraph {nodes [
   n1 = RigVMFunctionEntryNode()
   RigUnit_SequenceAggregate = &RigVMFunction_Sequence(ExecuteContext=@n1.ExecuteContext)
   n2 = &RigVMFunction_ControlFlowBranch(Condition=@n1.Negative, ExecuteContext=@RigUnit_SequenceAggregate.A)
   RigVMVariableNode(Value=1, ExecuteContext=@n2.False) {variable=isNegative}
   RigVMVariableNode(Value=-1, ExecuteContext=@n2.True) {variable=isNegative}
   n3 = RigVMDispatchNode(
    Array=RigVMDispatchNode(Array=@n1.controls, Other=@n1.ignore_ctrls) {template="DISPATCH_RigVMDispatch_ArrayDifference(in Array,in Other,out Result)"}.Result
    ExecuteContext=@RigUnit_SequenceAggregate.B
   ) {template="DISPATCH_RigVMDispatch_ArrayIterator(in Array,out Element,out Index,out Count,out Ratio)"}
   RigUnit_SequenceAggregate_1_1 = &RigVMFunction_Sequence(ExecuteContext=@n3.ExecuteContext)
   RigUnit_SequenceAggregate_1 = &RigVMFunction_Sequence(ExecuteContext=@RigUnit_SequenceAggregate_1_1.A)
   BranchNode = &RigVMFunction_ControlFlowBranch(Condition=@n1.X Plane, ExecuteContext=@RigUnit_SequenceAggregate_1.A)
   BranchNode_1 = &RigVMFunction_ControlFlowBranch(ExecuteContext=@BranchNode.False, Condition=@n1.Y Plane)
   BranchNode_2 = &RigVMFunction_ControlFlowBranch(ExecuteContext=@BranchNode_1.False, Condition=@n1.Z Plane)
   RigVMVariableNode(
    ExecuteContext=@BranchNode_2.True
    Value=&RigUnit_GetTransform(Item=(Type=Control), bInitial=true, Item.Name=@n3.Element.Name).Transform.Translation.Z
   ) {variable=position}
   RigVMVariableNode(
    ExecuteContext=@BranchNode_1.True
    Value=&RigUnit_GetTransform(Item=(Type=Control), bInitial=true, Item.Name=@n3.Element.Name).Transform.Translation.Y
   ) {variable=position}
   RigVMVariableNode(
    ExecuteContext=@BranchNode.True
    Value=&RigUnit_GetTransform(Item=(Type=Control), bInitial=true, Item.Name=@n3.Element.Name).Transform.Translation.X
   ) {variable=position}
   n4 = &RigVMFunction_MathFloatMul(B=RigVMVariableNode() {variable=isNegative}.Value, A=RigVMVariableNode() {variable=position}.Value) {template="Multiply::Execute(in A,in B,out Result)"}
   n5 = &RigVMFunction_ControlFlowBranch(
    Condition=/Script/RigVM.RigVMFunction_MathFloatGreater(B=0, A=@n4.Result) {is=struct RigVMUnitNode, template="Greater::Execute(in A,in B,out Result)"}.Result
    ExecuteContext=@RigUnit_SequenceAggregate_1.B
   )
   /Script/ControlRig.RigUnit_SetControlColor(Control=@n3.Element.Name, ExecutePin=@n5.True, Color=@n1.LeftColor) {is=struct RigVMUnitNode, template="SetControlColor::Execute(in Control,in Color)"}
   RigUnit_SequenceAggregate_1_2 = &RigVMFunction_Sequence(ExecuteContext=@RigUnit_SequenceAggregate_1_1.B)
   n6 = &RigVMFunction_ControlFlowBranch(
    Condition=/Script/RigVM.RigVMFunction_MathFloatLess(B=0, A=@n4.Result) {is=struct RigVMUnitNode, template="Less::Execute(in A,in B,out Result)"}.Result
    ExecuteContext=@RigUnit_SequenceAggregate_1_2.A
   )
   &RigUnit_SetControlColor(Control=@n3.Element.Name, ExecutePin=@n6.True, Color=@n1.RightColor) {template="SetControlColor::Execute(in Control,in Color)"}
   MathFloatEquals = RigVMDispatchNode(B=0, A=@n4.Result) {template="DISPATCH_RigVMDispatch_CoreEquals(in A,in B,out Result)"}
   n7 = &RigVMFunction_ControlFlowBranch(
    Condition=&RigVMFunction_MathBoolOr(
     B=@MathFloatEquals.Result
     A=/Script/RigVM.RigVMFunction_MathFloatIsNearlyZero(Tolerance=0.001, Value=@n4.Result) {is=struct RigVMUnitNode, template="IsNearlyZero::Execute(in Value,in Tolerance,out Result)"}.Result
    ).Result
    ExecuteContext=@RigUnit_SequenceAggregate_1_2.B
   )
   &RigUnit_SetControlColor(Control=@n3.Element.Name, ExecutePin=@n7.True, Color=@n1.CenterColor) {template="SetControlColor::Execute(in Control,in Color)"}
   RigVMFunctionReturnNode(ExecuteContext=@n3.Completed)
  ]}
 Finalize Controls_1_ContainedGraph {nodes [
   n1 = RigVMFunctionEntryNode()
   n2 = /Script/ControlRig.RigUnit_FindItemsWithMetadataTag(NameSpace=Self, Tag=Controls) {is=struct RigVMUnitNode}
   > Scale Module Controls = RigVMFunctionReferenceNode(
    Global Scale=RigVMDispatchNode(Default=1, Name=Control Size, NameSpace=Root) {template="DISPATCH_RigDispatch_GetModuleMetadata(in Name,in NameSpace,in Default,out Value,out Found)"}.Value
    Local Scale=@n1.Module Settings.ControlSize_12_#4
    Controls=@n2.Items
   )
   > Color Module Controls = RigVMFunctionReferenceNode(
    CenterColor=@n1.Module Settings.CenterColor_9_#5
    LeftColor=@n1.Module Settings.LeftColor_3_#6
    Override=@n1.Module Settings.OverrideColor_48_#7
    RightColor=@n1.Module Settings.RightColor_5_#8
    Controls=@n2.Items
   )
   > RigVMFunctionReturnNode()
  ]}
 Find_ContainedGraph {nodes [
   n1 = RigVMFunctionEntryNode()
   To_String = RigVMDispatchNode(Value=@n1.Value) {template="DISPATCH_RigDispatch_ToString(in Value,out Result)"}
   RigVMFunction_StringSplit = &RigVMFunction_StringSplit(Separator=-, Value=@To_String.Result)
   > RigVMFunctionReturnNode(Success=RigVMDispatchNode(Element=@n1.Element, Array=@RigVMFunction_StringSplit.Result) {template="DISPATCH_RigVMDispatch_ArrayFind(in Array,in Element,out Index,out Success)"}.Success)
  ]}
 Multiply_1_1_ContainedGraph {nodes [
   n1 = RigVMFunctionEntryNode()
   RigVMFunctionReturnNode(Result=/Script/RigVM.RigVMFunction_MathVectorMul(B=@n1.C, A=&RigVMFunction_MathVectorMul(A=@n1.A, B=@n1.B) {template="Multiply::Execute(in A,in B,out Result)"}.Result) {is=struct RigVMUnitNode, template="Multiply::Execute(in A,in B,out Result)"}.Result)
  ]}
 New Function_1_ContainedGraph {nodes [
   n1 = RigVMFunctionEntryNode()
   RigVMFunction_Sequence = &RigVMFunction_Sequence(ExecuteContext=@n1.ExecuteContext)
   n2 = /Script/ControlRig.RigUnit_GetCandidates() {is=struct RigVMUnitNode}
   n3 = &RigVMFunction_ControlFlowBranch(
    Condition=RigVMDispatchNode(B=@n1.Connector, A=@n2.Connector) {template="DISPATCH_RigVMDispatch_CoreEquals(in A,in B,out Result)"}.Result
    ExecuteContext=@RigVMFunction_Sequence.A
   )
   RigVMDispatchNode(Array=@n2.Candidates, ExecuteContext=@n3.True) {template="DISPATCH_RigVMDispatch_ArrayIterator(in Array,out Element,out Index,out Count,out Ratio)"}
   n4 = /Script/ControlRig.RigUnit_CollectionChildrenArray(
    TypeToSearch=Bone
    bDefaultChildren=True
    bIncludeParent=true
    bRecursive=true
    Parent=&RigUnit_ResolveConnector(SkipSocket=true, Connector=@n1.Primary Connector).Result
   ) {is=struct RigVMUnitNode}
   > Clamp Array = RigVMFunctionReferenceNode(Array=@n4.Items, Max=@n1.Clamp)
   > /Script/ControlRig.RigUnit_SetDefaultMatch(Default=RigVMDispatchNode(Index=-1, Array=@n4.Items) {template="DISPATCH_RigVMDispatch_ArrayGetAtIndex(in Array,in Index,out Element)"}.Element) {is=struct RigVMUnitNode, template="Set Default Match::Execute(in Default)"}
   RigVMFunctionReturnNode(ExecuteContext=@RigVMFunction_Sequence.B)
  ]}
 New Function_ContainedGraph {nodes [
   n1 = RigVMFunctionEntryNode()
   n2 = &RigUnit_FindItemsWithMetadataTag(NameSpace=Self, Tag=Controls)
   > Scale Module Controls = RigVMFunctionReferenceNode(
    Global Scale=RigVMDispatchNode(Default=1, Name=Control Size, NameSpace=Root) {template="DISPATCH_RigDispatch_GetModuleMetadata(in Name,in NameSpace,in Default,out Value,out Found)"}.Value
    Local Scale=@n1.Module Settings.ControlSize_12_#4
    Controls=@n2.Items
   )
   > Color Module Controls = RigVMFunctionReferenceNode(
    CenterColor=@n1.Module Settings.CenterColor_9_#5
    LeftColor=@n1.Module Settings.LeftColor_3_#6
    Override=@n1.Module Settings.OverrideColor_48_#7
    RightColor=@n1.Module Settings.RightColor_5_#8
    Controls=@n2.Items
   )
   > RigVMFunctionReturnNode()
  ]}
 New Function_ContainedGraph {nodes [
   n1 = RigVMVariableNode() {variable=Main Controls}
   Control Stack at Position = RigVMFunctionReferenceNode(
    Add Bottom Null=False
    Add Null=true
    Add Secondary Control=False
    Control Suffix=_ctrl
    Name=headneck
    Null Suffix=_null
    Orient to World=true
    Secondary Settings="(Shape=(bVisible=True,Name=\"Default\",Color=(R=1,G=0,B=0,A=1),Transform=(Rotation=(X=0,Y=0,Z=0,W=1),Translation=(X=0,Y=0,Z=0),Scale3D=(X=1,Y=1,Z=1))),Proxy=(bIsProxy=False,ShapeVisibility=BasedOnSelection),DisplayName=\"None\",InitialSpace=LocalSpace,bUsePref ...(truncated)"
    Tag Controls=true
    Zero Controls=true
    Parent=&RigUnit_HierarchyGetParent(
     bDefaultParent=True
     Child=RigVMDispatchNode(Index=-1, Array=@n1.Value) {template="DISPATCH_RigVMDispatch_ArrayGetAtIndex(in Array,in Index,out Element)"}.Element
    ).Parent
    ExecuteContext=@n25.A
    Global Position=/Script/ControlRigSpline.RigUnit_TransformFromControlRigSpline2(
     PrimaryAxis=(X=1,Y=0,Z=0)
     SecondaryAxis=(X=0,Y=0,Z=0)
     U=RigVMVariableNode() {variable=Headneck Position}.Value
     Spline=RigVMVariableNode() {variable=Creation Spline}.Value
    ) {is=struct RigVMUnitNode}.Transform
    Settings=RigVMVariableNode() {variable=HeadNeck Settings}.Value
   )
   Control Stack at Item_1 = RigVMFunctionReferenceNode(
    Add Bottom Null=False
    Add Null=true
    Add Secondary Control=False
    Control Suffix=_ctrl
    ExecuteContext=()
    Name=headlocal
    Null Suffix=_null
    Secondary Settings="(Shape=(bVisible=True,Name=\"Default\",Color=(R=1,G=0,B=0,A=1),Transform=(Rotation=(X=0,Y=0,Z=0,W=1),Translation=(X=0,Y=0,Z=0),Scale3D=(X=1,Y=1,Z=1))),Proxy=(bIsProxy=False,ShapeVisibility=BasedOnSelection),DisplayName=\"None\",InitialSpace=LocalSpace,bUsePref ...(truncated)"
    Tag Controls=true
    Zero Controls=true
    Parent=@"Control Stack at Position".Control
    Item=&RigUnit_ResolveConnector(Connector=(Type=Connector,Name="End Bone"), SkipSocket=true).Result
    Settings=RigVMVariableNode() {variable=Head Local Settings}.Value
   )
   > Set_Module_Metadata_1 = RigVMDispatchNode(Name=HeadLocalControl, NameSpace=Self, Value=@"Control Stack at Item_1".Control) {template="DISPATCH_RigDispatch_SetModuleMetadata(in Name,in NameSpace,in Value,out Success)"}
   n2 = &RigUnit_ResolveConnector(Connector=(Type=Connector,Name="Start Bone"), SkipSocket=False)
   RigVMDispatchNode(ExecuteContext=(), Name=Start Bone, NameSpace=Self, Value=@n2.Result) {template="DISPATCH_RigDispatch_SetModuleMetadata(in Name,in NameSpace,in Value,out Success)"}
   n3 = &RigUnit_ResolveConnector(Connector=(Type=Connector,Name="End Bone"), SkipSocket=False)
   > DISPATCH_RigDispatch_SetModuleMetadata = RigVMDispatchNode(Name=End Bone, NameSpace=Self, Value=@n3.Result) {template="DISPATCH_RigDispatch_SetModuleMetadata(in Name,in NameSpace,in Value,out Success)"}
   > RigVMDispatchNode(
    Name=Bones
    NameSpace=Self
    Value=/Script/ControlRig.RigUnit_HierarchyGetChainItemArray(bIncludeEnd=True, bIncludeStart=True, bReverse=false, End=@n3.Result, Start=@n2.Result) {is=struct RigVMUnitNode}.Chain
   ) {template="DISPATCH_RigDispatch_SetModuleMetadata(in Name,in NameSpace,in Value,out Success)"}
   n4 = RigVMVariableNode() {variable=Rig Bone Positions}
   > RigVMDispatchNode(Array=@n4.Value) {template="DISPATCH_RigVMDispatch_ArrayReset(io Array)"}
   > Set_Num = RigVMDispatchNode(Array=@n4.Value, Num=RigVMVariableNode() {variable=Number of Neck Bones}.Value) {template="DISPATCH_RigVMDispatch_ArraySetNum(io Array,in Num)"}
   > n5 = /Script/RigVM.RigVMFunction_ForLoopCount(Count=RigVMVariableNode() {variable=Number of Neck Bones}.Value) {is=struct RigVMUnitNode}
   n6 = RigVMDispatchNode(Default=(Type=Bone,Name="None"), Name=End Bone, NameSpace=Self) {template="DISPATCH_RigDispatch_GetModuleMetadata(in Name,in NameSpace,in Default,out Value,out Found)"}
   n7 = RigVMDispatchNode(Default=(Type=Bone,Name="None"), Name=Start Bone, NameSpace=Self) {template="DISPATCH_RigDispatch_GetModuleMetadata(in Name,in NameSpace,in Default,out Value,out Found)"}
   > Set_At = RigVMDispatchNode(
    Element=/Script/RigVM.RigVMFunction_MathTransformLerp(
     B=&RigUnit_GetTransform(bInitial=true, Item=@n6.Value).Transform
     A=&RigUnit_GetTransform(bInitial=true, Item=@n7.Value).Transform
     T=@n5.Ratio
    ) {is=struct RigVMUnitNode, template="Interpolate::Execute(in A,in B,in T,out Result)"}.Result
    Index=@n5.Index
    Array=RigVMVariableNode() {variable=Rig Bone Positions}.Value
   ) {template="DISPATCH_RigVMDispatch_ArraySetAtIndex(io Array,in Index,in Element)"}
   Get_Module_Metadata = RigVMDispatchNode(Default=(), Name=Bones, NameSpace=Self) {template="DISPATCH_RigDispatch_GetModuleMetadata(in Name,in NameSpace,in Default,out Value,out Found)"}
   Create Bone Chain = RigVMFunctionReferenceNode(
    Debug=true
    Debug Scale=15
    Debug Thickness=0.5
    Kind=Location
    Optional Suffix=rig
    Parent=&RigUnit_HierarchyGetParent(
     bDefaultParent=True
     Child=RigVMDispatchNode(Index=0, Array=@Get_Module_Metadata.Value) {template="DISPATCH_RigVMDispatch_ArrayGetAtIndex(in Array,in Index,out Element)"}.Element
    ).Parent
    ExecuteContext=@n5.Completed
    Enable Orient Bones=RigVMVariableNode() {variable=Enable Orient Bones}.Value
    Transforms=RigVMVariableNode() {variable=Rig Bone Positions}.Value
    Primary Axis=RigVMVariableNode() {variable=Primary Axis}.Value
    Secondary Axis=RigVMVariableNode() {variable=Secondary Axis}.Value
    Name=RigVMVariableNode() {variable=Module Settings}.Value.Prefix_20_#2
   )
   > RigVMDispatchNode(Name=Rig Bones, NameSpace=Self, Value=@"Create Bone Chain".Bones) {template="DISPATCH_RigDispatch_SetModuleMetadata(in Name,in NameSpace,in Value,out Success)"}
   n8 = RigVMFunctionEntryNode()
   RigVMFunction_Sequence = &RigVMFunction_Sequence(ExecuteContext=@n8.ExecuteContext)
   RigVMVariableNode(Value=@n8.Axis, ExecuteContext=@RigVMFunction_Sequence.A) {variable=Bone Axis}
   > RigVMFunction_Sequence_1 = RigVMAggregateNode()
   n9 = &RigVMFunction_ControlFlowBranch(ExecuteContext=@RigVMFunction_Sequence_1.B, Condition=RigVMVariableNode() {variable=Create Control Per Bone}.Value)
   RigVMFunction_Sequence_2 = RigVMAggregateNode(ExecuteContext=@n9.False)
   LessEqual = &RigVMFunction_MathIntLessEqual(
    B=3
    A=RigVMDispatchNode(Array=RigVMDispatchNode(Default=(), Name=Bones, NameSpace=Self) {template="DISPATCH_RigDispatch_GetModuleMetadata(in Name,in NameSpace,in Default,out Value,out Found)"}.Value) {template="DISPATCH_RigVMDispatch_ArrayGetNum(in Array,out Num)"}.Num
   ) {template="LessEqual::Execute(in A,in B,out Result)"}
   n10 = &RigVMFunction_ControlFlowBranch(Condition=@LessEqual.Result, ExecuteContext=@RigVMFunction_Sequence_2.A)
   SplineFromItems = RigVMFunctionReferenceNode(Samples Per Segment=16, Spline Mode=Hermite, ExecuteContext=@n10.Completed, Items=RigVMVariableNode() {variable=Creation Spline Bones}.Value)
   > n11 = RigVMVariableNode(Value=@SplineFromItems.Spline) {variable=Creation Spline}
   n12 = &RigVMFunction_ControlFlowBranch(ExecuteContext=@n11.ExecuteContext, Condition=RigVMVariableNode() {variable=Debug}.Value)
   &RigUnit_DrawControlRigSpline(Color=(R=1,G=0,B=0,A=1), Detail=16, Thickness=1, ExecutePin=@n12.True, Spline=@SplineFromItems.Spline)
   DISPATCH_RigDispatch_GetModuleMetadata = RigVMDispatchNode(Default=(), Name=Bones, NameSpace=Self) {template="DISPATCH_RigDispatch_GetModuleMetadata(in Name,in NameSpace,in Default,out Value,out Found)"}
   RigVMVariableNode(Value=@DISPATCH_RigDispatch_GetModuleMetadata.Value, ExecuteContext=@n10.False) {variable=Creation Spline Bones}
   n13 = RigVMDispatchNode(Default=(), Name=Bones, NameSpace=Self) {template="DISPATCH_RigDispatch_GetModuleMetadata(in Name,in NameSpace,in Default,out Value,out Found)"}
   Num = RigVMDispatchNode(Array=@n13.Value) {template="DISPATCH_RigVMDispatch_ArrayGetNum(in Array,out Num)"}
   n14 = &RigVMFunction_ControlFlowBranch(
    Condition=RigVMDispatchNode(B=3, A=@Num.Num) {template="DISPATCH_RigVMDispatch_CoreEquals(in A,in B,out Result)"}.Result
    ExecuteContext=@n10.True
   )
   n15 = &RigVMFunction_ControlFlowBranch(
    Condition=RigVMDispatchNode(B=2, A=@Num.Num) {template="DISPATCH_RigVMDispatch_CoreEquals(in A,in B,out Result)"}.Result
    ExecuteContext=@n14.Completed
   )
   n16 = &RigVMFunction_ControlFlowBranch(
    Condition=RigVMDispatchNode(B=1, A=@Num.Num) {template="DISPATCH_RigVMDispatch_CoreEquals(in A,in B,out Result)"}.Result
    ExecuteContext=@n15.Completed
   )
   n17 = RigVMDispatchNode(Index=0, Array=@n13.Value) {template="DISPATCH_RigVMDispatch_ArrayGetAtIndex(in Array,in Index,out Element)"}
   RigVMVariableNode(
    Value=((Type=None,Name="None"),(Type=None,Name="None"),(Type=None,Name="None"),(Type=None,Name="None"))
    Value.0=@n17.Element
    Value.1=@n17.Element
    Value.2=@n17.Element
    Value.3=@n17.Element
    ExecuteContext=@n16.True
   ) {variable=Creation Spline Bones}
   n18 = RigVMDispatchNode(Index=1, Array=@n13.Value) {template="DISPATCH_RigVMDispatch_ArrayGetAtIndex(in Array,in Index,out Element)"}
   RigVMVariableNode(
    Value=((Type=None,Name="None"),(Type=None,Name="None"),(Type=None,Name="None"),(Type=None,Name="None"))
    Value.0=@n17.Element
    Value.1=@n17.Element
    Value.2=@n17.Element
    Value.3=@n18.Element
    ExecuteContext=@n15.True
   ) {variable=Creation Spline Bones}
   RigVMVariableNode(
    Value=((Type=None,Name="None"),(Type=None,Name="None"),(Type=None,Name="None"),(Type=None,Name="None"))
    Value.0=@n17.Element
    Value.1=@n17.Element
    Value.2=@n18.Element
    Value.3=RigVMDispatchNode(Index=2, Array=@n13.Value) {template="DISPATCH_RigVMDispatch_ArrayGetAtIndex(in Array,in Index,out Element)"}.Element
    ExecuteContext=@n14.True
   ) {variable=Creation Spline Bones}
   RigVMDispatchNode(ExecuteContext=@RigVMFunction_Sequence_2.B, Array=RigVMVariableNode() {variable=IK Controls}.Value) {template="DISPATCH_RigVMDispatch_ArrayReset(io Array)"}
   > RigVMDispatchNode(Array=RigVMVariableNode() {variable=U Values}.Value) {template="DISPATCH_RigVMDispatch_ArrayReset(io Array)"}
   > RigVMFunction_ForLoopCount = &RigVMFunction_ForLoopCount(Count=RigVMVariableNode() {variable=Number of Controls}.Value)
   Auto Get Prefix = RigVMFunctionReferenceNode(Value=RigVMDispatchNode(
     Index=0
     Array=RigVMDispatchNode(Default=(), Name=Bones, NameSpace=Self) {template="DISPATCH_RigDispatch_GetModuleMetadata(in Name,in NameSpace,in Default,out Value,out Found)"}.Value
    ) {template="DISPATCH_RigVMDispatch_ArrayGetAtIndex(in Array,in Index,out Element)"}.Element.Name)
   n19 = RigVMVariableNode() {variable=Prefix}
   Add = /Script/RigVM.RigVMFunction_MathIntAdd(B=1, A=@RigVMFunction_ForLoopCount.Index) {is=struct RigVMUnitNode, template="Add::Execute(in A,in B,out Result)"}
   Int to Name = /Script/RigVM.RigVMFunction_MathIntToName(PaddedSize=2, Number=@Add.Result) {is=struct RigVMUnitNode, template="Int to Name::Execute(in Number,in PaddedSize,out Result)"}
   Concat_1 = RigVMAggregateNode(
    B=_cervical_
    A=RigVMDispatchNode(
     True=@"Auto Get Prefix".Result
     Condition=RigVMDispatchNode(B=None, A=@n19.Value) {template="DISPATCH_RigVMDispatch_CoreEquals(in A,in B,out Result)"}.Result
     False=RigVMVariableNode() {variable=Prefix}.Value
    ) {template="DISPATCH_RigVMDispatch_If(in Condition,in True,in False,out Result)"}.Result
    D=@"Int to Name".Result
   )
   > Control Stack at Position_1 = RigVMFunctionReferenceNode(
    Add Bottom Null=False
    Add Null=true
    Add Secondary Control=False
    Null Suffix=_null
    Secondary Settings="(DisplayName=\"None\",Shape=(bVisible=True,Name=\"Default\",Color=(R=1,G=0,B=0,A=1),Transform=(Rotation=(X=0,Y=0,Z=0,W=1),Translation=(X=0,Y=0,Z=0),Scale3D=(X=1,Y=1,Z=1))),Proxy=(bIsProxy=False,ShapeVisibility=BasedOnSelection),InitialSpace=LocalSpace,bUsePref ...(truncated)"
    Tag Controls=true
    Zero Controls=true
    Name=@Concat_1.Result
    Parent=&RigUnit_ResolveConnector(Connector=(Type=Connector,Name="Parent Control"), SkipSocket=true).Result
    Global Position=&RigUnit_TransformFromControlRigSpline2(
     PrimaryAxis=(X=1,Y=0,Z=0)
     SecondaryAxis=(X=0,Y=0,Z=0)
     U=@RigVMFunction_ForLoopCount.Ratio
     Spline=RigVMVariableNode() {variable=Creation Spline}.Value
    ).Transform
    Control Suffix=RigVMVariableNode() {variable=Suffix}.Value
    Settings=RigVMVariableNode() {variable=Inner Settings}.Value
   )
   > RigVMDispatchNode(Name=FK Parent, NameSpace=Self, Value=@"Control Stack at Position_1".Control) {template="DISPATCH_RigDispatch_SetModuleMetadata(in Name,in NameSpace,in Value,out Success)"}
   > RigVMDispatchNode(Element=@RigVMFunction_ForLoopCount.Ratio, Array=RigVMVariableNode() {variable=U Values}.Value) {template="DISPATCH_RigVMDispatch_ArrayAdd(io Array,in Element,out Index)"}
   > RigVMDispatchNode(Element=@"Control Stack at Position_1".Control, Array=RigVMVariableNode() {variable=IK Controls}.Value) {template="DISPATCH_RigVMDispatch_ArrayAdd(io Array,in Element,out Index)"}
   > &RigUnit_SetMetadataTag(NameSpace=Self, Tag=Controls to Hide, Item=@"Control Stack at Position_1".Control)
   RigVMDispatchNode(Name=U Values, NameSpace=Self, Value=RigVMVariableNode() {variable=U Values}.Value, ExecuteContext=@RigVMFunction_ForLoopCount.Completed) {template="DISPATCH_RigDispatch_SetModuleMetadata(in Name,in NameSpace,in Value,out Success)"}
   > Set_Module_Metadata = RigVMDispatchNode(Name=IK Controls, NameSpace=Self, Value=RigVMVariableNode() {variable=IK Controls}.Value) {template="DISPATCH_RigDispatch_SetModuleMetadata(in Name,in NameSpace,in Value,out Success)"}
   > RigVMDispatchNode(Array=RigVMVariableNode() {variable=Main Controls}.Value) {template="DISPATCH_RigVMDispatch_ArrayReset(io Array)"}
   > n20 = &RigVMFunction_ForLoopCount(Count=@n8.Number of Main Controls)
   n21 = RigVMDispatchNode(
    A=@n20.Index
    B=&RigVMFunction_MathIntSub(B=1, A=RigVMVariableNode() {variable=Number of Main Controls}.Value) {template="Subtract::Execute(in A,in B,out Result)"}.Result
   ) {template="DISPATCH_RigVMDispatch_CoreEquals(in A,in B,out Result)"}
   Auto Get Prefix_1 = RigVMFunctionReferenceNode(Value=RigVMDispatchNode(
     Index=0
     Array=RigVMDispatchNode(Default=(), Name=Bones, NameSpace=Self) {template="DISPATCH_RigDispatch_GetModuleMetadata(in Name,in NameSpace,in Default,out Value,out Found)"}.Value
    ) {template="DISPATCH_RigVMDispatch_ArrayGetAtIndex(in Array,in Index,out Element)"}.Element.Name)
   n22 = RigVMVariableNode() {variable=Prefix}
   Int to Name_2 = &RigVMFunction_MathIntToName(PaddedSize=2, Number=&RigVMFunction_MathIntAdd(B=1, A=@n20.Index) {template="Add::Execute(in A,in B,out Result)"}.Result) {template="Int to Name::Execute(in Number,in PaddedSize,out Result)"}
   Concat_1_2 = RigVMAggregateNode(
    B=_tangent_
    C=None
    A=RigVMDispatchNode(
     True=@"Auto Get Prefix_1".Result
     Condition=RigVMDispatchNode(B=None, A=@n22.Value) {template="DISPATCH_RigVMDispatch_CoreEquals(in A,in B,out Result)"}.Result
     False=RigVMVariableNode() {variable=Prefix}.Value
    ) {template="DISPATCH_RigVMDispatch_If(in Condition,in True,in False,out Result)"}.Result
    D=@"Int to Name_2".Result
   )
   > Control Stack at Position_1_1 = RigVMFunctionReferenceNode(
    Add Bottom Null=False
    Add Null=false
    Add Secondary Control=False
    Null Suffix=_null
    Secondary Settings="(DisplayName=\"None\",Shape=(bVisible=True,Name=\"Default\",Color=(R=1,G=0,B=0,A=1),Transform=(Rotation=(X=0,Y=0,Z=0,W=1),Translation=(X=0,Y=0,Z=0),Scale3D=(X=1,Y=1,Z=1))),Proxy=(bIsProxy=False,ShapeVisibility=BasedOnSelection),InitialSpace=LocalSpace,bUsePref ...(truncated)"
    Tag Controls=true
    Zero Controls=true
    Settings=RigVMDispatchNode(
     Condition=@n21.Result
     False=RigVMVariableNode() {variable=Neck Tangent Control Settings}.Value
     True=RigVMVariableNode() {variable=Head Control Settings}.Value
    ) {template="DISPATCH_RigVMDispatch_If(in Condition,in True,in False,out Result)"}.Result
    Orient to World=RigVMDispatchNode(False=False, True=true, Condition=@n21.Result) {template="DISPATCH_RigVMDispatch_If(in Condition,in True,in False,out Result)"}.Result
    Name=RigVMDispatchNode(True=head, False=@Concat_1_2.Result, Condition=@n21.Result) {template="DISPATCH_RigVMDispatch_If(in Condition,in True,in False,out Result)"}.Result
    Parent=&RigUnit_ResolveConnector(Connector=(Type=Connector,Name="Parent Control"), SkipSocket=true).Result
    Global Position=&RigUnit_TransformFromControlRigSpline2(PrimaryAxis=(X=1,Y=0,Z=0), SecondaryAxis=(X=0,Y=0,Z=0), U=@n20.Ratio, Spline=RigVMVariableNode() {variable=Creation Spline}.Value).Transform
    Control Suffix=RigVMVariableNode() {variable=Suffix}.Value
   )
   > RigVMDispatchNode(Element=@"Control Stack at Position_1_1".Control, Array=RigVMVariableNode() {variable=Main Controls}.Value) {template="DISPATCH_RigVMDispatch_ArrayAdd(io Array,in Element,out Index)"}
   > Set_Module_Metadata_2 = RigVMDispatchNode(Name=head_ctrl, NameSpace=Self, Value=@"Control Stack at Position_1_1".Control) {template="DISPATCH_RigDispatch_SetModuleMetadata(in Name,in NameSpace,in Value,out Success)"}
   n23 = RigVMDispatchNode(Name=Main Controls, NameSpace=Self, ExecuteContext=@n20.Completed, Value=RigVMVariableNode() {variable=Main Controls}.Value) {template="DISPATCH_RigDispatch_SetModuleMetadata(in Name,in NameSpace,in Value,out Success)"}
   n24 = &RigVMFunction_ControlFlowBranch(ExecuteContext=@n23.ExecuteContext, Condition=RigVMVariableNode() {variable=Add Stretch Option}.Value)
   n25 = &RigVMFunction_Sequence(ExecuteContext=@n24.Completed)
   /Script/ControlRig.RigUnit_HierarchyAddAnimationChannelBool(
    InitialValue=true
    LimitsEnabled=()
    MaximumValue=True
    MinimumValue=False
    Name=local
    Parent=RigVMDispatchNode(Index=-1, Array=RigVMVariableNode() {variable=Main Controls}.Value) {template="DISPATCH_RigVMDispatch_ArrayGetAtIndex(in Array,in Index,out Element)"}.Element
    ExecutePin=@n25.B
   ) {
    is=struct RigVMUnitNode
    template="SpawnAnimationChannel::Execute(in InitialValue,in MinimumValue,in MaximumValue,in LimitsEnabled,in Parent,in Name,out Item)"
   }
   Get_Module_Metadata_2_1 = RigVMDispatchNode(Default=(Type=Bone,Name="None"), Name=HeadLocalControl, NameSpace=Root) {template="DISPATCH_RigDispatch_GetModuleMetadata(in Name,in NameSpace,in Default,out Value,out Found)"}
   Get_Module_Metadata_2 = RigVMDispatchNode(Default=(Type=Bone,Name="None"), Name=Root Bone, NameSpace=Root) {template="DISPATCH_RigDispatch_GetModuleMetadata(in Name,in NameSpace,in Default,out Value,out Found)"}
   > n26 = &RigUnit_HierarchyAddNull(
    Name=head_global_null
    Space=GlobalSpace
    Transform=(Rotation=(X=0,Y=0,Z=0,W=1),Translation=(X=0,Y=0,Z=0),Scale3D=(X=1,Y=1,Z=1))
    Transform.Translation=&RigUnit_GetTransform(bInitial=true, Item=@Get_Module_Metadata_2_1.Value).Transform.Translation
    Parent=@Get_Module_Metadata_2.Value
   )
   > Set_Module_Metadata_1_1 = RigVMDispatchNode(Name=head_global_null, NameSpace=Self, Value=@n26.Item) {template="DISPATCH_RigDispatch_SetModuleMetadata(in Name,in NameSpace,in Value,out Success)"}
   &RigUnit_HierarchyAddAnimationChannelBool(
    InitialValue=true
    LimitsEnabled=()
    MaximumValue=True
    MinimumValue=False
    Name=Stretch
    Parent=RigVMDispatchNode(Index=-1, Array=RigVMVariableNode() {variable=Main Controls}.Value) {template="DISPATCH_RigVMDispatch_ArrayGetAtIndex(in Array,in Index,out Element)"}.Element
    ExecutePin=@n24.True
   ) {template="SpawnAnimationChannel::Execute(in InitialValue,in MinimumValue,in MaximumValue,in LimitsEnabled,in Parent,in Name,out Item)"}
   RigVMFunction_ControlFlowBranch = &RigVMFunction_ControlFlowBranch(Condition=RigVMVariableNode() {variable=Parent Controls to each other}.Value, ExecuteContext=@RigVMFunction_Sequence_2.C)
   n27 = RigVMVariableNode() {variable=Main Controls}
   &RigUnit_SetDefaultParent(
    Parent=RigVMDispatchNode(Index=0, Array=@n27.Value) {template="DISPATCH_RigVMDispatch_ArrayGetAtIndex(in Array,in Index,out Element)"}.Element
    Child=RigVMDispatchNode(Index=1, Array=@n27.Value) {template="DISPATCH_RigVMDispatch_ArrayGetAtIndex(in Array,in Index,out Element)"}.Element
    ExecutePin=@RigVMFunction_ControlFlowBranch.False
   )
   > &RigUnit_SetDefaultParent(
    Parent=RigVMDispatchNode(Index=3, Array=@n27.Value) {template="DISPATCH_RigVMDispatch_ArrayGetAtIndex(in Array,in Index,out Element)"}.Element
    Child=RigVMDispatchNode(Index=2, Array=@n27.Value) {template="DISPATCH_RigVMDispatch_ArrayGetAtIndex(in Array,in Index,out Element)"}.Element
   )
   Get_Module_Metadata_1 = RigVMDispatchNode(Default=(), Name=Main Controls, NameSpace=Self) {template="DISPATCH_RigDispatch_GetModuleMetadata(in Name,in NameSpace,in Default,out Value,out Found)"}
   DISPATCH_RigVMDispatch_ArrayIterator = RigVMDispatchNode(Array=@Get_Module_Metadata_1.Value, ExecuteContext=@RigVMFunction_ControlFlowBranch.True) {template="DISPATCH_RigVMDispatch_ArrayIterator(in Array,out Element,out Index,out Count,out Ratio)"}
   Subtract = &RigVMFunction_MathIntSub(B=1, A=@DISPATCH_RigVMDispatch_ArrayIterator.Count) {template="Subtract::Execute(in A,in B,out Result)"}
   DISPATCH_RigVMDispatch_CoreEquals = RigVMDispatchNode(A=@DISPATCH_RigVMDispatch_ArrayIterator.Index, B=@Subtract.Result) {template="DISPATCH_RigVMDispatch_CoreEquals(in A,in B,out Result)"}
   RigVMFunction_MathBoolNot = &RigVMFunction_MathBoolNot(Value=@DISPATCH_RigVMDispatch_CoreEquals.Result)
   n28 = &RigVMFunction_ControlFlowBranch(ExecuteContext=@DISPATCH_RigVMDispatch_ArrayIterator.ExecuteContext, Condition=@RigVMFunction_MathBoolNot.Result)
   DISPATCH_RigVMDispatch_ArrayGetAtIndex = RigVMDispatchNode(
    Index=&RigVMFunction_MathIntAdd(B=1, A=@DISPATCH_RigVMDispatch_ArrayIterator.Index) {template="Add::Execute(in A,in B,out Result)"}.Result
    Array=@Get_Module_Metadata_1.Value
   ) {template="DISPATCH_RigVMDispatch_ArrayGetAtIndex(in Array,in Index,out Element)"}
   &RigUnit_SetDefaultParent(Child=@DISPATCH_RigVMDispatch_ArrayGetAtIndex.Element, Parent=@DISPATCH_RigVMDispatch_ArrayIterator.Element, ExecutePin=@n28.True)
   n29 = RigVMDispatchNode(
    Array=RigVMDispatchNode(Default=(), Name=Rig Bones, NameSpace=Self) {template="DISPATCH_RigDispatch_GetModuleMetadata(in Name,in NameSpace,in Default,out Value,out Found)"}.Value
    ExecuteContext=@n9.True
   ) {template="DISPATCH_RigVMDispatch_ArrayIterator(in Array,out Element,out Index,out Count,out Ratio)"}
   n30 = RigVMDispatchNode(B=0, A=@n29.Index) {template="DISPATCH_RigVMDispatch_CoreEquals(in A,in B,out Result)"}
   n31 = &RigVMFunction_ControlFlowBranch(ExecuteContext=@n29.ExecuteContext, Condition=@n30.Result)
   n32 = RigVMDispatchNode(
    Array=RigVMDispatchNode(Default=(), Name=Bones, NameSpace=Self) {template="DISPATCH_RigDispatch_GetModuleMetadata(in Name,in NameSpace,in Default,out Value,out Found)"}.Value
    Index=@n29.Index
   ) {template="DISPATCH_RigVMDispatch_ArrayGetAtIndex(in Array,in Index,out Element)"}
   Concat = &RigVMFunction_NameConcat(B=_ctrl, A=&RigUnit_HierarchyGetParent(bDefaultParent=True, Child=@n32.Element).Parent.Name) {template="Concat::Execute(in A,in B,out Result)"}
   Item = &RigUnit_Item(Item=(Type=Control,Name="None"), Item.Name=@Concat.Result)
   RigVMFunction_MathBoolOr = &RigVMFunction_MathBoolOr(A=RigVMVariableNode() {variable=Add Null Above}.Value, B=RigVMVariableNode() {variable=Add Null}.Value)
   Control Stack at Item = RigVMFunctionReferenceNode(
    Add Bottom Null=False
    Add Secondary Control=False
    Null Suffix=_null
    Secondary Settings="(DisplayName=\"None\",Shape=(bVisible=True,Name=\"Default\",Color=(R=1,G=0,B=0,A=1),Transform=(Rotation=(X=0,Y=0,Z=0,W=1),Translation=(X=0,Y=0,Z=0),Scale3D=(X=1,Y=1,Z=1))),Proxy=(bIsProxy=False,ShapeVisibility=BasedOnSelection),InitialSpace=LocalSpace,bUsePref ...(truncated)"
    Tag Controls=true
    Zero Controls=true
    Name=@n32.Element.Name
    Item=@n29.Element
    Parent=RigVMDispatchNode(
     Condition=@n30.Result
     False=@Item.Item
     True=&RigUnit_ResolveConnector(Connector=(Type=Connector,Name="Parent Control"), SkipSocket=true).Result
    ) {template="DISPATCH_RigVMDispatch_If(in Condition,in True,in False,out Result)"}.Result
    ExecuteContext=@n31.Completed
    Add Null=@RigVMFunction_MathBoolOr.Result
    Control Suffix=RigVMVariableNode() {variable=Suffix}.Value
    Settings=RigVMVariableNode() {variable=Inner Settings}.Value
   )
   > RigVMDispatchNode(Name=FK Parent, NameSpace=Self, Value=@"Control Stack at Item".Control) {template="DISPATCH_RigDispatch_SetModuleMetadata(in Name,in NameSpace,in Value,out Success)"}
   RigVMVariableNode(Value=false, ExecuteContext=@n31.False) {variable=Add Null Above}
   RigVMVariableNode(Value=true, ExecuteContext=@n31.True) {variable=Add Null Above}
   n33 = &RigVMFunction_ControlFlowBranch(ExecuteContext=@RigVMFunction_Sequence_1.C, Condition=RigVMVariableNode() {variable=Dynamics}.Value)
   n34 = &RigVMFunction_Sequence(ExecuteContext=@n33.True)
   n35 = RigVMDispatchNode(Default=(Type=Bone,Name="None"), Name=Settings Control, NameSpace=Root) {template="DISPATCH_RigDispatch_GetModuleMetadata(in Name,in NameSpace,in Default,out Value,out Found)"}
   DISPATCH_RigVMDispatch_If = RigVMDispatchNode(
    Condition=@n35.Found
    True=@n35.Value
    False=RigVMDispatchNode(Index=0, Array=&RigUnit_FindItemsWithMetadataTag(NameSpace=Self, Tag=Controls).Items) {template="DISPATCH_RigVMDispatch_ArrayGetAtIndex(in Array,in Index,out Element)"}.Element
   ) {template="DISPATCH_RigVMDispatch_If(in Condition,in True,in False,out Result)"}
   RigVMDispatchNode(Name=Channel Parent, NameSpace=Self, Value=@DISPATCH_RigVMDispatch_If.Result, ExecuteContext=@n34.B) {template="DISPATCH_RigDispatch_SetModuleMetadata(in Name,in NameSpace,in Value,out Success)"}
   Format Name = RigVMFunctionReferenceNode(name=" Dynamics", side=None, suffix=None, prefix=RigVMVariableNode() {variable=Prefix}.Value)
   > SpawnAnimationChannel = &RigUnit_HierarchyAddAnimationChannelBool(
    InitialValue=true
    LimitsEnabled=()
    MaximumValue=True
    MinimumValue=False
    Parent=@DISPATCH_RigVMDispatch_If.Result
    Name=@"Format Name".Result
   ) {template="SpawnAnimationChannel::Execute(in InitialValue,in MinimumValue,in MaximumValue,in LimitsEnabled,in Parent,in Name,out Item)"}
   > RigVMDispatchNode(Name=Dynamics Channel, NameSpace=Self, Value=(Type=Control,Name="None"), Value.Name=@"Format Name".Result) {template="DISPATCH_RigDispatch_SetModuleMetadata(in Name,in NameSpace,in Value,out Success)"}
   Format Name_1 = RigVMFunctionReferenceNode(name=" Strength", side=None, suffix=None, prefix=RigVMVariableNode() {variable=Prefix}.Value)
   > /Script/ControlRig.RigUnit_HierarchyAddAnimationChannelFloat(
    LimitsEnabled=(Enabled=(bMinimum=True,bMaximum=True))
    MaximumValue=999
    MinimumValue=0
    Parent=@DISPATCH_RigVMDispatch_If.Result
    Name=@"Format Name_1".Result
    InitialValue=RigVMVariableNode() {variable=Dynamics Strength}.Value
   ) {
    is=struct RigVMUnitNode
    template="SpawnAnimationChannel::Execute(in InitialValue,in MinimumValue,in MaximumValue,in LimitsEnabled,in Parent,in Name,out Item)"
   }
   > RigVMDispatchNode(Name=Strength Channel, NameSpace=Self, Value=(Type=Control,Name="None"), Value.Name=@"Format Name_1".Result) {template="DISPATCH_RigDispatch_SetModuleMetadata(in Name,in NameSpace,in Value,out Success)"}
   Format Name_1_1 = RigVMFunctionReferenceNode(name=" Damping", side=None, suffix=None, prefix=RigVMVariableNode() {variable=Prefix}.Value)
   > &RigUnit_HierarchyAddAnimationChannelFloat(
    LimitsEnabled=(Enabled=(bMinimum=True,bMaximum=True))
    MaximumValue=999
    MinimumValue=0
    Parent=@DISPATCH_RigVMDispatch_If.Result
    Name=@"Format Name_1_1".Result
    InitialValue=RigVMVariableNode() {variable=Dynamics Damping}.Value
   ) {template="SpawnAnimationChannel::Execute(in InitialValue,in MinimumValue,in MaximumValue,in LimitsEnabled,in Parent,in Name,out Item)"}
   > RigVMDispatchNode(Name=Damping Channel, NameSpace=Self, Value=(Type=Control,Name="None"), Value.Name=@"Format Name_1_1".Result) {template="DISPATCH_RigDispatch_SetModuleMetadata(in Name,in NameSpace,in Value,out Success)"}
   Format Name_1_2 = RigVMFunctionReferenceNode(name=" Blend Start", side=None, suffix=None, prefix=RigVMVariableNode() {variable=Prefix}.Value)
   > &RigUnit_HierarchyAddAnimationChannelFloat(
    LimitsEnabled=(Enabled=(bMinimum=True,bMaximum=True))
    MaximumValue=999
    MinimumValue=0
    Parent=@DISPATCH_RigVMDispatch_If.Result
    Name=@"Format Name_1_2".Result
    InitialValue=RigVMVariableNode() {variable=Dynamics Blend Start}.Value
   ) {template="SpawnAnimationChannel::Execute(in InitialValue,in MinimumValue,in MaximumValue,in LimitsEnabled,in Parent,in Name,out Item)"}
   > RigVMDispatchNode(Name=Blend Start Channel, NameSpace=Self, Value=(Type=Bone,Name="None"), Value.Name=@"Format Name_1_2".Result) {template="DISPATCH_RigDispatch_SetModuleMetadata(in Name,in NameSpace,in Value,out Success)"}
   Format Name_1_2_1 = RigVMFunctionReferenceNode(name=" Blend End", side=None, suffix=None, prefix=RigVMVariableNode() {variable=Prefix}.Value)
   > &RigUnit_HierarchyAddAnimationChannelFloat(
    LimitsEnabled=(Enabled=(bMinimum=True,bMaximum=True))
    MaximumValue=999
    MinimumValue=0
    Parent=@DISPATCH_RigVMDispatch_If.Result
    Name=@"Format Name_1_2_1".Result
    InitialValue=RigVMVariableNode() {variable=Dynamics Blend End}.Value
   ) {template="SpawnAnimationChannel::Execute(in InitialValue,in MinimumValue,in MaximumValue,in LimitsEnabled,in Parent,in Name,out Item)"}
   > RigVMDispatchNode(Name=Blend End Channel, NameSpace=Self, Value=(Type=Bone,Name="None"), Value.Name=@"Format Name_1_2_1".Result) {template="DISPATCH_RigDispatch_SetModuleMetadata(in Name,in NameSpace,in Value,out Success)"}
   Format Name_1_2_1_1 = RigVMFunctionReferenceNode(name=" Blend Twist", side=None, suffix=None, prefix=RigVMVariableNode() {variable=Prefix}.Value)
   > &RigUnit_HierarchyAddAnimationChannelFloat(
    LimitsEnabled=(Enabled=(bMinimum=True,bMaximum=True))
    MaximumValue=999
    MinimumValue=0
    Parent=@DISPATCH_RigVMDispatch_If.Result
    Name=@"Format Name_1_2_1_1".Result
    InitialValue=RigVMVariableNode() {variable=Dynamics Blend Twist}.Value
   ) {template="SpawnAnimationChannel::Execute(in InitialValue,in MinimumValue,in MaximumValue,in LimitsEnabled,in Parent,in Name,out Item)"}
   > RigVMDispatchNode(Name=Blend Twist Channel, NameSpace=Self, Value=(Type=Bone,Name="None"), Value.Name=@"Format Name_1_2_1_1".Result) {template="DISPATCH_RigDispatch_SetModuleMetadata(in Name,in NameSpace,in Value,out Success)"}
   Create Dual Spline = RigVMFunctionReferenceNode(
    Samples Per Segment=16
    Spline Mode=BSpline
    Stretch=true
    debug=true
    Controls=RigVMDispatchNode(Default=(), Name=IK Controls, NameSpace=Self) {template="DISPATCH_RigDispatch_GetModuleMetadata(in Name,in NameSpace,in Default,out Value,out Found)"}.Value
    Bones=RigVMDispatchNode(Default=(), Name=Rig Bones, NameSpace=Self) {template="DISPATCH_RigDispatch_GetModuleMetadata(in Name,in NameSpace,in Default,out Value,out Found)"}.Value
    Primary Axis=@n8.Primary Axis
    Secondary Spline Direction=@n8.Secondary Spline Direction
    Up Axis=@n8.Up Axis
    ExecuteContext=@RigVMFunction_Sequence_1.D
   )
   > RigVMVariableNode(Value=@"Create Dual Spline".Local Spline) {variable=Local Spline}
   > RigVMVariableNode(Value=@"Create Dual Spline".Template Spline) {variable=Template Spline}
   > For_Each = RigVMDispatchNode(Array=RigVMDispatchNode(Default=(), Name=Rig Bones, NameSpace=Self) {template="DISPATCH_RigDispatch_GetModuleMetadata(in Name,in NameSpace,in Default,out Value,out Found)"}.Value) {template="DISPATCH_RigVMDispatch_ArrayIterator(in Array,out Element,out Index,out Count,out Ratio)"}
   > &RigUnit_SetTransform(
    Weight=1
    bInitial=true
    Item=@For_Each.Element
    Value=&RigUnit_GetTransform(bInitial=False, Item=@For_Each.Element).Transform
   )
   RigVMFunctionReturnNode(ExecuteContext=@RigVMFunction_Sequence.B)
   Int to Name_1 = &RigVMFunction_MathIntToName(PaddedSize=0, Number=&RigVMFunction_MathIntAdd(B=1, A=@n29.Index) {template="Add::Execute(in A,in B,out Result)"}.Result) {template="Int to Name::Execute(in Number,in PaddedSize,out Result)"}
   Concat_1_1 = RigVMAggregateNode(B=_, C=_, D=@"Int to Name_1".Result, A=RigVMVariableNode() {variable=Prefix}.Value)
   &RigUnit_GetTransformItemArray(Space=GlobalSpace, bInitial=False, Items=@Get_Module_Metadata.Value)
   &RigUnit_SetDefaultParent(
    Child=RigVMDispatchNode(Index=-1, Array=RigVMVariableNode() {variable=Main Controls}.Value) {template="DISPATCH_RigVMDispatch_ArrayGetAtIndex(in Array,in Index,out Element)"}.Element
    Parent=@"Control Stack at Position".Control
    ExecutePin=@n39.ExecutePin
   )
   Get_Module_Metadata_3 = RigVMDispatchNode(Default=(Type=Bone,Name="None"), Name=head_ctrl, NameSpace=Self) {template="DISPATCH_RigDispatch_GetModuleMetadata(in Name,in NameSpace,in Default,out Value,out Found)"}
   n36 = &RigUnit_GetRelativeTransformForItem(bChildInitial=False, bParentInitial=False, Parent=@"Control Stack at Position".Control, Child=@Get_Module_Metadata_3.Value)
   n37 = /Script/ControlRig.RigUnit_GetShapeTransform(Control=@"Control Stack at Position".Control.Name) {is=struct RigVMUnitNode}
   n38 = &RigVMFunction_MathDoubleAdd(A=@n36.RelativeTransform.Translation.Z, B=@n37.Transform.Translation.Z)
   n39 = /Script/ControlRig.RigUnit_SetShapeTransform(
    Transform=(Rotation=(X=0,Y=0,Z=0,W=1),Translation=(X=0,Y=0,Z=0),Scale3D=(X=2.5,Y=2.5,Z=2.5))
    Transform.Translation.Z=@n38.Result
    Control=@"Control Stack at Position".Control.Name
    Transform.Translation.X=@n36.RelativeTransform.Translation.X
    Transform.Translation.Y=@n36.RelativeTransform.Translation.Y
    Transform.Rotation=@n37.Transform.Rotation
    Transform.Scale3D=@n37.Transform.Scale3D
    ExecutePin=@Set_Module_Metadata_1_1_1.ExecuteContext
   ) {is=struct RigVMUnitNode}
   Set_Module_Metadata_1_1_1 = RigVMDispatchNode(Name=headneck_ctrl, NameSpace=Self, Value=@"Control Stack at Position".Control, ExecuteContext=@"Control Stack at Position".ExecuteContext) {template="DISPATCH_RigDispatch_SetModuleMetadata(in Name,in NameSpace,in Value,out Success)"}
  ]}
 New Function_ContainedGraph {nodes [
   n1 = RigVMFunctionEntryNode()
   > RigVMFunction_Sequence = RigVMAggregateNode()
   Get_Module_Metadata_3_1_1_1 = RigVMDispatchNode(Default=(Type=Bone,Name="None"), Name=head_ctrl, NameSpace=Self) {template="DISPATCH_RigDispatch_GetModuleMetadata(in Name,in NameSpace,in Default,out Value,out Found)"}
   n2 = &RigVMFunction_ControlFlowBranch(
    Condition=&RigUnit_GetBoolAnimationChannel(Channel=local, bInitial=False, Control=@Get_Module_Metadata_3_1_1_1.Value.Name) {template="GetAnimationChannel::Execute(out Value,in Control,in Channel,in bInitial)"}.Value
    ExecuteContext=@RigVMFunction_Sequence.A
   )
   Get_Module_Metadata_3_1 = RigVMDispatchNode(Default=(Type=Bone,Name="None"), Name=head_global_null, NameSpace=Self) {template="DISPATCH_RigDispatch_GetModuleMetadata(in Name,in NameSpace,in Default,out Value,out Found)"}
   Get_Module_Metadata_3_1_1 = RigVMDispatchNode(Default=(Type=Bone,Name="None"), Name=headneck_ctrl, NameSpace=Self) {template="DISPATCH_RigDispatch_GetModuleMetadata(in Name,in NameSpace,in Default,out Value,out Found)"}
   Set Transform_2 = &RigUnit_SetRotation(
    Space=GlobalSpace
    Weight=1
    bInitial=False
    Value=&RigUnit_GetTransform(bInitial=False, Item=@Get_Module_Metadata_3_1.Value).Transform.Rotation
    Item=&RigUnit_HierarchyGetParent(bDefaultParent=True, Child=@Get_Module_Metadata_3_1_1.Value).Parent
    ExecutePin=@n2.True
   )
   RigVMFunction_Sequence_1 = RigVMAggregateNode(ExecuteContext=@RigVMFunction_Sequence.B)
   RigVMFunction_ControlFlowBranch = &RigVMFunction_ControlFlowBranch(ExecuteContext=@RigVMFunction_Sequence_1.A, Condition=RigVMVariableNode() {variable=Use SplineIK}.Value)
   n3 = RigVMDispatchNode(
    Array=RigVMDispatchNode(Default=(), Name=IK Controls, NameSpace=Self) {template="DISPATCH_RigDispatch_GetModuleMetadata(in Name,in NameSpace,in Default,out Value,out Found)"}.Value
    ExecuteContext=@RigVMFunction_ControlFlowBranch.False
   ) {template="DISPATCH_RigVMDispatch_ArrayIterator(in Array,out Element,out Index,out Count,out Ratio)"}
   DISPATCH_RigDispatch_GetMetadata = RigVMDispatchNode(Default=(Type=Bone,Name="None"), Name=Item, NameSpace=Self, Item=@n3.Element) {template="DISPATCH_RigDispatch_GetMetadata(in Item,in Name,in NameSpace,in Default,out Value,out Found)"}
   > &RigUnit_ParentConstraint(
    AdvancedSettings=(InterpolationType=Average,RotationOrderForFilter=XZY)
    Parents=((Item=(Type=Bone,Name="None"),Weight=1))
    Weight=1
    Child=@DISPATCH_RigDispatch_GetMetadata.Value
    Parents.0.Item=@n3.Element
   )
   > &RigUnit_ParentConstraint(
    AdvancedSettings=(InterpolationType=Average,RotationOrderForFilter=XZY)
    Parents=((Item=(Type=Bone,Name="None"),Weight=1))
    Weight=1
    Child=RigVMDispatchNode(Default=(Type=Bone,Name="None"), Name=Dyn Bone, NameSpace=Self, Item=@DISPATCH_RigDispatch_GetMetadata.Value) {template="DISPATCH_RigDispatch_GetMetadata(in Item,in Name,in NameSpace,in Default,out Value,out Found)"}.Value
    Parents.0.Item=@n3.Element
   )
   n4 = &RigVMFunction_ControlFlowBranch(ExecuteContext=@n3.Completed, Condition=RigVMVariableNode() {variable=Create Control per Bone}.Value)
   DISPATCH_RigVMDispatch_ArrayIterator = RigVMDispatchNode(
    Array=RigVMDispatchNode(Default=(), Name=IK Controls, NameSpace=Self) {template="DISPATCH_RigDispatch_GetModuleMetadata(in Name,in NameSpace,in Default,out Value,out Found)"}.Value
    ExecuteContext=@n4.False
   ) {template="DISPATCH_RigVMDispatch_ArrayIterator(in Array,out Element,out Index,out Count,out Ratio)"}
   > FindClosestBone_2 = RigVMFunctionReferenceNode(
    Initial=true
    Bones=RigVMDispatchNode(Default=(), Name=Bones, NameSpace=Self) {template="DISPATCH_RigDispatch_GetModuleMetadata(in Name,in NameSpace,in Default,out Value,out Found)"}.Value
    Item=@DISPATCH_RigVMDispatch_ArrayIterator.Element
   )
   > Parent = RigVMFunctionReferenceNode(New Parent=@DISPATCH_RigVMDispatch_ArrayIterator.Element, Child=@FindClosestBone_2.Value)
   GetAnimationChannel = &RigUnit_GetBoolAnimationChannel(
    Channel=Stretch
    bInitial=False
    Control=&RigUnit_ResolveConnector(Connector=(Type=Connector,Name="Settings Control"), SkipSocket=true).Result.Name
   ) {template="GetAnimationChannel::Execute(out Value,in Control,in Channel,in bInitial)"}
   RigVMFunction_MathBoolAnd = &RigVMFunction_MathBoolAnd(B=@GetAnimationChannel.Value, A=RigVMVariableNode() {variable=Add Stretch Option}.Value)
   Dual SplineIK = RigVMFunctionReferenceNode(
    Samples Per Segment=16
    Bones=RigVMDispatchNode(Default=(), Name=Rig Bones, NameSpace=Self) {template="DISPATCH_RigDispatch_GetModuleMetadata(in Name,in NameSpace,in Default,out Value,out Found)"}.Value
    Controls=RigVMDispatchNode(Default=(), Name=IK Controls, NameSpace=Self) {template="DISPATCH_RigDispatch_GetModuleMetadata(in Name,in NameSpace,in Default,out Value,out Found)"}.Value
    Primary Axis=@n1.Primary Axis
    Secondary Spline Direction=@n1.Secondary Spline Direction
    Spline Mode=@n1.Spline Mode
    Up Axis=@n1.Up Axis
    debug=@n1.debug
    ExecuteContext=@RigVMFunction_ControlFlowBranch.True
    Stretch=@RigVMFunction_MathBoolAnd.Result
    Aim Spline=RigVMVariableNode() {variable=Local Spline}.Value
    Template Spline=RigVMVariableNode() {variable=Template Spline}.Value
   )
   > RigVMVariableNode(Value=@"Dual SplineIK".Output Spline) {variable=Dual Spline}
   Get_Module_Metadata_2_1 = RigVMDispatchNode(Default=(), Name=IK Controls, NameSpace=Self) {template="DISPATCH_RigDispatch_GetModuleMetadata(in Name,in NameSpace,in Default,out Value,out Found)"}
   > For_Each = RigVMDispatchNode(Array=@Get_Module_Metadata_2_1.Value) {template="DISPATCH_RigVMDispatch_ArrayIterator(in Array,out Element,out Index,out Count,out Ratio)"}
   Get_Module_Metadata_2 = RigVMDispatchNode(Default=(), Name=Main Controls, NameSpace=Self) {template="DISPATCH_RigDispatch_GetModuleMetadata(in Name,in NameSpace,in Default,out Value,out Found)"}
   > Find Closest_0 = RigVMFunctionReferenceNode(Initial=true, Item=@For_Each.Element, Bones=@Get_Module_Metadata_2.Value)
   Clone = RigVMDispatchNode(Array=@Get_Module_Metadata_2.Value) {template="DISPATCH_RigVMDispatch_ArrayClone(in Array,out Clone)"}
   > n5 = RigVMDispatchNode(
    Array=@Clone.Clone
    Index=RigVMDispatchNode(Array=@Clone.Clone, Element=@"Find Closest_0".Value) {template="DISPATCH_RigVMDispatch_ArrayFind(in Array,in Element,out Index,out Success)"}.Index
   ) {template="DISPATCH_RigVMDispatch_ArrayRemove(io Array,in Index)"}
   > Find Closest_0_1 = RigVMFunctionReferenceNode(Initial=true, Item=@For_Each.Element, Bones=@n5.Array)
   n6 = &RigUnit_HierarchyGetParent(bDefaultParent=True, Child=@For_Each.Element)
   IsNearlyZero = /Script/RigVM.RigVMFunction_MathDoubleIsNearlyZero(Tolerance=0, Value=@"Find Closest_0".Distance) {is=struct RigVMUnitNode, template="IsNearlyZero::Execute(in Value,in Tolerance,out Result)"}
   Interpolate = &RigVMFunction_MathTransformLerp(
    T=RigVMDispatchNode(False=0.5, True=1, Condition=@IsNearlyZero.Result) {template="DISPATCH_RigVMDispatch_If(in Condition,in True,in False,out Result)"}.Result
    B=/Script/ControlRig.RigUnit_ProjectTransformToNewParent(
     bChildInitial=True
     bNewParentInitial=False
     bOldParentInitial=True
     NewParent=@"Find Closest_0".Value
     OldParent=@"Find Closest_0".Value
     Child=@n6.Parent
    ) {is=struct RigVMUnitNode}.Transform
    A=&RigUnit_ProjectTransformToNewParent(
     bChildInitial=True
     bNewParentInitial=False
     bOldParentInitial=True
     NewParent=@"Find Closest_0_1".Value
     OldParent=@"Find Closest_0_1".Value
     Child=@n6.Parent
    ).Transform
   ) {template="Interpolate::Execute(in A,in B,in T,out Result)"}
   > &RigUnit_SetTransform(Weight=1, bInitial=False, Item=@n6.Parent, Value=@Interpolate.Result)
   n7 = &RigVMFunction_ControlFlowBranch(ExecuteContext=@RigVMFunction_Sequence_1.B, Condition=RigVMVariableNode() {variable=Dynamics}.Value)
   Local Chain Dynamics = RigVMFunctionReferenceNode(
    Bones=RigVMDispatchNode(Default=(), Name=Rig Bones, NameSpace=Self) {template="DISPATCH_RigDispatch_GetModuleMetadata(in Name,in NameSpace,in Default,out Value,out Found)"}.Value
    Blend End=@n1.Blend End
    Blend Start=@n1.Blend Start
    Blend Twist=@n1.Blend Twist
    Damping=@n1.Damping
    Strength=@n1.Strength
    ExecuteContext=@n7.True
    Draw Debug=RigVMVariableNode() {variable=debug}.Value
   )
   n8 = &RigVMFunction_ControlFlowBranch(ExecuteContext=@RigVMFunction_Sequence_1.D, Condition=RigVMVariableNode() {variable=Attach Skeleton}.Value)
   Get_Module_Metadata_1_2 = RigVMDispatchNode(Default=(), Name=Bones, NameSpace=Self) {template="DISPATCH_RigDispatch_GetModuleMetadata(in Name,in NameSpace,in Default,out Value,out Found)"}
   n9 = RigVMDispatchNode(Array=@Get_Module_Metadata_1_2.Value, ExecuteContext=@n8.True) {template="DISPATCH_RigVMDispatch_ArrayIterator(in Array,out Element,out Index,out Count,out Ratio)"}
   Get_Module_Metadata_4 = RigVMDispatchNode(Default=(), Name=Rig Bones, NameSpace=Self) {template="DISPATCH_RigDispatch_GetModuleMetadata(in Name,in NameSpace,in Default,out Value,out Found)"}
   > FindClosestBone_0_1 = RigVMFunctionReferenceNode(Initial=true, Item=@n9.Element, Bones=@Get_Module_Metadata_4.Value)
   > Parent_0 = RigVMFunctionReferenceNode(Child=@n9.Element, New Parent=@FindClosestBone_0_1.Value)
   RigVMFunctionReturnNode(ExecuteContext=@RigVMFunction_Sequence.C)
   Get_Module_Metadata_4_1 = RigVMDispatchNode(Default=(), Name=Rig Bones, NameSpace=Self) {template="DISPATCH_RigDispatch_GetModuleMetadata(in Name,in NameSpace,in Default,out Value,out Found)"}
   Get_Module_Metadata_3 = RigVMDispatchNode(Default=(Type=Bone,Name="None"), Name=HeadLocalControl, NameSpace=Self) {template="DISPATCH_RigDispatch_GetModuleMetadata(in Name,in NameSpace,in Default,out Value,out Found)"}
   &RigUnit_ParentConstraint(
    AdvancedSettings=(InterpolationType=Average,RotationOrderForFilter=XZY)
    ExecutePin=()
    Parents=((Item=(Type=Bone,Name="None"),Weight=1))
    Weight=1
    Parents.0.Item=RigVMDispatchNode(Index=-1, Array=@Get_Module_Metadata_4_1.Value) {template="DISPATCH_RigVMDispatch_ArrayGetAtIndex(in Array,in Index,out Element)"}.Element
    Child=&RigUnit_HierarchyGetParent(bDefaultParent=True, Child=@Get_Module_Metadata_3.Value).Parent
   )
   Get_Module_Metadata_4_1_1 = RigVMDispatchNode(Default=(), Name=Rig Bones, NameSpace=Self) {template="DISPATCH_RigDispatch_GetModuleMetadata(in Name,in NameSpace,in Default,out Value,out Found)"}
   > &RigUnit_ParentConstraint(
    AdvancedSettings=(InterpolationType=Average,RotationOrderForFilter=XZY)
    Parents=((Item=(Type=Bone,Name="None"),Weight=1))
    Weight=1
    Child=RigVMDispatchNode(Index=-1, Array=@Get_Module_Metadata_4_1_1.Value) {template="DISPATCH_RigVMDispatch_ArrayGetAtIndex(in Array,in Index,out Element)"}.Element
    Parents.0.Item=@Get_Module_Metadata_3.Value
   )
   n10 = &RigVMFunction_ControlFlowBranch(ExecuteContext=(), Condition=RigVMVariableNode() {variable=Attach Skeleton}.Value)
   Get_Module_Metadata = RigVMDispatchNode(Default=(), Name=Rig Bones, NameSpace=Self) {template="DISPATCH_RigDispatch_GetModuleMetadata(in Name,in NameSpace,in Default,out Value,out Found)"}
   n11 = RigVMDispatchNode(Array=@Get_Module_Metadata.Value, ExecuteContext=@n10.True) {template="DISPATCH_RigVMDispatch_ArrayIterator(in Array,out Element,out Index,out Count,out Ratio)"}
   Get_Module_Metadata_1 = RigVMDispatchNode(Default=(), Name=Bones, NameSpace=Self) {template="DISPATCH_RigDispatch_GetModuleMetadata(in Name,in NameSpace,in Default,out Value,out Found)"}
   > Parent_1 = RigVMFunctionReferenceNode(
    Child=RigVMDispatchNode(Index=@n11.Index, Array=@Get_Module_Metadata_1.Value) {template="DISPATCH_RigVMDispatch_ArrayGetAtIndex(in Array,in Index,out Element)"}.Element
    New Parent=@n11.Element
   )
   Num = RigVMDispatchNode(Array=@Get_Module_Metadata.Value) {template="DISPATCH_RigVMDispatch_ArrayGetNum(in Array,out Num)"}
   RigVMDispatchNode(Enabled=True, ScreenColor=(R=1,G=1,B=1,A=1), ScreenDuration=0.05, ExecuteContext=@n11.Completed, Value=@Num.Num) {template="DISPATCH_RigVMDispatch_Print(in Prefix,in Value,in Enabled,in ScreenDuration,in ScreenColor)"}
   > RigVMDispatchNode(
    Enabled=true
    ScreenColor=(R=1,G=1,B=1,A=1)
    ScreenDuration=0.05
    Value=RigVMDispatchNode(Array=@Get_Module_Metadata_1.Value) {template="DISPATCH_RigVMDispatch_ArrayGetNum(in Array,out Num)"}.Num
   ) {template="DISPATCH_RigVMDispatch_Print(in Prefix,in Value,in Enabled,in ScreenDuration,in ScreenColor)"}
   &RigUnit_FindItemsWithMetadataTag(NameSpace=Self, Tag=Controls)
  ]}
 New Function_ContainedGraph {nodes [
   n1 = RigVMFunctionEntryNode()
   Concat_4 = RigVMAggregateNode(B=@n1.name, A=@n1.prefix, C=@n1.side, D=@n1.suffix)
   > RigVMFunctionReturnNode(Result=@Concat_4.Result)
  ]}
 New Function_ContainedGraph {nodes [
   n1 = RigVMFunctionEntryNode()
   RigVMFunction_Sequence = &RigVMFunction_Sequence(ExecuteContext=@n1.ExecuteContext)
   Compute Spline Positions = RigVMFunctionReferenceNode(Debug=false, Array=@n1.Items, ExecuteContext=@RigVMFunction_Sequence.A)
   n2 = RigVMDispatchNode(Array=@n1.Items) {template="DISPATCH_RigVMDispatch_ArrayGetNum(in Array,out Num)"}
   n3 = &RigVMFunction_ControlFlowBranch(
    ExecuteContext=@"Compute Spline Positions".ExecuteContext
    Condition=/Script/RigVM.RigVMFunction_MathIntGreaterEqual(B=4, A=@n2.Num) {is=struct RigVMUnitNode, template="GreaterEqual::Execute(in A,in B,out Result)"}.Result
   )
   n4 = &RigVMFunction_ControlFlowBranch(
    Condition=RigVMDispatchNode(B=3, A=@n2.Num) {template="DISPATCH_RigVMDispatch_CoreEquals(in A,in B,out Result)"}.Result
    ExecuteContext=@n3.Completed
   )
   RigVMFunctionReturnNode(ExecuteContext=@RigVMFunction_Sequence.B, Value=RigVMVariableNode() {variable=Final Points}.Value)
   n5 = RigVMDispatchNode(Index=0, Array=@"Compute Spline Positions".Value) {template="DISPATCH_RigVMDispatch_ArrayGetAtIndex(in Array,in Index,out Element)"}
   n6 = RigVMDispatchNode(Index=1, Array=@"Compute Spline Positions".Value) {template="DISPATCH_RigVMDispatch_ArrayGetAtIndex(in Array,in Index,out Element)"}
   n7 = RigVMDispatchNode(Index=2, Array=@"Compute Spline Positions".Value) {template="DISPATCH_RigVMDispatch_ArrayGetAtIndex(in Array,in Index,out Element)"}
   SplineThrough3Points = RigVMFunctionReferenceNode(t=0.5, Point 0=@n5.Element, Point 1=@n6.Element, Point 2=@n7.Element, Execute=@n4.True)
   RigVMVariableNode(
    Value=((X=0,Y=0,Z=0),(X=0,Y=0,Z=0),(X=0,Y=0,Z=0),(X=0,Y=0,Z=0))
    ExecuteContext=@SplineThrough3Points.Execute
    Value.0=@SplineThrough3Points.Out Point 1
    Value.1=@SplineThrough3Points.Out Point 2
    Value.2=@SplineThrough3Points.Out Point 3
    Value.3=@SplineThrough3Points.Out Point 4
   ) {variable=Final Points}
   SplineThrough4Points = RigVMFunctionReferenceNode(
    Point 0=@n5.Element
    Point 1=@n6.Element
    Point 2=@n7.Element
    Point 3=RigVMDispatchNode(Index=3, Array=@"Compute Spline Positions".Value) {template="DISPATCH_RigVMDispatch_ArrayGetAtIndex(in Array,in Index,out Element)"}.Element
    Execute=@n3.True
   )
   RigVMVariableNode(
    Value=((X=0,Y=0,Z=0),(X=0,Y=0,Z=0),(X=0,Y=0,Z=0),(X=0,Y=0,Z=0))
    ExecuteContext=@SplineThrough4Points.Execute
    Value.0=@SplineThrough4Points.Out Point 1
    Value.1=@SplineThrough4Points.Out Point 2
    Value.2=@SplineThrough4Points.Out Point 3
    Value.3=@SplineThrough4Points.Out Point 4
   ) {variable=Final Points}
  ]}
 New Function_ContainedGraph {nodes [
   n1 = RigVMFunctionEntryNode()
   > For_Each_1 = RigVMDispatchNode(Array=@n1.Items) {template="DISPATCH_RigVMDispatch_ArrayIterator(in Array,out Element,out Index,out Count,out Ratio)"}
   n2 = &RigVMFunction_ControlFlowBranch(
    Condition=RigVMDispatchNode(
     A=@For_Each_1.Index
     B=&RigVMFunction_MathIntSub(B=1, A=@For_Each_1.Count) {template="Subtract::Execute(in A,in B,out Result)"}.Result
    ) {template="DISPATCH_RigVMDispatch_CoreEquals(in A,in B,out Result)"}.Result
    ExecuteContext=@For_Each_1.ExecuteContext
   )
   n3 = &RigUnit_GetTransform(bInitial=False, Item=@For_Each_1.Element)
   n4 = &RigUnit_GetTransform(
    bInitial=False
    Item=RigVMDispatchNode(
     Index=&RigVMFunction_MathIntAdd(B=1, A=@For_Each_1.Index) {template="Add::Execute(in A,in B,out Result)"}.Result
     Array=@n1.Items
    ) {template="DISPATCH_RigVMDispatch_ArrayGetAtIndex(in Array,in Index,out Element)"}.Element
   )
   RigVMFunction_DebugLineNoSpace = /Script/RigVM.RigVMFunction_DebugLineNoSpace(
    DebugDrawSettings=(DepthPriority=SDPG_Foreground,Lifetime=-1)
    Color=@n1.Color
    bEnabled=@n1.Enabled
    Thickness=@n1.Thickness
    WorldOffset=@n1.World Offset
    A=@n3.Transform.Translation
    B=@n4.Transform.Translation
    ExecutePin=@n2.False
   ) {is=struct RigVMUnitNode}
   RigVMFunctionReturnNode(ExecuteContext=@For_Each_1.Completed)
   Add = &RigVMFunction_MathVectorAdd(B=(X=0,Y=-30,Z=0), A=@n3.Transform.Translation) {template="Add::Execute(in A,in B,out Result)"}
   &RigVMFunction_MathVectorAdd(B=(X=0,Y=-30,Z=0), A=@n4.Transform.Translation) {template="Add::Execute(in A,in B,out Result)"}
  ]}
 New Function_ContainedGraph {nodes [
   n1 = RigVMFunctionEntryNode()
   > RigVMDispatchNode(Array=RigVMVariableNode() {variable=Tail FK Controls}.Value) {template="DISPATCH_RigVMDispatch_ArrayReset(io Array)"}
   > n2 = RigVMDispatchNode(Array=&RigUnit_HierarchyGetChainItemArray(bIncludeEnd=True, bIncludeStart=True, bReverse=False, End=@n1.End, Start=@n1.Start).Chain) {template="DISPATCH_RigVMDispatch_ArrayIterator(in Array,out Element,out Index,out Count,out Ratio)"}
   n3 = &RigUnit_HierarchyGetParent(bDefaultParent=True, Child=@n2.Element)
   > Control Stack at Item_1 = RigVMFunctionReferenceNode(
    Add Bottom Null=False
    Add Null=true
    Add Secondary Control=False
    Control Suffix=_ctrl
    Null Suffix=_null
    Secondary Settings="(DisplayName=\"None\",Shape=(bVisible=True,Name=\"Default\",Color=(R=1,G=0,B=0,A=1),Transform=(Rotation=(X=0,Y=0,Z=0,W=1),Translation=(X=0,Y=0,Z=0),Scale3D=(X=1,Y=1,Z=1))),Proxy=(bIsProxy=False,ShapeVisibility=BasedOnSelection),InitialSpace=LocalSpace,bSelecta ...(truncated)"
    Tag Controls=true
    Zero Controls=true
    Name=&RigVMFunction_NameConcat(B=_fk, A=@n2.Element.Name) {template="Concat::Execute(in A,in B,out Result)"}.Result
    Item=@n2.Element
    Settings=@n1.Settings
    Parent=RigVMDispatchNode(
     False=(Type=Control,Name="None")
     False.Name=&RigVMFunction_NameConcat(B=_fk_ctrl, A=@n3.Parent.Name) {template="Concat::Execute(in A,in B,out Result)"}.Result
     Condition=RigVMDispatchNode(B=0, A=@n2.Index) {template="DISPATCH_RigVMDispatch_CoreEquals(in A,in B,out Result)"}.Result
     True=&RigUnit_ResolveConnector(Connector=(Type=Connector,Name="Parent Control"), SkipSocket=true).Result
    ) {template="DISPATCH_RigVMDispatch_If(in Condition,in True,in False,out Result)"}.Result
   )
   > &RigUnit_SetTransform(
    Weight=1
    bInitial=true
    Item=@"Control Stack at Item_1".Control
    Value=&RigUnit_GetTransform(bInitial=true, Item=@"Control Stack at Item_1".Control).Transform
   )
   > RigVMDispatchNode(Element=@"Control Stack at Item_1".Control, Array=RigVMVariableNode() {variable=Tail FK Controls}.Value) {template="DISPATCH_RigVMDispatch_ArrayAdd(io Array,in Element,out Index)"}
   Set_Module_Metadata = RigVMDispatchNode(Name=FK Controls, NameSpace=Self, ExecuteContext=@n2.Completed, Value=RigVMVariableNode() {variable=Tail FK Controls}.Value) {template="DISPATCH_RigDispatch_SetModuleMetadata(in Name,in NameSpace,in Value,out Success)"}
   > RigVMFunctionReturnNode(FK Controls=RigVMVariableNode() {variable=Tail FK Controls}.Value)
  ]}
 New Function_ContainedGraph {nodes [
   n1 = RigVMFunctionEntryNode()
   Concat_4 = RigVMAggregateNode(B=@n1.name, A=@n1.prefix, C=@n1.side, D=@n1.suffix)
   > RigVMFunctionReturnNode(Result=@Concat_4.Result)
  ]}
 New Function_ContainedGraph {nodes [
   n1 = RigVMFunctionEntryNode()
   Format Name_2 = RigVMFunctionReferenceNode(suffix=_fk_ik_switch, name=@n1.name, prefix=@n1.prefix, side=@n1.side)
   > n2 = &RigUnit_HierarchyAddAnimationChannelBool(
    LimitsEnabled=()
    MaximumValue=True
    MinimumValue=False
    InitialValue=@n1.Initial Value
    Name=@"Format Name_2".Result
    Parent=&RigUnit_ResolveConnector(Connector=(Type=Connector,Name="Settings Control"), SkipSocket=true).Result
   ) {template="SpawnAnimationChannel::Execute(in InitialValue,in MinimumValue,in MaximumValue,in LimitsEnabled,in Parent,in Name,out Item)"}
   > RigVMDispatchNode(Name=fk_ik_switch, NameSpace=Self, Value=@n2.Item.Name) {template="DISPATCH_RigDispatch_SetModuleMetadata(in Name,in NameSpace,in Value,out Success)"}
   > RigVMFunctionReturnNode()
  ]}
 New Function_ContainedGraph {nodes [
   n1 = RigVMFunctionEntryNode()
   > &RigUnit_SetTransform(
    Weight=1
    bInitial=False
    Item=@n1.Child
    Value=&RigUnit_ProjectTransformToNewParent(bChildInitial=True, bNewParentInitial=False, bOldParentInitial=True, Child=@n1.Child, NewParent=@n1.New Parent, OldParent=@n1.New Parent).Transform
   )
   > RigVMFunctionReturnNode()
  ]}
 New Function_ContainedGraph {nodes [
   RigVMFunctionEntryNode()
   Get_Module_Metadata = RigVMDispatchNode(Default=(), Name=FK Controls, NameSpace=Self) {template="DISPATCH_RigDispatch_GetModuleMetadata(in Name,in NameSpace,in Default,out Value,out Found)"}
   > n1 = RigVMDispatchNode(Array=@Get_Module_Metadata.Value) {template="DISPATCH_RigVMDispatch_ArrayIterator(in Array,out Element,out Index,out Count,out Ratio)"}
   > Parent_3_1_1_1_1_1 = RigVMFunctionReferenceNode(
    Child=RigVMDispatchNode(Default=(Type=Bone,Name="None"), Name=Item, NameSpace=Self, Item=@n1.Element) {template="DISPATCH_RigDispatch_GetMetadata(in Item,in Name,in NameSpace,in Default,out Value,out Found)"}.Value
    New Parent=@n1.Element
   )
   RigVMFunctionReturnNode(ExecuteContext=@n1.Completed)
  ]}
 New Function_ContainedGraph {nodes [
   n1 = RigVMFunctionEntryNode()
   DISPATCH_RigDispatch_GetModuleMetadata = RigVMDispatchNode(Default=None, NameSpace=Self, Name=@n1.Name) {template="DISPATCH_RigDispatch_GetModuleMetadata(in Name,in NameSpace,in Default,out Value,out Found)"}
   GetAnimationChannel = &RigUnit_GetBoolAnimationChannel(
    bInitial=False
    Channel=@DISPATCH_RigDispatch_GetModuleMetadata.Value
    Control=&RigUnit_ResolveConnector(SkipSocket=true, Connector=@n1.Connector).Result.Name
   ) {template="GetAnimationChannel::Execute(out Value,in Control,in Channel,in bInitial)"}
   > RigVMFunctionReturnNode(Value=@GetAnimationChannel.Value)
  ]}
 New Function_ContainedGraph {nodes [
   n1 = RigVMFunctionEntryNode()
   > RigVMFunction_Sequence = RigVMAggregateNode()
   RigVMFunction_Sequence_1 = RigVMAggregateNode(ExecuteContext=@RigVMFunction_Sequence.A)
   DISPATCH_RigVMDispatch_ArrayIterator = RigVMDispatchNode(Array=@n1.Bones, ExecuteContext=@RigVMFunction_Sequence_1.A) {template="DISPATCH_RigVMDispatch_ArrayIterator(in Array,out Element,out Index,out Count,out Ratio)"}
   > DISPATCH_RigDispatch_SetMetadata = RigVMDispatchNode(
    Name=Orig Parent
    NameSpace=Self
    Item=@DISPATCH_RigVMDispatch_ArrayIterator.Element
    Value=&RigUnit_HierarchyGetParent(bDefaultParent=True, Child=@DISPATCH_RigVMDispatch_ArrayIterator.Element).Parent
   ) {template="DISPATCH_RigDispatch_SetMetadata(in Item,in Name,in NameSpace,in Value,out Success)"}
   n2 = RigVMVariableNode() {variable=Bones}
   > &RigUnit_SetDefaultParent(
    Child=@DISPATCH_RigVMDispatch_ArrayIterator.Element
    Parent=&RigUnit_HierarchyGetParent(
     bDefaultParent=True
     Child=RigVMDispatchNode(Index=0, Array=@n2.Value) {template="DISPATCH_RigVMDispatch_ArrayGetAtIndex(in Array,in Index,out Element)"}.Element
    ).Parent
   )
   DISPATCH_RigVMDispatch_ArrayGetNum = RigVMDispatchNode(Array=@n1.Bones) {template="DISPATCH_RigVMDispatch_ArrayGetNum(in Array,out Num)"}
   Interpolate = &RigVMFunction_MathTransformLerp(
    T=0.5
    A=&RigUnit_GetTransform(
     bInitial=False
     Item=RigVMDispatchNode(Index=0, Array=@n1.Bones) {template="DISPATCH_RigVMDispatch_ArrayGetAtIndex(in Array,in Index,out Element)"}.Element
    ).Transform
    B=&RigUnit_GetTransform(
     bInitial=False
     Item=RigVMDispatchNode(Index=-1, Array=@n1.Bones) {template="DISPATCH_RigVMDispatch_ArrayGetAtIndex(in Array,in Index,out Element)"}.Element
    ).Transform
   ) {template="Interpolate::Execute(in A,in B,in T,out Result)"}
   DISPATCH_RigVMDispatch_If = RigVMDispatchNode(
    False=(X=0,Y=0,Z=0)
    Condition=RigVMDispatchNode(B=3, A=@DISPATCH_RigVMDispatch_ArrayGetNum.Num) {template="DISPATCH_RigVMDispatch_CoreEquals(in A,in B,out Result)"}.Result
    True=@Interpolate.Result.Translation
   ) {template="DISPATCH_RigVMDispatch_If(in Condition,in True,in False,out Result)"}
   RigVMFunction_DebugTransformMutableNoSpace = &RigVMFunction_DebugTransformMutableNoSpace(
    Color=(R=1,G=0,B=0,A=1)
    DebugDrawSettings=(DepthPriority=SDPG_Foreground,Lifetime=-1)
    Mode=Point
    Scale=10
    Thickness=1
    Transform=(Rotation=(X=0,Y=0,Z=0,W=1),Translation=(X=0,Y=0,Z=0),Scale3D=(X=1,Y=1,Z=1))
    WorldOffset=(Rotation=(X=0,Y=0,Z=0,W=1),Translation=(X=0,Y=0,Z=0),Scale3D=(X=1,Y=1,Z=1))
    Transform.Translation=@DISPATCH_RigVMDispatch_If.Result
    Transform.Rotation=@Interpolate.Result.Rotation
    ExecutePin=@RigVMFunction_Sequence_1.B
    bEnabled=RigVMVariableNode() {variable=Debug}.Value
   )
   > RigVMVariableNode(Value=@DISPATCH_RigVMDispatch_If.Result) {variable=Secondary Target}
   n3 = RigVMDispatchNode(Array=@n1.Bones, ExecuteContext=@RigVMFunction_Sequence_1.C) {template="DISPATCH_RigVMDispatch_ArrayIterator(in Array,out Element,out Index,out Count,out Ratio)"}
   Subtract = &RigVMFunction_MathIntSub(B=1, A=@n3.Count) {template="Subtract::Execute(in A,in B,out Result)"}
   DISPATCH_RigVMDispatch_CoreEquals = RigVMDispatchNode(A=@n3.Index, B=@Subtract.Result) {template="DISPATCH_RigVMDispatch_CoreEquals(in A,in B,out Result)"}
   n4 = &RigVMFunction_ControlFlowBranch(ExecuteContext=@n3.ExecuteContext, Condition=@DISPATCH_RigVMDispatch_CoreEquals.Result)
   Add = &RigVMFunction_MathIntAdd(B=1, A=@n3.Index) {template="Add::Execute(in A,in B,out Result)"}
   n5 = &RigUnit_AimBoneMath(
    DebugSettings=(bEnabled=False,Scale=10,WorldOffset=(Rotation=(X=0,Y=0,Z=0,W=1),Translation=(X=0,Y=0,Z=0),Scale3D=(X=1,Y=1,Z=1)))
    Primary=(Weight=1,Axis=(X=1,Y=0,Z=0),Target=(X=1,Y=0,Z=0),Kind=Location,Space=(Type=Bone,Name="None"))
    Secondary=(Weight=1,Axis=(X=0,Y=0,Z=1),Target=(X=1,Y=0,Z=0),Kind=Direction,Space=(Type=Bone,Name="None"))
    Weight=1
    Secondary.Kind=@n1.Kind
    Primary.Axis=@n1.Primary Axis
    Secondary.Axis=@n1.Secondary Axis
    InputTransform=&RigUnit_GetTransform(bInitial=False, Item=@n3.Element).Transform
    Primary.Target=&RigUnit_GetTransform(
     bInitial=False
     Item=RigVMDispatchNode(Index=@Add.Result, Array=@n1.Bones) {template="DISPATCH_RigVMDispatch_ArrayGetAtIndex(in Array,in Index,out Element)"}.Element
    ).Transform.Translation
    Secondary.Target=RigVMVariableNode() {variable=Secondary Target}.Value
   )
   &RigUnit_SetTransform(Weight=1, bInitial=true, Value=@n5.Result, Item=@n3.Element, ExecutePin=@n4.False)
   n6 = RigVMDispatchNode(ExecuteContext=@n3.Completed, Array=RigVMVariableNode() {variable=Bones}.Value) {template="DISPATCH_RigVMDispatch_ArrayIterator(in Array,out Element,out Index,out Count,out Ratio)"}
   DISPATCH_RigDispatch_GetMetadata = RigVMDispatchNode(Default=(Type=Bone,Name="None"), Name=Orig Parent, NameSpace=Self, Item=@n6.Element) {template="DISPATCH_RigDispatch_GetMetadata(in Item,in Name,in NameSpace,in Default,out Value,out Found)"}
   > &RigUnit_SetDefaultParent(Parent=@DISPATCH_RigDispatch_GetMetadata.Value, Child=@n6.Element)
   DISPATCH_RigVMDispatch_ArrayGetAtIndex = RigVMDispatchNode(Index=-1, Array=RigVMVariableNode() {variable=Bones}.Value) {template="DISPATCH_RigVMDispatch_ArrayGetAtIndex(in Array,in Index,out Element)"}
   Set Transform = &RigUnit_SetRotation(Value=(X=0,Y=0,Z=0,W=1), Weight=1, bInitial=true, Item=@DISPATCH_RigVMDispatch_ArrayGetAtIndex.Element, ExecutePin=@n6.Completed)
   RigVMFunction_DebugTransformArrayMutableNoSpace = /Script/RigVM.RigVMFunction_DebugTransformArrayMutableNoSpace(
    Color=(R=1,G=1,B=1,A=1)
    DebugDrawSettings=(DepthPriority=SDPG_Foreground,Lifetime=-1)
    Mode=Axes
    ParentIndices=()
    Scale=15
    Thickness=1
    WorldOffset=(Rotation=(X=0,Y=0,Z=0,W=1),Translation=(X=0,Y=0,Z=0),Scale3D=(X=1,Y=1,Z=1))
    bEnabled=@n1.Debug
    Transforms=&RigUnit_GetTransformItemArray(Space=GlobalSpace, bInitial=False, Items=@n1.Bones).Transforms
    ExecutePin=@RigVMFunction_Sequence.B
   ) {is=struct RigVMUnitNode}
   RigVMFunctionReturnNode(ExecuteContext=@RigVMFunction_Sequence.C)
   &RigUnit_FindItemsWithMetadataTag(NameSpace=Self, Tag=New Bones)
  ]}
 New Function_ContainedGraph {nodes [
   n1 = RigVMFunctionEntryNode()
   RigVMFunction_Sequence = &RigVMFunction_Sequence(ExecuteContext=@n1.ExecuteContext)
   n2 = &RigVMFunction_Sequence(ExecuteContext=@RigVMFunction_Sequence.A)
   RigVMDispatchNode(ExecuteContext=@n2.A, Array=RigVMVariableNode() {variable=New Bones}.Value) {template="DISPATCH_RigVMDispatch_ArrayReset(io Array)"}
   > DISPATCH_RigVMDispatch_ArrayIterator = RigVMDispatchNode(Array=@n1.Transforms) {template="DISPATCH_RigVMDispatch_ArrayIterator(in Array,out Element,out Index,out Count,out Ratio)"}
   RigVMFunction_StringPadInteger = /Script/RigVM.RigVMFunction_StringPadInteger(
    Digits=2
    Value=&RigVMFunction_MathIntAdd(B=1, A=@DISPATCH_RigVMDispatch_ArrayIterator.Index) {template="Add::Execute(in A,in B,out Result)"}.Result
   ) {is=struct RigVMUnitNode}
   DISPATCH_RigDispatch_FromString = RigVMDispatchNode(String=@RigVMFunction_StringPadInteger.Result) {template="DISPATCH_RigDispatch_FromString(in String,out Result)"}
   Concat = RigVMAggregateNode(B=_, D=_, C=@DISPATCH_RigDispatch_FromString.Result, A=@n1.Name, E=@n1.Optional Suffix)
   Concat_1 = RigVMAggregateNode(B=_, D=None, C=@DISPATCH_RigDispatch_FromString.Result, A=RigVMVariableNode() {variable=Name}.Value)
   EndsWith = /Script/RigVM.RigVMFunction_EndsWith(Ending=_, Name=@Concat.Result) {is=struct RigVMUnitNode, template="EndsWith::Execute(in Name,in Ending,out Result)"}
   DISPATCH_RigVMDispatch_If = RigVMDispatchNode(False=@Concat.Result, True=@Concat_1.Result, Condition=@EndsWith.Result) {template="DISPATCH_RigVMDispatch_If(in Condition,in True,in False,out Result)"}
   > n3 = /Script/ControlRig.RigUnit_HierarchyAddBone(Space=GlobalSpace, Transform=@DISPATCH_RigVMDispatch_ArrayIterator.Element, Name=@DISPATCH_RigVMDispatch_If.Result, Parent=@n1.Parent) {is=struct RigVMUnitNode}
   > Add = RigVMDispatchNode(Element=@n3.Item, Array=RigVMVariableNode() {variable=New Bones}.Value) {template="DISPATCH_RigVMDispatch_ArrayAdd(io Array,in Element,out Index)"}
   n4 = RigVMVariableNode() {variable=New Bones}
   n5 = RigVMDispatchNode(ExecuteContext=@n2.B, Array=@n4.Value) {template="DISPATCH_RigVMDispatch_ArrayIterator(in Array,out Element,out Index,out Count,out Ratio)"}
   n6 = &RigVMFunction_MathIntSub(B=1, A=@n5.Index) {template="Subtract::Execute(in A,in B,out Result)"}
   > &RigUnit_SetDefaultParent(
    Child=@n5.Element
    Parent=RigVMDispatchNode(
     False=RigVMDispatchNode(Index=@n6.Result, Array=@n4.Value) {template="DISPATCH_RigVMDispatch_ArrayGetAtIndex(in Array,in Index,out Element)"}.Element
     Condition=RigVMDispatchNode(B=0, A=@n5.Index) {template="DISPATCH_RigVMDispatch_CoreEquals(in A,in B,out Result)"}.Result
     True=RigVMVariableNode() {variable=Parent}.Value
    ) {template="DISPATCH_RigVMDispatch_If(in Condition,in True,in False,out Result)"}.Result
   )
   RigVMFunction_ControlFlowBranch = &RigVMFunction_ControlFlowBranch(ExecuteContext=@n5.Completed, Condition=RigVMVariableNode() {variable=Enable Orient Bones}.Value)
   Orient Bones = RigVMFunctionReferenceNode(
    ExecuteContext=@RigVMFunction_ControlFlowBranch.True
    Secondary Axis=RigVMVariableNode() {variable=Secondary Axis}.Value
    Primary Axis=RigVMVariableNode() {variable=Primary Axis}.Value
    Bones=RigVMVariableNode() {variable=New Bones}.Value
    Debug=RigVMVariableNode() {variable=Debug}.Value
    Kind=RigVMVariableNode() {variable=Kind}.Value
   )
   RigVMFunctionReturnNode(ExecuteContext=@RigVMFunction_Sequence.B, Bones=RigVMVariableNode() {variable=New Bones}.Value)
   &RigUnit_FindItemsWithMetadataTag(NameSpace=Self, Tag=New Bones)
  ]}
 New Function_ContainedGraph {nodes [
   n1 = RigVMFunctionEntryNode()
   RigVMFunction_Sequence = &RigVMFunction_Sequence(ExecuteContext=@n1.ExecuteContext)
   n2 = &RigUnit_ResolveConnector(Connector=(Type=Connector,Name="Start Bone"), SkipSocket=False)
   RigVMDispatchNode(Name=Start Bone, NameSpace=Self, Value=@n2.Result, ExecuteContext=@RigVMFunction_Sequence.A) {template="DISPATCH_RigDispatch_SetModuleMetadata(in Name,in NameSpace,in Value,out Success)"}
   n3 = &RigUnit_ResolveConnector(Connector=(Type=Connector,Name="End Bone"), SkipSocket=False)
   > RigVMDispatchNode(Name=End Bone, NameSpace=Self, Value=@n3.Result) {template="DISPATCH_RigDispatch_SetModuleMetadata(in Name,in NameSpace,in Value,out Success)"}
   > RigVMDispatchNode(
    Name=Bones
    NameSpace=Self
    Value=&RigUnit_HierarchyGetChainItemArray(bIncludeEnd=True, bIncludeStart=True, bReverse=false, End=@n3.Result, Start=@n2.Result).Chain
   ) {template="DISPATCH_RigDispatch_SetModuleMetadata(in Name,in NameSpace,in Value,out Success)"}
   n4 = RigVMVariableNode() {variable=Rig Bone Positions}
   > RigVMDispatchNode(Array=@n4.Value) {template="DISPATCH_RigVMDispatch_ArrayReset(io Array)"}
   > Set_Num = RigVMDispatchNode(Array=@n4.Value, Num=RigVMVariableNode() {variable=Number of Neck Bones}.Value) {template="DISPATCH_RigVMDispatch_ArraySetNum(io Array,in Num)"}
   > n5 = &RigVMFunction_ForLoopCount(Count=RigVMVariableNode() {variable=Number of Neck Bones}.Value)
   n6 = RigVMDispatchNode(Default=(Type=Bone,Name="None"), Name=End Bone, NameSpace=Self) {template="DISPATCH_RigDispatch_GetModuleMetadata(in Name,in NameSpace,in Default,out Value,out Found)"}
   n7 = RigVMDispatchNode(Default=(Type=Bone,Name="None"), Name=Start Bone, NameSpace=Self) {template="DISPATCH_RigDispatch_GetModuleMetadata(in Name,in NameSpace,in Default,out Value,out Found)"}
   > Set_At = RigVMDispatchNode(
    Element=&RigVMFunction_MathTransformLerp(
     B=&RigUnit_GetTransform(bInitial=true, Item=@n6.Value).Transform
     A=&RigUnit_GetTransform(bInitial=true, Item=@n7.Value).Transform
     T=@n5.Ratio
    ) {template="Interpolate::Execute(in A,in B,in T,out Result)"}.Result
    Index=@n5.Index
    Array=RigVMVariableNode() {variable=Rig Bone Positions}.Value
   ) {template="DISPATCH_RigVMDispatch_ArraySetAtIndex(io Array,in Index,in Element)"}
   Get_Module_Metadata_1 = RigVMDispatchNode(Default=(), Name=Bones, NameSpace=Self) {template="DISPATCH_RigDispatch_GetModuleMetadata(in Name,in NameSpace,in Default,out Value,out Found)"}
   Create Bone Chain_1 = RigVMFunctionReferenceNode(
    Debug=true
    Debug Scale=15
    Debug Thickness=0.5
    Kind=Location
    Optional Suffix=rig
    Enable Orient Bones=@n1.Enable Orient Bones
    Parent=&RigUnit_HierarchyGetParent(
     bDefaultParent=True
     Child=RigVMDispatchNode(Index=0, Array=@Get_Module_Metadata_1.Value) {template="DISPATCH_RigVMDispatch_ArrayGetAtIndex(in Array,in Index,out Element)"}.Element
    ).Parent
    ExecuteContext=@n5.Completed
    Transforms=RigVMVariableNode() {variable=Rig Bone Positions}.Value
    Primary Axis=RigVMVariableNode() {variable=Primary Axis}.Value
    Secondary Axis=RigVMVariableNode() {variable=Secondary Axis}.Value
    Name=RigVMVariableNode() {variable=Module Settings}.Value.Prefix_20_#2
   )
   > RigVMDispatchNode(Name=Rig Bones, NameSpace=Self, Value=@"Create Bone Chain_1".Bones) {template="DISPATCH_RigDispatch_SetModuleMetadata(in Name,in NameSpace,in Value,out Success)"}
   RigVMFunctionReturnNode(ExecuteContext=@RigVMFunction_Sequence.B)
   &RigUnit_GetTransformItemArray(Space=GlobalSpace, bInitial=False, Items=@Get_Module_Metadata_1.Value)
  ]}
 New Function_ContainedGraph {nodes [
   n1 = RigVMFunctionEntryNode()
   > &RigUnit_SetTransform(
    Weight=1
    bInitial=False
    Item=@n1.Child
    Value=&RigUnit_ProjectTransformToNewParent(bChildInitial=True, bNewParentInitial=False, bOldParentInitial=True, Child=@n1.Child, NewParent=@n1.New Parent, OldParent=@n1.New Parent).Transform
   )
   > RigVMFunctionReturnNode()
  ]}
 New Function_ContainedGraph {nodes [
   n1 = RigVMFunctionEntryNode()
   n2 = &RigVMFunction_ControlFlowBranch(Condition=@n1.Attach, ExecuteContext=@n1.ExecuteContext)
   RigVMFunctionReturnNode(ExecuteContext=@n2.Completed)
   Get_Module_Metadata = RigVMDispatchNode(Default=(), Name=Rig Bones, NameSpace=Self) {template="DISPATCH_RigDispatch_GetModuleMetadata(in Name,in NameSpace,in Default,out Value,out Found)"}
   n3 = RigVMDispatchNode(Array=@Get_Module_Metadata.Value, ExecuteContext=@n2.True) {template="DISPATCH_RigVMDispatch_ArrayIterator(in Array,out Element,out Index,out Count,out Ratio)"}
   Get_Module_Metadata_1 = RigVMDispatchNode(Default=(), Name=Bones, NameSpace=Self) {template="DISPATCH_RigDispatch_GetModuleMetadata(in Name,in NameSpace,in Default,out Value,out Found)"}
   > Parent_1 = RigVMFunctionReferenceNode(
    Child=RigVMDispatchNode(Index=@n3.Index, Array=@Get_Module_Metadata_1.Value) {template="DISPATCH_RigVMDispatch_ArrayGetAtIndex(in Array,in Index,out Element)"}.Element
    New Parent=@n3.Element
   )
  ]}
 New Function_ContainedGraph {nodes [
   n1 = RigVMFunctionEntryNode()
   DISPATCH_RigDispatch_GetModuleMetadata = RigVMDispatchNode(Default=None, NameSpace=Self, Name=@n1.Name) {template="DISPATCH_RigDispatch_GetModuleMetadata(in Name,in NameSpace,in Default,out Value,out Found)"}
   GetAnimationChannel = &RigUnit_GetBoolAnimationChannel(
    bInitial=False
    Channel=@DISPATCH_RigDispatch_GetModuleMetadata.Value
    Control=&RigUnit_ResolveConnector(SkipSocket=true, Connector=@n1.Connector).Result.Name
   ) {template="GetAnimationChannel::Execute(out Value,in Control,in Channel,in bInitial)"}
   > RigVMFunctionReturnNode(Value=@GetAnimationChannel.Value)
  ]}
 New Function_ContainedGraph {nodes [
   n1 = RigVMFunctionEntryNode()
   > n2 = RigVMDispatchNode(Array=@n1.IK Controls) {template="DISPATCH_RigVMDispatch_ArrayIterator(in Array,out Element,out Index,out Count,out Ratio)"}
   FWD Get FK or IK_1 = RigVMFunctionReferenceNode(Connector=(Type=Connector,Name="Settings Control"), Name=fk_ik_switch)
   > /Script/ControlRig.RigUnit_SetControlVisibility(Item=@n2.Element, bVisible=@"FWD Get FK or IK_1".Value) {is=struct RigVMUnitNode, template="SetControlVisibility::Execute(in Item,in Pattern,in bVisible)"}
   n3 = RigVMDispatchNode(ExecuteContext=@n2.Completed, Array=@n1.FK Controls) {template="DISPATCH_RigVMDispatch_ArrayIterator(in Array,out Element,out Index,out Count,out Ratio)"}
   FWD Get FK or IK_1_1 = RigVMFunctionReferenceNode(Connector=(Type=Connector,Name="Settings Control"), Name=fk_ik_switch)
   > &RigUnit_SetControlVisibility(Item=@n3.Element, bVisible=&RigVMFunction_MathBoolNot(Value=@"FWD Get FK or IK_1_1".Value).Result) {template="SetControlVisibility::Execute(in Item,in Pattern,in bVisible)"}
   For_Each = RigVMDispatchNode(ExecuteContext=@n3.Completed, Array=&RigUnit_FindItemsWithMetadataTag(NameSpace=Self, Tag=Controls to Hide).Items) {template="DISPATCH_RigVMDispatch_ArrayIterator(in Array,out Element,out Index,out Count,out Ratio)"}
   > &RigUnit_SetControlVisibility(bVisible=false, Item=@For_Each.Element) {template="SetControlVisibility::Execute(in Item,in Pattern,in bVisible)"}
   RigVMFunctionReturnNode(ExecuteContext=@For_Each.Completed)
  ]}
 New Function_ContainedGraph {nodes [
   n1 = RigVMFunctionEntryNode()
   > DISPATCH_RigVMDispatch_ArrayIterator = RigVMDispatchNode(Array=@n1.Controls) {template="DISPATCH_RigVMDispatch_ArrayIterator(in Array,out Element,out Index,out Count,out Ratio)"}
   n2 = &RigUnit_GetShapeTransform(Control=@DISPATCH_RigVMDispatch_ArrayIterator.Element.Name)
   Multiply_1 = RigVMAggregateNode(
    B=(X=1,Y=1,Z=1)
    C=(X=0,Y=0,Z=0)
    B.X=@n1.Global Scale
    B.Y=@n1.Global Scale
    B.Z=@n1.Global Scale
    C.X=@n1.Local Scale
    C.Y=@n1.Local Scale
    C.Z=@n1.Local Scale
    A=@n2.Transform.Scale3D
   )
   > &RigUnit_SetShapeTransform(
    Transform=(Rotation=(X=0,Y=0,Z=0,W=1),Translation=(X=0,Y=0,Z=0),Scale3D=(X=1,Y=1,Z=1))
    Control=@DISPATCH_RigVMDispatch_ArrayIterator.Element.Name
    Transform.Rotation=@n2.Transform.Rotation
    Transform.Translation=@n2.Transform.Translation
    Transform.Scale3D=@Multiply_1.Result
   )
   RigVMFunctionReturnNode(ExecuteContext=@DISPATCH_RigVMDispatch_ArrayIterator.Completed)
  ]}
 New Function_ContainedGraph {nodes [
   n1 = RigVMFunctionEntryNode()
   DISPATCH_RigDispatch_GetModuleMetadata = RigVMDispatchNode(Default=(R=1,G=1,B=1,A=1), Name=Color Center, NameSpace=Root) {template="DISPATCH_RigDispatch_GetModuleMetadata(in Name,in NameSpace,in Default,out Value,out Found)"}
   RigVMFunction_MathBoolNot = &RigVMFunction_MathBoolNot(Value=@n1.Override)
   DISPATCH_RigVMDispatch_If = RigVMDispatchNode(True=@DISPATCH_RigDispatch_GetModuleMetadata.Value, False=@n1.CenterColor, Condition=@RigVMFunction_MathBoolNot.Result) {template="DISPATCH_RigVMDispatch_If(in Condition,in True,in False,out Result)"}
   > ColorizeControls = RigVMFunctionReferenceNode(
    X Plane=true
    ignore_ctrls=()
    CenterColor=@DISPATCH_RigVMDispatch_If.Result
    LeftColor=RigVMDispatchNode(
     True=RigVMDispatchNode(Default=(R=1,G=1,B=1,A=1), Name=Color left, NameSpace=Root) {template="DISPATCH_RigDispatch_GetModuleMetadata(in Name,in NameSpace,in Default,out Value,out Found)"}.Value
     False=@n1.LeftColor
     Condition=@RigVMFunction_MathBoolNot.Result
    ) {template="DISPATCH_RigVMDispatch_If(in Condition,in True,in False,out Result)"}.Result
    RightColor=RigVMDispatchNode(
     True=RigVMDispatchNode(Default=(R=1,G=1,B=1,A=1), Name=Color Right, NameSpace=Root) {template="DISPATCH_RigDispatch_GetModuleMetadata(in Name,in NameSpace,in Default,out Value,out Found)"}.Value
     False=@n1.RightColor
     Condition=@RigVMFunction_MathBoolNot.Result
    ) {template="DISPATCH_RigVMDispatch_If(in Condition,in True,in False,out Result)"}.Result
    controls=@n1.Controls
   )
   > RigVMFunctionReturnNode()
  ]}
 New Function_ContainedGraph {nodes [
   n1 = RigVMFunctionEntryNode()
   DISPATCH_RigDispatch_GetModuleMetadata = RigVMDispatchNode(Default=None, Name=fk_ik_switch, NameSpace=Self) {template="DISPATCH_RigDispatch_GetModuleMetadata(in Name,in NameSpace,in Default,out Value,out Found)"}
   > SetAnimationChannel = /Script/ControlRig.RigUnit_SetBoolAnimationChannel(
    bInitial=False
    Channel=@DISPATCH_RigDispatch_GetModuleMetadata.Value
    Value=@n1.IK
    Control=&RigUnit_ResolveConnector(Connector=(Type=Connector,Name="Settings Control"), SkipSocket=true).Result.Name
   ) {is=struct RigVMUnitNode, template="SetAnimationChannel::Execute(in Value,in Control,in Channel,in bInitial)"}
   > RigVMFunctionReturnNode()
  ]}
 New Function_ContainedGraph {nodes [
   n1 = RigVMFunctionEntryNode()
   > For_Each_1 = RigVMDispatchNode(Array=&RigUnit_CollectionChildrenArray(TypeToSearch=Bone, bDefaultChildren=True, bIncludeParent=False, bRecursive=False, Parent=@n1.Parent).Items) {template="DISPATCH_RigVMDispatch_ArrayIterator(in Array,out Element,out Index,out Count,out Ratio)"}
   > /Script/ControlRig.RigUnit_HierarchyAddSocket(
    Color=(R=1,G=1,B=1,A=1)
    Description=Place a new Module here . . .
    Space=LocalSpace
    Transform=(Rotation=(X=0,Y=0,Z=0,W=1),Translation=(X=0,Y=0,Z=0),Scale3D=(X=1,Y=1,Z=1))
    Name=&RigVMFunction_NameConcat(B=_socket, A=@For_Each_1.Element.Name) {template="Concat::Execute(in A,in B,out Result)"}.Result
    Parent=@For_Each_1.Element
   ) {is=struct RigVMUnitNode}
   RigVMFunctionReturnNode(ExecuteContext=@For_Each_1.Completed)
  ]}
 RigVMFunctionLibrary {nodes [
   Add Null Above = RigVMCollapseNode(ExecuteContext=(), Item=(Name="None"))
   Auto Get Prefix = RigVMCollapseNode(Value=None)
   CONSTR IK Animal Neck = RigVMCollapseNode(
    Add Stretch Option=False
    Axis=(X=1,Y=0,Z=0)
    Create Control Per Bone=False
    Debug=False
    Dynamics Blend End=1
    Dynamics Blend Start=1
    Dynamics Blend Twist=1
    Dynamics Damping=0.5
    Dynamics Strength=4
    Enable Orient Bones=true
    ExecuteContext=()
    Head Control Settings="(DisplayName=\"None\",bSelectable=True,InitialSpace=LocalSpace,bUsePreferredRotationOrder=False,PreferredRotationOrder=YZX,Limits=(LimitTranslationX=(bMinimum=False,bMaximum=False),LimitTranslationY=(bMinimum=False,bMaximum=False),LimitTranslationZ=(bMinimum ...(truncated)"
    Headneck Position=0.8
    Neck Tangent Control Settings="(DisplayName=\"None\",bSelectable=True,InitialSpace=LocalSpace,bUsePreferredRotationOrder=False,PreferredRotationOrder=YZX,Limits=(LimitTranslationX=(bMinimum=False,bMaximum=False),LimitTranslationY=(bMinimum=False,bMaximum=False),LimitTranslationZ=(bMinimum ...(truncated)"
    Number of Controls=4
    Number of Main Controls=3
    Parent Controls to each other=False
    Primary Axis=(X=1,Y=0,Z=0)
    Secondary Spline Direction=(X=0,Y=5,Z=0)
    Up Axis=(X=0,Y=1,Z=0)
   )
   Color Module Controls = RigVMCollapseNode(CenterColor=(R=0,G=0,B=0,A=0), Controls=(), ExecuteContext=(), LeftColor=(R=0,G=0,B=0,A=0), Override=False, RightColor=(R=0,G=0,B=0,A=0))
   ColorizeControls = RigVMCollapseNode(
    CenterColor=(R=1,G=1,B=0,A=1)
    ExecuteContext=()
    LeftColor=(R=0,G=1,B=0,A=1)
    RightColor=(R=1,G=0,B=0,A=1)
    X Plane=true
    controls=()
    ignore_ctrls=()
   )
   Compute Spline Positions = RigVMCollapseNode(Array=(), ExecuteContext=())
   Connect to Leaf Bones = RigVMCollapseNode(Clamp=0, Connector=(Type=Connector,Name="End"), ExecuteContext=(), Primary Connector=(Type=Connector,Name="Finger"))
   Control Stack at Item = RigVMCollapseNode(
    Add Bottom Null=False
    Add Null=False
    Add Secondary Control=False
    Control Suffix=_ctrl
    ExecuteContext=()
    Item=(Name="None")
    Null Suffix=_null
    Parent=(Name="None")
    Secondary Settings="(Shape=(bVisible=True,Name=\"Default\",Color=(R=1,G=0,B=0,A=1),Transform=(Rotation=(X=0,Y=0,Z=0,W=1),Translation=(X=0,Y=0,Z=0),Scale3D=(X=1,Y=1,Z=1))),Proxy=(bIsProxy=False,ShapeVisibility=BasedOnSelection),DisplayName=\"None\",bSelectable=True,InitialSpace=Lo ...(truncated)"
    Settings="(Shape=(bVisible=True,Name=\"Default\",Color=(R=1,G=0,B=0,A=1),Transform=(Rotation=(X=0,Y=0,Z=0,W=1),Translation=(X=0,Y=0,Z=0),Scale3D=(X=1,Y=1,Z=1))),Proxy=(bIsProxy=False,ShapeVisibility=BasedOnSelection),DisplayName=\"None\",bSelectable=True,InitialSpace=Lo ...(truncated)"
    Tag Controls=true
    Zero Controls=true
   )
   Control Stack at Position = RigVMCollapseNode(
    Add Bottom Null=False
    Add Null=False
    Add Secondary Control=False
    Control Suffix=_ctrl
    ExecuteContext=()
    Global Position=(Rotation=(X=0,Y=0,Z=0,W=1),Translation=(X=0,Y=0,Z=0),Scale3D=(X=1,Y=1,Z=1))
    Null Suffix=_null
    Parent=(Type=None,Name="None")
    Secondary Settings="(Shape=(bVisible=True,Name=\"Default\",Color=(R=1,G=0,B=0,A=1),Transform=(Rotation=(X=0,Y=0,Z=0,W=1),Translation=(X=0,Y=0,Z=0),Scale3D=(X=1,Y=1,Z=1))),Proxy=(bIsProxy=False,ShapeVisibility=BasedOnSelection),DisplayName=\"None\",bSelectable=True,InitialSpace=Lo ...(truncated)"
    Settings="(Shape=(bVisible=True,Name=\"Default\",Color=(R=1,G=0,B=0,A=1),Transform=(Rotation=(X=0,Y=0,Z=0,W=1),Translation=(X=0,Y=0,Z=0),Scale3D=(X=1,Y=1,Z=1))),Proxy=(bIsProxy=False,ShapeVisibility=BasedOnSelection),DisplayName=\"None\",bSelectable=True,InitialSpace=Lo ...(truncated)"
    Tag Controls=true
    Zero Controls=true
   )
   Create Bone Chain = RigVMCollapseNode(
    Debug=false
    Debug Scale=15
    Debug Thickness=0.5
    Enable Orient Bones=true
    ExecuteContext=()
    Kind=Direction
    Name=NewBone
    Parent=(Type=Bone,Name="None")
    Primary Axis=(X=1,Y=0,Z=0)
    Secondary Axis=(X=0,Y=1,Z=0)
    Transforms=()
   )
   Create Dual Spline = RigVMCollapseNode(
    Bones=()
    Controls=()
    ExecuteContext=()
    Primary Axis=(X=1,Y=0,Z=0)
    Samples Per Segment=16
    Secondary Spline Direction=(X=0,Y=5,Z=0)
    Spline Mode=BSpline
    Up Axis=(X=0,Y=1,Z=0)
   )
   Create FK IK Switch = RigVMCollapseNode(name=tail, ExecuteContext=(), Initial Value=false, prefix=None, side=None)
   Create FK Tail Controls = RigVMCollapseNode(
    End=(Type=Bone,Name="Tail15")
    ExecuteContext=()
    Settings="(DisplayName=\"None\",bSelectable=True,InitialSpace=LocalSpace,bUsePreferredRotationOrder=False,PreferredRotationOrder=YZX,Limits=(LimitTranslationX=(bMinimum=False,bMaximum=False),LimitTranslationY=(bMinimum=False,bMaximum=False),LimitTranslationZ=(bMinimum ...(truncated)"
    Start=(Type=Bone,Name="Tail")
   )
   Create Neck Rig Bones = RigVMCollapseNode(Enable Orient Bones=true, ExecuteContext=())
   Draw Line Between Items = RigVMCollapseNode(
    Color=(R=1,G=1,B=0,A=1)
    Enabled=True
    ExecuteContext=()
    Items=()
    Thickness=0.15
    World Offset=(Rotation=(X=0,Y=0,Z=0,W=1),Translation=(X=0,Y=0,Z=0),Scale3D=(X=1,Y=1,Z=1))
   )
   Dual SplineIK = RigVMCollapseNode(
    Aim Spline=()
    Bones=()
    Controls=()
    ExecuteContext=()
    Primary Axis=(X=1,Y=0,Z=0)
    Samples Per Segment=16
    Secondary Spline Direction=(X=0,Y=5,Z=0)
    Spline Mode=BSpline
    Template Spline=()
    Up Axis=(X=0,Y=1,Z=0)
   )
   FWD Attach Bones to Rig Bones = RigVMCollapseNode(Attach=true, ExecuteContext=())
   FWD FK Animal Neck = RigVMCollapseNode(ExecuteContext=())
   FWD FK or IK Control Vis = RigVMCollapseNode(ExecuteContext=(), FK Controls=(), IK Controls=())
   FWD Get FK or IK = RigVMCollapseNode(Connector=(Type=Connector,Name="Settings Control"), Name=fk_ik_switch)
   FWD Get FK or IK_1 = RigVMCollapseNode(Connector=(Type=Connector,Name="Settings Control"), Name=fk_ik_switch)
   FWD IK Animal Neck = RigVMCollapseNode(
    Add Stretch Option=true
    Attach Skeleton=true
    Blend End=1
    Blend Start=1
    Blend Twist=1
    Damping=0.5
    Dynamics=False
    ExecuteContext=()
    Primary Axis=(X=1,Y=0,Z=0)
    Secondary Spline Direction=(X=0,Y=5,Z=0)
    Spline Mode=BSpline
    Strength=4
    Up Axis=(X=0,Y=1,Z=0)
    Use SplineIK=False
   )
   FWD Set FK or IK = RigVMCollapseNode(ExecuteContext=(), IK=False)
   Finalize Controls = RigVMCollapseNode(
    ExecuteContext=()
    Module Settings="(OverrideColor_48_#7=False,LeftColor_3_#6=(R=0,G=0.0231,B=0.4115,A=1),RightColor_5_#8=(R=1,G=0,B=0,A=1),CenterColor_9_#5=(R=1,G=1,B=0,A= ...(truncated)"
   )
   Find = RigVMCollapseNode(Element=foot, Value=None)
   Find Closest = RigVMCollapseNode(Bones=(), ExecuteContext=(), Initial=true, Item=())
   FindClosestBone = RigVMCollapseNode(Bones=(), ExecuteContext=(), Initial=true, Item=())
   Format Name = RigVMCollapseNode(name=arm, prefix=None, side=_fk_ik_switch, suffix=_fk_ik_switch)
   Format Name_1 = RigVMCollapseNode(name=arm, prefix=None, side=_fk_ik_switch, suffix=_fk_ik_switch)
   Generate Spline Points = RigVMCollapseNode(ExecuteContext=(), Items=())
   Is Right Side = RigVMCollapseNode(ExecuteContext=(), Item=(Type=Bone))
   Orient Bones = RigVMCollapseNode(Bones=(), Debug=false, ExecuteContext=(), Kind=Direction, Primary Axis=(X=1,Y=0,Z=0), Secondary Axis=(X=0,Y=1,Z=0))
   Parent = RigVMCollapseNode(Child=(Type=Bone,Name="None"), ExecuteContext=(), New Parent=(Type=Bone,Name="None"))
   Parent_1 = RigVMCollapseNode(Child=(Type=Bone,Name="None"), ExecuteContext=(), New Parent=(Type=Bone,Name="None"))
   Scale Module Controls = RigVMCollapseNode(Controls=(), ExecuteContext=(), Global Scale=1, Local Scale=0)
   Set Controls Scale and Color = RigVMCollapseNode(
    ExecuteContext=()
    Module Settings="(OverrideColor_48_#7=False,LeftColor_3_#6=(R=0,G=0.0231,B=0.4115,A=1),RightColor_5_#8=(R=1,G=0,B=0,A=1),CenterColor_9_#5=(R=1,G=1,B=0,A= ...(truncated)"
   )
   Sockets on Direct Children = RigVMCollapseNode(ExecuteContext=(), Parent=(Type=Bone,Name="None"))
   SplineThrough3Points = RigVMCollapseNode(Execute=(), Point 0=(X=0,Y=0,Z=0), Point 1=(X=0,Y=0,Z=0), Point 2=(X=0,Y=0,Z=0), t=0.5)
   SplineThrough4Points = RigVMCollapseNode(Execute=(), Point 0=(X=0,Y=0,Z=0), Point 1=(X=0,Y=0,Z=0), Point 2=(X=0,Y=0,Z=0), Point 3=(X=0,Y=0,Z=0))
  ]}
 RigVMFunction_Sequence_1_1_ContainedGraph {nodes [
   n1 = RigVMFunctionEntryNode()
   n2 = &RigVMFunction_Sequence(ExecuteContext=@n1.ExecuteContext)
   n3 = &RigVMFunction_Sequence(ExecuteContext=@n2.B)
   n4 = &RigVMFunction_Sequence(ExecuteContext=@n3.B)
   n5 = &RigVMFunction_Sequence(ExecuteContext=@n4.B)
   n6 = &RigVMFunction_Sequence(ExecuteContext=@n5.B)
   RigVMFunctionReturnNode(A=@n2.A, B=@n3.A, C=@n4.A, D=@n5.A, E=@n6.A, F=@n6.B)
  ]}
 RigVMFunction_Sequence_1_1_ContainedGraph {nodes [
   n1 = RigVMFunctionEntryNode()
   n2 = &RigVMFunction_Sequence(ExecuteContext=@n1.ExecuteContext)
   n3 = &RigVMFunction_Sequence(ExecuteContext=@n2.B)
   n4 = &RigVMFunction_Sequence(ExecuteContext=@n3.B)
   n5 = &RigVMFunction_Sequence(ExecuteContext=@n4.B)
   RigVMFunctionReturnNode(A=@n2.A, B=@n3.A, C=@n4.A, D=@n5.A, E=@n5.B)
  ]}
 RigVMFunction_Sequence_1_1_ContainedGraph {nodes [
   n1 = RigVMFunctionEntryNode()
   n2 = &RigVMFunction_Sequence(ExecuteContext=@n1.ExecuteContext)
   n3 = &RigVMFunction_Sequence(ExecuteContext=@n2.B)
   n4 = &RigVMFunction_Sequence(ExecuteContext=@n3.B)
   RigVMFunctionReturnNode(A=@n2.A, B=@n3.A, C=@n4.A, D=@n4.B)
  ]}
 RigVMFunction_Sequence_1_ContainedGraph {nodes [
   n1 = RigVMFunctionEntryNode()
   RigVMFunction_Sequence = &RigVMFunction_Sequence(ExecuteContext=@n1.ExecuteContext)
   n2 = &RigVMFunction_Sequence(ExecuteContext=@RigVMFunction_Sequence.B)
   n3 = &RigVMFunction_Sequence(ExecuteContext=@n2.B)
   RigVMFunctionReturnNode(A=@RigVMFunction_Sequence.A, B=@n2.A, C=@n3.A, D=@n3.B)
  ]}
 RigVMFunction_Sequence_1_ContainedGraph {nodes [
   n1 = RigVMFunctionEntryNode()
   RigVMFunction_Sequence = &RigVMFunction_Sequence(ExecuteContext=@n1.ExecuteContext)
   n2 = &RigVMFunction_Sequence(ExecuteContext=@RigVMFunction_Sequence.B)
   RigVMFunctionReturnNode(A=@RigVMFunction_Sequence.A, B=@n2.A, C=@n2.B)
  ]}
 RigVMFunction_Sequence_1_ContainedGraph {nodes [
   n1 = RigVMFunctionEntryNode()
   RigVMFunction_Sequence = &RigVMFunction_Sequence(ExecuteContext=@n1.ExecuteContext)
   n2 = &RigVMFunction_Sequence(ExecuteContext=@RigVMFunction_Sequence.B)
   RigVMFunctionReturnNode(A=@RigVMFunction_Sequence.A, B=@n2.A, C=@n2.B)
  ]}
 RigVMFunction_Sequence_2_1_ContainedGraph {nodes [
   n1 = RigVMFunctionEntryNode()
   n2 = &RigVMFunction_Sequence(ExecuteContext=@n1.ExecuteContext)
   n3 = &RigVMFunction_Sequence(ExecuteContext=@n2.B)
   RigVMFunctionReturnNode(A=@n2.A, B=@n3.A, C=@n3.B)
  ]}
 RigVMFunction_Sequence_2_ContainedGraph {nodes [
   n1 = RigVMFunctionEntryNode()
   RigVMFunction_Sequence = &RigVMFunction_Sequence(ExecuteContext=@n1.ExecuteContext)
   n2 = &RigVMFunction_Sequence(ExecuteContext=@RigVMFunction_Sequence.B)
   RigVMFunctionReturnNode(A=@RigVMFunction_Sequence.A, B=@n2.A, C=@n2.B)
  ]}
 RigVMFunction_Sequence_2_ContainedGraph {nodes [
   n1 = RigVMFunctionEntryNode()
   RigVMFunction_Sequence = &RigVMFunction_Sequence(ExecuteContext=@n1.ExecuteContext)
   n2 = &RigVMFunction_Sequence(ExecuteContext=@RigVMFunction_Sequence.B)
   RigVMFunctionReturnNode(A=@RigVMFunction_Sequence.A, B=@n2.A, C=@n2.B)
  ]}
 RigVMModel {nodes [
   RigUnit_BeginExecution = &RigUnit_BeginExecution()
   FWD Get FK or IK = RigVMFunctionReferenceNode(Connector=(Type=Connector,Name="Settings Control"), Name=fk_ik_switch)
   n1 = &RigVMFunction_ControlFlowBranch(Condition=@"FWD Get FK or IK".Value, ExecuteContext=@RigUnit_BeginExecution.ExecutePin)
   FWD Attach Bones to Rig Bones = RigVMFunctionReferenceNode(Attach=true, ExecuteContext=@n1.Completed)
   Get_Module_Metadata_2 = RigVMDispatchNode(Default=(), Name=Main Controls, NameSpace=Self) {template="DISPATCH_RigDispatch_GetModuleMetadata(in Name,in NameSpace,in Default,out Value,out Found)"}
   Clone = RigVMDispatchNode(Array=@Get_Module_Metadata_2.Value) {template="DISPATCH_RigVMDispatch_ArrayClone(in Array,out Clone)"}
   Get_Module_Metadata_1 = RigVMDispatchNode(Default=(), Name=IK Controls, NameSpace=Self) {template="DISPATCH_RigDispatch_GetModuleMetadata(in Name,in NameSpace,in Default,out Value,out Found)"}
   > n2 = RigVMDispatchNode(Array=@Clone.Clone, Other=@Get_Module_Metadata_1.Value) {template="DISPATCH_RigVMDispatch_ArrayAppend(io Array,in Other)"}
   Get_Module_Metadata_1_1 = RigVMDispatchNode(Default=(), Name=FK Controls, NameSpace=Self) {template="DISPATCH_RigDispatch_GetModuleMetadata(in Name,in NameSpace,in Default,out Value,out Found)"}
   > FWD FK or IK Control Vis = RigVMFunctionReferenceNode(IK Controls=@n2.Array, FK Controls=@Get_Module_Metadata_1_1.Value)
   FWD FK Animal Neck = RigVMFunctionReferenceNode(ExecuteContext=@n1.False)
   Get Dynamics Channels = RigVMCollapseNode()
   FWD IK Animal Neck = RigVMFunctionReferenceNode(
    Attach Skeleton=true
    Create Control per Bone=false
    Dynamics=@"Get Dynamics Channels".Value
    Strength=@"Get Dynamics Channels".Value_1
    Damping=@"Get Dynamics Channels".Value_2
    Blend Start=@"Get Dynamics Channels".Value_3
    Blend End=@"Get Dynamics Channels".Value_4
    Blend Twist=@"Get Dynamics Channels".Value_5
    ExecuteContext=@n1.True
    Primary Axis=RigVMVariableNode() {variable=Module Settings}.Value.PrimaryBoneAxis_32_#3
    Use SplineIK=RigVMVariableNode() {variable=Solve with SplineIK}.Value
    Secondary Spline Direction=RigVMVariableNode() {variable=Secondary Spline Direction}.Value
    debug=RigVMVariableNode() {variable=Debug}.Value
    Spline Mode=RigVMVariableNode() {variable=Spline Mode}.Value
    Add Stretch Option=RigVMVariableNode() {variable=Stretch}.Value
    Up Axis=RigVMVariableNode() {variable=Secondary Axis}.Value
   )
  ]}
 RigVMModel Backwards Solve Graph {nodes [
   n1 = /Script/ControlRig.RigUnit_InverseExecution() {is=struct RigVMUnitNode}
   FWD Set FK or IK = RigVMFunctionReferenceNode(IK=false, ExecuteContext=@n1.ExecutePin)
   > DISPATCH_RigVMDispatch_ArrayIterator = RigVMDispatchNode(Array=RigVMDispatchNode(Default=(), Name=FK Controls, NameSpace=Self) {template="DISPATCH_RigDispatch_GetModuleMetadata(in Name,in NameSpace,in Default,out Value,out Found)"}.Value) {template="DISPATCH_RigVMDispatch_ArrayIterator(in Array,out Element,out Index,out Count,out Ratio)"}
   n2 = RigVMDispatchNode(
    Array=RigVMDispatchNode(Default=(), Name=Bones, NameSpace=Self) {template="DISPATCH_RigDispatch_GetModuleMetadata(in Name,in NameSpace,in Default,out Value,out Found)"}.Value
    Index=@DISPATCH_RigVMDispatch_ArrayIterator.Index
   ) {template="DISPATCH_RigVMDispatch_ArrayGetAtIndex(in Array,in Index,out Element)"}
   > &RigUnit_SetTransform(
    Weight=1
    bInitial=False
    Item=@DISPATCH_RigVMDispatch_ArrayIterator.Element
    Value=&RigUnit_ProjectTransformToNewParent(
     bChildInitial=True
     bNewParentInitial=False
     bOldParentInitial=True
     NewParent=@n2.Element
     OldParent=@n2.Element
     Child=@DISPATCH_RigVMDispatch_ArrayIterator.Element
    ).Transform
   )
  ]}
 RigVMModel Connector Event Graph {nodes [
   n1 = /Script/ControlRig.RigUnit_ConnectorExecution() {is=struct RigVMUnitNode}
   RigVMFunction_Sequence = RigVMAggregateNode(ExecuteContext=@n1.ExecuteContext)
   Connect to Module Metadata_1 = RigVMFunctionReferenceNode(
    Connector=(Type=Connector,Name="Parent Control")
    Module Metadata Name=Parent Control
    Module Metadata NameSpace=Parent
    ExecuteContext=@RigVMFunction_Sequence.A
   )
   > Connect to Module Metadata_1_1 = RigVMFunctionReferenceNode(Connector=(Type=Connector,Name="Parent Control"), Module Metadata Name=FK Parent, Module Metadata NameSpace=Parent)
   > Connect to Module Metadata_1_2 = RigVMFunctionReferenceNode(Connector=(Type=Connector,Name="Settings Control"), Module Metadata Name=Settings Control, Module Metadata NameSpace=Root)
   n2 = &RigUnit_GetCandidates()
   n3 = &RigVMFunction_ControlFlowBranch(
    Condition=RigVMDispatchNode(B=(Type=Connector,Name="End Bone"), A=@n2.Connector) {template="DISPATCH_RigVMDispatch_CoreEquals(in A,in B,out Result)"}.Result
    ExecuteContext=@RigVMFunction_Sequence.C
   )
   n4 = RigVMDispatchNode(ExecuteContext=@n3.True, Array=@n2.Candidates) {template="DISPATCH_RigVMDispatch_ArrayIterator(in Array,out Element,out Index,out Count,out Ratio)"}
   Find_1_1_1 = RigVMFunctionReferenceNode(Element=head, Value=@n4.Element.Name)
   n5 = &RigVMFunction_ControlFlowBranch(ExecuteContext=@n4.ExecuteContext, Condition=@Find_1_1_1.Success)
   &RigUnit_SetDefaultMatch(Default=@n4.Element, ExecutePin=@n5.True) {template="Set Default Match::Execute(in Default)"}
   n6 = &RigUnit_GetCandidates()
   n7 = &RigVMFunction_ControlFlowBranch(
    Condition=RigVMDispatchNode(B=(Type=Connector,Name="Start Bone"), A=@n6.Connector) {template="DISPATCH_RigVMDispatch_CoreEquals(in A,in B,out Result)"}.Result
    ExecuteContext=@RigVMFunction_Sequence.D
   )
   RigVMDispatchNode(ExecuteContext=@n7.True, Array=@n6.Candidates) {template="DISPATCH_RigVMDispatch_ArrayIterator(in Array,out Element,out Index,out Count,out Ratio)"}
   n8 = &RigUnit_ResolveConnector(Connector=(Type=Connector,Name="Animal Neck Primary"), SkipSocket=true)
   > &RigUnit_SetDefaultMatch(Default=@n8.Result) {template="Set Default Match::Execute(in Default)"}
   > RigVMVariableNode(Value=true) {variable=Found}
   For_Each_1_1 = RigVMDispatchNode(Array=(), ExecuteContext=()) {template="DISPATCH_RigVMDispatch_ArrayIterator(in Array,out Element,out Index,out Count,out Ratio)"}
   n9 = &RigVMFunction_ControlFlowBranch(
    ExecuteContext=@For_Each_1_1.ExecuteContext
    Condition=&RigVMFunction_MathBoolOr(
     A=/Script/RigVM.RigVMFunction_Contains(Search=neck, Name=@For_Each_1_1.Element.Name) {is=struct RigVMUnitNode, template="Contains::Execute(in Name,in Search,out Result)"}.Result
     B=&RigVMFunction_Contains(Search=Neck, Name=@For_Each_1_1.Element.Name) {template="Contains::Execute(in Name,in Search,out Result)"}.Result
    ).Result
   )
   &RigUnit_SetDefaultMatch(Default=@For_Each_1_1.Element, ExecutePin=@n9.True) {template="Set Default Match::Execute(in Default)"}
   n10 = &RigUnit_GetCandidates()
   n11 = &RigVMFunction_ControlFlowBranch(
    ExecuteContext=()
    Condition=RigVMDispatchNode(B=(Type=Connector,Name="End Bone"), A=@n10.Connector) {template="DISPATCH_RigVMDispatch_CoreEquals(in A,in B,out Result)"}.Result
   )
   n12 = RigVMDispatchNode(ExecuteContext=@n11.True, Array=@n10.Candidates) {template="DISPATCH_RigVMDispatch_ArrayIterator(in Array,out Element,out Index,out Count,out Ratio)"}
   DISPATCH_RigDispatch_GetModuleMetadata = RigVMDispatchNode(Default=False, Name=Connected at End, NameSpace=Self) {template="DISPATCH_RigDispatch_GetModuleMetadata(in Name,in NameSpace,in Default,out Value,out Found)"}
   n13 = &RigVMFunction_ControlFlowBranch(Condition=@DISPATCH_RigDispatch_GetModuleMetadata.Found, ExecuteContext=@n12.ExecuteContext)
   n14 = &RigVMFunction_ControlFlowBranch(Condition=@DISPATCH_RigDispatch_GetModuleMetadata.Value, ExecuteContext=@n13.True)
   RigVMFunction_ControlFlowBranch = &RigVMFunction_ControlFlowBranch(
    Condition=RigVMDispatchNode(A=@n12.Element, B=&RigUnit_ResolveConnector(Connector=(Type=Connector,Name="Animal Neck Primary"), SkipSocket=true).Result) {template="DISPATCH_RigVMDispatch_CoreEquals(in A,in B,out Result)"}.Result
    ExecuteContext=@n14.True
   )
   &RigUnit_SetDefaultMatch(Default=@n12.Element, ExecutePin=@RigVMFunction_ControlFlowBranch.True) {template="Set Default Match::Execute(in Default)"}
   n15 = RigVMDispatchNode(Default=False, Name=Connected at End, NameSpace=Self) {template="DISPATCH_RigDispatch_GetModuleMetadata(in Name,in NameSpace,in Default,out Value,out Found)"}
   n16 = &RigVMFunction_ControlFlowBranch(Condition=@n15.Found, ExecuteContext=@n12.Completed)
   n17 = &RigVMFunction_ControlFlowBranch(Condition=@n15.Value, ExecuteContext=@n16.False)
   Connect to Leaf Bones = RigVMFunctionReferenceNode(
    Clamp=4
    Connector=(Type=Connector,Name="End Bone")
    Primary Connector=(Type=Connector,Name="Animal Neck Primary")
    ExecuteContext=@n17.False
   )
   n18 = &RigUnit_GetCandidates()
   n19 = &RigVMFunction_ControlFlowBranch(
    ExecuteContext=()
    Condition=RigVMDispatchNode(B=(Type=Connector,Name="Start Bone"), A=@n18.Connector) {template="DISPATCH_RigVMDispatch_CoreEquals(in A,in B,out Result)"}.Result
   )
   n20 = RigVMDispatchNode(ExecuteContext=@n19.True, Array=@n18.Candidates) {template="DISPATCH_RigVMDispatch_ArrayIterator(in Array,out Element,out Index,out Count,out Ratio)"}
   n21 = &RigUnit_ResolveConnector(Connector=(Type=Connector,Name="Animal Neck Primary"), SkipSocket=true)
   n22 = &RigUnit_CollectionChildrenArray(TypeToSearch=Bone, bDefaultChildren=True, bIncludeParent=False, bRecursive=False, Parent=@n21.Result)
   n23 = &RigVMFunction_ControlFlowBranch(
    ExecuteContext=@n20.ExecuteContext
    Condition=RigVMDispatchNode(B=0, A=RigVMDispatchNode(Array=@n22.Items) {template="DISPATCH_RigVMDispatch_ArrayGetNum(in Array,out Num)"}.Num) {template="DISPATCH_RigVMDispatch_CoreEquals(in A,in B,out Result)"}.Result
   )
   &RigUnit_SetDefaultMatch(Default=@n21.Result, ExecutePin=@n23.False) {template="Set Default Match::Execute(in Default)"}
   &RigUnit_SetDefaultMatch(Default=&RigUnit_HierarchyGetParent(bDefaultParent=True, Child=@n21.Result).Parent, ExecutePin=@n23.True) {template="Set Default Match::Execute(in Default)"}
   > DISPATCH_RigDispatch_SetModuleMetadata = RigVMDispatchNode(Name=Connected at End, NameSpace=Self, Value=true) {template="DISPATCH_RigDispatch_SetModuleMetadata(in Name,in NameSpace,in Value,out Success)"}
   RigVMVariableNode(ExecuteContext=(), Value=false) {variable=Found}
   > For_Each = RigVMDispatchNode(Array=&RigUnit_CollectionChildrenArray(TypeToSearch=Bone, bDefaultChildren=True, bIncludeParent=False, bRecursive=true, Parent=@n8.Result).Items) {template="DISPATCH_RigVMDispatch_ArrayIterator(in Array,out Element,out Index,out Count,out Ratio)"}
   Find_1_1 = RigVMFunctionReferenceNode(Element=neck, Value=@For_Each.Element.Name)
   RigVMFunction_MathBoolNot = &RigVMFunction_MathBoolNot(Value=RigVMVariableNode() {variable=Found}.Value)
   RigVMFunction_MathBoolAnd = &RigVMFunction_MathBoolAnd(A=@Find_1_1.Success, B=@RigVMFunction_MathBoolNot.Result)
   &RigVMFunction_ControlFlowBranch(ExecuteContext=@For_Each.ExecuteContext, Condition=@RigVMFunction_MathBoolAnd.Result)
   &RigVMFunction_ControlFlowBranch(ExecuteContext=(), Condition=&RigVMFunction_MathBoolNot(Value=RigVMVariableNode() {variable=Found}.Value).Result)
  ]}
]
hierarchy [
 Animal Neck Primary {type=Connector, loc=<0 0 0>, rot=<0 0 0 1>}
 Start Bone {type=Connector, loc=<0 0 0>, rot=<0 0 0 1>}
 End Bone {type=Connector, loc=<0 0 0>, rot=<0 0 0 1>}
 Parent Control {type=Connector, loc=<0 0 0>, rot=<0 0 0 1>}
 Settings Control {type=Connector, loc=<0 0 0>, rot=<0 0 0 1>}
 Pose_0 {type=Curve, loc=<0 0 0>, rot=<0 0 0 1>}
 Pose_1 {type=Curve, loc=<0 0 0>, rot=<0 0 0 1>}
]
variables [
 Add Null {type=bool, default=False}
 Create Control per Bone {type=bool, default=False}
 Creation Spline Bones {type="TArray<FRigElementKey>"}
 Curve Points {type="TArray<FVector>", default=((X=0,Y=0,Z=0))}
 Debug {type=bool, default=False}
 Default Blend End {type=double, default=1}
 Default Blend Start {type=double, default=0.5}
 Default Blend Twist {type=double, default=1}
 Default Damping {type=double, default=0.8}
 Default IK or FK {type=bool, default=True}
 Default Strength {type=double, default=4}
 Dual Spline {type=FControlRigSpline, default=()}
 Dynamics {type=bool, default=False}
 FK Control Settings {
  type=FRigUnit_HierarchyAddControlTransform_Settings
  default="(InitialSpace=LocalSpace,bUsePreferredRotationOrder=False,PreferredRotationOrder=YZX,Limits=(LimitTranslationX=(bMinimum=False,bMaximum=False),LimitTranslationY=(bMinimum=False,bMaximum=False),LimitTranslationZ=(bMinimum=False,bMaximum=False),LimitPitch=(b ...(truncated)"
 }
 Found {type=bool, default=False}
 Head Local Control {type=FRigElementKey, default=(Type=None,Name="")}
 Head Local Settings {
  type=FRigUnit_HierarchyAddControlTransform_Settings
  default="(InitialSpace=LocalSpace,bUsePreferredRotationOrder=False,PreferredRotationOrder=YZX,Limits=(LimitTranslationX=(bMinimum=False,bMaximum=False),LimitTranslationY=(bMinimum=False,bMaximum=False),LimitTranslationZ=(bMinimum=False,bMaximum=False),LimitPitch=(b ...(truncated)"
 }
 Head Settings {
  type=FRigUnit_HierarchyAddControlTransform_Settings
  default="(InitialSpace=LocalSpace,bUsePreferredRotationOrder=False,PreferredRotationOrder=YZX,Limits=(LimitTranslationX=(bMinimum=False,bMaximum=False),LimitTranslationY=(bMinimum=False,bMaximum=False),LimitTranslationZ=(bMinimum=False,bMaximum=False),LimitPitch=(b ...(truncated)"
 }
 HeadNeck Settings {
  type=FRigUnit_HierarchyAddControlTransform_Settings
  default="(InitialSpace=LocalSpace,bUsePreferredRotationOrder=False,PreferredRotationOrder=YZX,Limits=(LimitTranslationX=(bMinimum=False,bMaximum=False),LimitTranslationY=(bMinimum=False,bMaximum=False),LimitTranslationZ=(bMinimum=False,bMaximum=False),LimitPitch=(b ...(truncated)"
 }
 Headneck Control Pivot Position {type=double, default=0.75}
 Inner Settings {
  type=FRigUnit_HierarchyAddControlTransform_Settings
  default="(InitialSpace=LocalSpace,bUsePreferredRotationOrder=False,PreferredRotationOrder=YZX,Limits=(LimitTranslationX=(bMinimum=False,bMaximum=False),LimitTranslationY=(bMinimum=False,bMaximum=False),LimitTranslationZ=(bMinimum=False,bMaximum=False),LimitPitch=(b ...(truncated)"
 }
 Local Spline {type=FControlRigSpline, default=()}
 Main Controls as FK {type=bool, default=False}
 Module Settings {
  type=FUserDefinedStruct_#9
  default="(OverrideColor_48_#7=False,LeftColor_3_#6=(R=0,G=0.0231,B=0.4115,A=1),RightColor_5_#8=(R=1,G=0,B=0,A=1),CenterColor_9_#5=(R=1,G=1,B=0,A= ...(truncated)"
 }
 Name {type=FName, default=None}
 Number of Controls {type=int32, default=5}
 Number of Main Controls {type=int32, default=4}
 Number of Neck Bones {type=int32, default=7}
 Prefix {type=FName, default=FOo}
 Primary Axis {type=FVector, default=(X=1,Y=0,Z=0)}
 Re-Orient Controls {type=bool, default=True}
 Rig Bone Positions {type="TArray<FTransform>", default=(())}
 Secondary Axis {type=FVector, default=(X=0,Y=1,Z=0)}
 Secondary Spline Direction {type=FVector, default=(X=0,Y=5,Z=0)}
 Solve with SplineIK {type=bool, default=True}
 Spline Mode {type=ESplineType, default=BSpline}
 Stretch {type=bool, default=True}
 Suffix {type=FName, default=_ctrl}
 Tangent Settings {
  type=FRigUnit_HierarchyAddControlTransform_Settings
  default="(InitialSpace=LocalSpace,bUsePreferredRotationOrder=False,PreferredRotationOrder=YZX,Limits=(LimitTranslationX=(bMinimum=False,bMaximum=False),LimitTranslationY=(bMinimum=False,bMaximum=False),LimitTranslationZ=(bMinimum=False,bMaximum=False),LimitPitch=(b ...(truncated)"
 }
 Template Spline {type=FControlRigSpline, default=()}
]
