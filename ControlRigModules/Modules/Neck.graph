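# Modular Control Rig neck module that builds and drives a switchable FK/IK neck and head between connector-resolved bones.
# Construction creates the neck space, rig bone chain and splines from module settings.
# IK places bones along a B-spline to the head control with optional stretch, aiming each using interpolated rotation.
# The inactive controls are hidden and follow for matching, switches request auto-keys, and head spaces can take a reference's orientation.
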ControlRigBlueprint /ControlRigModules/Modules/Neck
graphs [
 Concat_4_1_ContainedGraph {nodes [
   n1 = RigVMFunctionEntryNode()
   n2 = /Script/RigVM.RigVMFunction_NameConcat(A=@n1.A, B=@n1.B) {is=struct RigVMUnitNode, template="Concat::Execute(in A,in B,out Result)"}
   RigVMFunctionReturnNode(Result=&RigVMFunction_NameConcat(A=&RigVMFunction_NameConcat(A=@n2.Result, B=@n1.C) {template="Concat::Execute(in A,in B,out Result)"}.Result, B=@n1.D) {template="Concat::Execute(in A,in B,out Result)"}.Result)
  ]}
 Concat_4_1_ContainedGraph {nodes [
   n1 = RigVMFunctionEntryNode()
   n2 = &RigVMFunction_NameConcat(A=@n1.A, B=@n1.B) {template="Concat::Execute(in A,in B,out Result)"}
   RigVMFunctionReturnNode(Result=&RigVMFunction_NameConcat(A=&RigVMFunction_NameConcat(A=@n2.Result, B=@n1.C) {template="Concat::Execute(in A,in B,out Result)"}.Result, B=@n1.D) {template="Concat::Execute(in A,in B,out Result)"}.Result)
  ]}
 ConstructionGraph {nodes [
   n1 = &RigUnit_PrepareForExecution()
   CONSTR Neck = RigVMFunctionReferenceNode(
    Head Connector=(Type=Connector,Name="Head Bone")
    Neck Start Connector=(Type=Connector,Name="Neck Start Bone")
    Parent Control Connector=(Type=Connector,Name="Parent Control")
    Settings Control Connector=(Type=Connector,Name="Settings Control")
    ExecuteContext=@n1.ExecutePin
    prefix=RigVMVariableNode() {variable=Module Settings}.Value.Prefix_20_#1
    Head IK Settings=RigVMVariableNode() {variable=Head IK Settings}.Value
    Neck FK Settings=RigVMVariableNode() {variable=Neck FK Settings}.Value
    Head FK Settings=RigVMVariableNode() {variable=Head FK Settings}.Value
    FK IK Default=RigVMVariableNode() {variable=FK IK Default}.Value
   )
   > RigVMVariableNode(Value=@"CONSTR Neck".Local Spline Out) {variable=Local Spline Out}
   > RigVMVariableNode(Value=@"CONSTR Neck".Template Spline Out) {variable=Template Spline Out}
  ]}
 ContainedGraph {nodes [
   n1 = RigVMFunctionEntryNode()
   > n2 = RigVMDispatchNode(Array=@n1.Children) {template="DISPATCH_RigVMDispatch_ArrayIterator(in Array,out Element,out Index,out Count,out Ratio)"}
   n3 = RigVMDispatchNode(Index=@n2.Index, Array=@n1.Parents) {template="DISPATCH_RigVMDispatch_ArrayGetAtIndex(in Array,in Index,out Element)"}
   > &RigUnit_SetTransform(
    Weight=1
    Item=@n2.Element
    bPropagateToChildren=@n1.Propagate to Children
    Value=/Script/ControlRig.RigUnit_ProjectTransformToNewParent(bChildInitial=True, bOldParentInitial=True, NewParent=@n3.Element, OldParent=@n3.Element, Child=@n2.Element) {is=struct RigVMUnitNode}.Transform
   )
   RigVMFunctionReturnNode(ExecuteContext=@n2.Completed)
  ]}
 ContainedGraph {nodes [
   n1 = RigVMFunctionEntryNode()
   > SequenceExecution = RigVMAggregateNode()
   n2 = RigVMDispatchNode(ExecuteContext=@SequenceExecution.A, Array=RigVMVariableNode() {variable=FK Controls}.Value) {template="DISPATCH_RigVMDispatch_ArrayIterator(in Array,out Element,out Index,out Count,out Ratio)"}
   > /Script/ControlRig.RigUnit_SetControlVisibility(
    Item=@n2.Element
    bVisible=RigVMDispatchNode(False=False, Condition=@n1.Show Controls, True=&RigVMFunction_MathBoolNot(Value=@n1.IKFK Switch).Result) {template="DISPATCH_RigVMDispatch_If(in Condition,in True,in False,out Result)"}.Result
   ) {is=struct RigVMUnitNode, template="SetControlVisibility::Execute(in Item,in Pattern,in bVisible)"}
   n3 = RigVMDispatchNode(ExecuteContext=@SequenceExecution.B, Array=RigVMVariableNode() {variable=IK Controls}.Value) {template="DISPATCH_RigVMDispatch_ArrayIterator(in Array,out Element,out Index,out Count,out Ratio)"}
   > &RigUnit_SetControlVisibility(
    Item=@n3.Element
    bVisible=RigVMDispatchNode(False=False, True=@n1.IKFK Switch, Condition=@n1.Show Controls) {template="DISPATCH_RigVMDispatch_If(in Condition,in True,in False,out Result)"}.Result
   ) {template="SetControlVisibility::Execute(in Item,in Pattern,in bVisible)"}
   RigVMFunctionReturnNode(ExecuteContext=@SequenceExecution.D)
  ]}
 ContainedGraph {nodes [
   n1 = RigVMFunctionEntryNode()
   SequenceExecution = &RigVMFunction_Sequence(ExecuteContext=@n1.ExecuteContext)
   SequenceExecution_1 = &RigVMFunction_Sequence(ExecuteContext=@SequenceExecution.B)
   SequenceExecution_2 = &RigVMFunction_Sequence(ExecuteContext=@SequenceExecution_1.B)
   RigVMFunctionReturnNode(A=@SequenceExecution.A, B=@SequenceExecution_1.A, C=@SequenceExecution_2.A, D=@SequenceExecution_2.B)
  ]}
 ContainedGraph {nodes [
   n1 = RigVMFunctionEntryNode()
   > SequenceExecution_1 = RigVMAggregateNode()
   MathBoolNotEquals = RigVMDispatchNode(A=@n1.IK Mode, B=@n1.Previous IK Mode) {template="DISPATCH_RigVMDispatch_CoreNotEquals(in A,in B,out Result)"}
   n2 = &RigVMFunction_ControlFlowBranch(Condition=@MathBoolNotEquals.Result, ExecuteContext=@SequenceExecution_1.A)
   SequenceExecution = RigVMAggregateNode(ExecuteContext=@n2.True)
   n3 = RigVMDispatchNode(ExecuteContext=@SequenceExecution.A, Array=RigVMVariableNode() {variable=FK Controls}.Value) {template="DISPATCH_RigVMDispatch_ArrayIterator(in Array,out Element,out Index,out Count,out Ratio)"}
   > /Script/ControlRig.RigUnit_SendEvent(Event=RequestAutoKey, bEnable=True, bOnlyDuringInteraction=True, Item=@n3.Element) {is=struct RigVMUnitNode, template="SendEvent::Execute(in Event,in Item,in OffsetInSeconds,in bEnable,in bOnlyDuringInteraction)"}
   n4 = RigVMDispatchNode(ExecuteContext=@SequenceExecution.B, Array=RigVMVariableNode() {variable="IK Controls "}.Value) {template="DISPATCH_RigVMDispatch_ArrayIterator(in Array,out Element,out Index,out Count,out Ratio)"}
   > &RigUnit_SendEvent(Event=RequestAutoKey, bEnable=True, bOnlyDuringInteraction=True, Item=@n4.Element) {template="SendEvent::Execute(in Event,in Item,in OffsetInSeconds,in bEnable,in bOnlyDuringInteraction)"}
   &RigUnit_SendEvent(
    Event=RequestAutoKey
    bEnable=True
    bOnlyDuringInteraction=True
    ExecutePin=@SequenceExecution.C
    Item=RigVMVariableNode() {variable=Switch Control}.Value
   ) {template="SendEvent::Execute(in Event,in Item,in OffsetInSeconds,in bEnable,in bOnlyDuringInteraction)"}
   RigVMFunctionReturnNode(ExecuteContext=@SequenceExecution_1.D)
  ]}
 ContainedGraph {nodes [
   n1 = RigVMFunctionEntryNode()
   SequenceExecution_1 = &RigVMFunction_Sequence(ExecuteContext=@n1.ExecuteContext)
   SequenceExecution_1_1 = &RigVMFunction_Sequence(ExecuteContext=@SequenceExecution_1.B)
   SequenceExecution_1_2 = &RigVMFunction_Sequence(ExecuteContext=@SequenceExecution_1_1.B)
   RigVMFunctionReturnNode(A=@SequenceExecution_1.A, B=@SequenceExecution_1_1.A, C=@SequenceExecution_1_2.A, D=@SequenceExecution_1_2.B)
  ]}
 ContainedGraph {nodes [
   n1 = RigVMFunctionEntryNode()
   SequenceExecution = &RigVMFunction_Sequence(ExecuteContext=@n1.ExecuteContext)
   SequenceExecution_1 = &RigVMFunction_Sequence(ExecuteContext=@SequenceExecution.B)
   SequenceExecution_2 = &RigVMFunction_Sequence(ExecuteContext=@SequenceExecution_1.B)
   RigVMFunctionReturnNode(A=@SequenceExecution.A, B=@SequenceExecution_1.A, C=@SequenceExecution_2.A, D=@SequenceExecution_2.B)
  ]}
 ContainedGraph {nodes [
   RigVMFunctionEntryNode()
   n1 = RigVMVariableNode() {variable=IKFK Switch Control}
   > RigVMVariableNode(Value=/Script/RigVM.RigVMFunction_MathFloatGreater(B=0.5, A=&RigUnit_GetTransform(Space=LocalSpace, Item=@n1.Value).Transform.Translation.X) {is=struct RigVMUnitNode, template="Greater::Execute(in A,in B,out Result)"}.Result) {variable=IK Mode}
   n2 = RigVMVariableNode() {variable=Head Orient Control}
   > RigVMVariableNode(Value=&RigVMFunction_MathFloatGreater(B=0.5, A=&RigUnit_GetTransform(Space=LocalSpace, Item=@n2.Value).Transform.Translation.X) {template="Greater::Execute(in A,in B,out Result)"}.Result) {variable=Local Space}
   > SequenceExecution = RigVMAggregateNode()
   Set IKFK Visibility = RigVMFunctionReferenceNode(
    IK Controls=((Type=Control,Name="head_ik_ctrl"))
    Show Controls=true
    ExecuteContext=@SequenceExecution.A
    IKFK Switch=RigVMVariableNode() {variable=IK Mode}.Value
    FK Controls=RigVMVariableNode() {variable=FK Controls}.Value
    IK Controls.0=RigVMVariableNode() {variable=IK Control}.Value
   )
   n3 = RigVMDispatchNode(Index=1, Array=RigVMVariableNode() {variable=FK Controls}.Value) {template="DISPATCH_RigVMDispatch_ArrayGetAtIndex(in Array,in Index,out Element)"}
   n4 = RigVMVariableNode() {variable=Head FK Space}
   n5 = &RigUnit_ProjectTransformToNewParent(bChildInitial=True, bOldParentInitial=True, NewParent=@n3.Element, OldParent=@n3.Element, Child=@n4.Value)
   n6 = RigVMVariableNode() {variable=Head Orient Reference}
   &RigUnit_SetTransform(
    Value=(Rotation=(X=0,Y=0,Z=0,W=1),Translation=(X=0,Y=0,Z=0),Scale3D=(X=1,Y=1,Z=1))
    Weight=1
    Value.Rotation=RigVMDispatchNode(
     False=@n5.Transform.Rotation
     True=&RigUnit_ProjectTransformToNewParent(bChildInitial=True, bOldParentInitial=True, Child=@n4.Value, NewParent=@n6.Value, OldParent=@n6.Value).Transform.Rotation
     Condition=RigVMVariableNode() {variable=Local Space}.Value
    ) {template="DISPATCH_RigVMDispatch_If(in Condition,in True,in False,out Result)"}.Result
    Value.Translation=@n5.Transform.Translation
    ExecutePin=@SequenceExecution.B
    Item=@n4.Value
   )
   n7 = RigVMVariableNode() {variable=Head IK Space}
   n8 = RigVMVariableNode() {variable=Head Orient Reference}
   n9 = /Script/ControlRig.RigUnit_HierarchyGetParent(bDefaultParent=True, Child=@n7.Value) {is=struct RigVMUnitNode}
   n10 = &RigUnit_ProjectTransformToNewParent(bChildInitial=True, bOldParentInitial=True, NewParent=@n9.Parent, OldParent=@n9.Parent, Child=@n7.Value)
   > &RigUnit_SetTransform(
    Value=(Rotation=(X=0,Y=0,Z=0,W=1),Translation=(X=0,Y=0,Z=0),Scale3D=(X=1,Y=1,Z=1))
    Weight=1
    Value.Rotation=RigVMDispatchNode(
     True=&RigUnit_ProjectTransformToNewParent(bChildInitial=True, bOldParentInitial=True, Child=@n7.Value, NewParent=@n8.Value, OldParent=@n8.Value).Transform.Rotation
     False=@n10.Transform.Rotation
     Condition=RigVMVariableNode() {variable=Local Space}.Value
    ) {template="DISPATCH_RigVMDispatch_If(in Condition,in True,in False,out Result)"}.Result
    Value.Translation=@n10.Transform.Translation
    Item=@n7.Value
   )
   n11 = &RigVMFunction_ControlFlowBranch(ExecuteContext=@SequenceExecution.C, Condition=RigVMVariableNode() {variable=IK Mode}.Value)
   Project to New Parents_2 = RigVMFunctionReferenceNode(
    Propagate to Children=True
    ExecuteContext=@n11.False
    Children=RigVMVariableNode() {variable=Bones}.Value
    Parents=RigVMVariableNode() {variable=FK Controls}.Value
   )
   n12 = RigVMVariableNode() {variable=Bones}
   n13 = RigVMDispatchNode(
    Index=/Script/RigVM.RigVMFunction_MathIntSub(B=1, A=RigVMDispatchNode(Array=@n12.Value) {template="DISPATCH_RigVMDispatch_ArrayGetNum(in Array,out Num)"}.Num) {is=struct RigVMUnitNode, template="Subtract::Execute(in A,in B,out Result)"}.Result
    Array=@n12.Value
   ) {template="DISPATCH_RigVMDispatch_ArrayGetAtIndex(in Array,in Index,out Element)"}
   n14 = RigVMVariableNode() {variable=IK Control}
   > &RigUnit_SetTransform(
    Weight=1
    Value=&RigUnit_ProjectTransformToNewParent(bChildInitial=True, bOldParentInitial=True, NewParent=@n13.Element, OldParent=@n13.Element, Child=@n14.Value).Transform
    Item=@n14.Value
   )
   Neck IK Neck = RigVMCollapseNode(
    ExecuteContext=@n11.True
    Length=RigVMVariableNode() {variable=Length}.Value
    Enable Stretch=RigVMVariableNode() {variable=Enable Stretch}.Value
    IK Control=RigVMVariableNode() {variable=IK Control}.Value
    Mid Null=RigVMVariableNode() {variable=Mid IK Null}.Value
    Bones=RigVMVariableNode() {variable=Bones}.Value
    Positions=RigVMVariableNode() {variable=Positions}.Value
   )
   > Project to New Parents = RigVMFunctionReferenceNode(Propagate to Children=True, Parents=RigVMVariableNode() {variable=Bones}.Value, Children=RigVMVariableNode() {variable=FK Controls}.Value)
   Set IKFK Keys = RigVMFunctionReferenceNode(
    "IK Controls "=((Type=Control,Name="head_ik_ctrl"))
    ExecuteContext=@SequenceExecution.D
    IK Mode=RigVMVariableNode() {variable=IK Mode}.Value
    Previous IK Mode=RigVMVariableNode() {variable=Previous IK Mode}.Value
    FK Controls=RigVMVariableNode() {variable=FK Controls}.Value
    Switch Control=RigVMVariableNode() {variable=IKFK Switch Control}.Value
    IK Controls .0=RigVMVariableNode() {variable=IK Control}.Value
   )
   > RigVMFunctionReturnNode()
  ]}
 ContainedGraph {nodes [
   n1 = RigVMFunctionEntryNode()
   > Get Points on Spline = RigVMCollapseNode(
    Bones=@n1.Bones
    Enable Stretch=@n1.Enable Stretch
    IK Control=@n1.IK Control
    Length=@n1.Length
    Mid Null=@n1.Mid Null
    Array=RigVMVariableNode() {variable=Positions}.Value
   )
   > SequenceExecution_1 = RigVMAggregateNode()
   n2 = RigVMDispatchNode(ExecuteContext=@SequenceExecution_1.A, Array=RigVMVariableNode() {variable=Bones}.Value) {template="DISPATCH_RigVMDispatch_ArrayIterator(in Array,out Element,out Index,out Count,out Ratio)"}
   At and Next = RigVMCollapseNode(Index=@n2.Index, Array=RigVMVariableNode() {variable=Positions}.Value)
   First Last Rotation = RigVMCollapseNode(
    First Bone=RigVMDispatchNode(Index=0, Array=RigVMVariableNode() {variable=Bones}.Value) {template="DISPATCH_RigVMDispatch_ArrayGetAtIndex(in Array,in Index,out Element)"}.Element
    T=@n2.Ratio
    IK Control=RigVMVariableNode() {variable=IK Control}.Value
   )
   n3 = /Script/ControlRig.RigUnit_AimBoneMath(
    DebugSettings=(Scale=10,WorldOffset=(Rotation=(X=0,Y=0,Z=0,W=1),Translation=(X=0,Y=0,Z=0),Scale3D=(X=1,Y=1,Z=1)))
    InputTransform=(Rotation=(X=0,Y=0,Z=0,W=1),Translation=(X=0,Y=0,Z=0),Scale3D=(X=1,Y=1,Z=1))
    Primary=(Weight=1,Axis=(X=1,Y=0,Z=0),Target=(X=1,Y=0,Z=0),Kind=Location,Space=(Type=Bone))
    Secondary=(Weight=1,Axis=(X=0,Y=0,Z=1),Target=(X=1,Y=0,Z=0),Kind=Direction,Space=(Type=Bone))
    Weight=1
    InputTransform.Translation=@"At and Next".At
    Primary.Target=@"At and Next".Next
    Secondary.Target=/Script/RigVM.RigVMFunction_MathQuaternionRotateVector(Vector=(X=0,Y=0,Z=1), Transform=@"First Last Rotation".Interpolate) {is=struct RigVMUnitNode, template="Rotate Vector::Execute(in Transform,in Vector,out Result)"}.Result
   ) {is=struct RigVMUnitNode}
   Is Last = RigVMCollapseNode(Count=@n2.Count, Index=@n2.Index)
   > &RigUnit_SetTransform(
    Value=(Rotation=(X=0,Y=0,Z=0,W=1),Translation=(X=0,Y=0,Z=0),Scale3D=(X=1,Y=1,Z=1))
    Weight=1
    Item=@n2.Element
    Value.Translation=@"At and Next".At
    Value.Rotation=RigVMDispatchNode(False=@n3.Result.Rotation, True=@"First Last Rotation".Last, Condition=@"Is Last".Result) {template="DISPATCH_RigVMDispatch_If(in Condition,in True,in False,out Result)"}.Result
   )
   RigVMFunctionReturnNode(ExecuteContext=@SequenceExecution_1.D)
  ]}
 ContainedGraph {nodes [
   RigVMFunctionEntryNode()
   > SequenceExecution = RigVMAggregateNode()
   n1 = RigVMDispatchNode(ExecuteContext=@SequenceExecution.A, Array=RigVMVariableNode() {variable=Array}.Value) {template="DISPATCH_RigVMDispatch_ArrayReset(io Array)"}
   > n2 = RigVMDispatchNode(Array=RigVMVariableNode() {variable=Bones}.Value) {template="DISPATCH_RigVMDispatch_ArrayIterator(in Array,out Element,out Index,out Count,out Ratio)"}
   n3 = &RigUnit_GetTransform(Item=RigVMDispatchNode(Index=0, Array=RigVMVariableNode() {variable=Bones}.Value) {template="DISPATCH_RigVMDispatch_ArrayGetAtIndex(in Array,in Index,out Element)"}.Element)
   RigUnit_GetTransform = &RigUnit_GetTransform(Item=RigVMVariableNode() {variable=Mid Null}.Value)
   n4 = /Script/ControlRigSpline.RigUnit_ControlRigSplineFromPoints(
    Points=((X=0,Y=0,Z=0),(X=0,Y=0,Z=0),(X=0,Y=0,Z=0),(X=0,Y=0,Z=0))
    SamplesPerSegment=16
    SplineMode=BSpline
    Points.0=@n3.Transform.Translation
    Points.1=@n3.Transform.Translation
    Points.3=&RigUnit_GetTransform(Item=RigVMVariableNode() {variable=IK Control}.Value).Transform.Translation
    Points.2=@RigUnit_GetTransform.Transform.Translation
   ) {is=struct RigVMUnitNode}
   n5 = /Script/RigVM.RigVMFunction_MathFloatDiv(
    B=/Script/ControlRigSpline.RigUnit_GetLengthControlRigSpline(Spline=@n4.Spline) {is=struct RigVMUnitNode}.Length
    A=RigVMVariableNode() {variable=Length}.Value
   ) {is=struct RigVMUnitNode, template="Divide::Execute(in A,in B,out Result)"}
   n6 = &RigVMFunction_MathFloatMul(
    A=@n2.Ratio
    B=RigVMDispatchNode(
     True=1
     False=/Script/RigVM.RigVMFunction_MathFloatMin(B=1, A=@n5.Result) {is=struct RigVMUnitNode, template="Minimum::Execute(in A,in B,out Result)"}.Result
     Condition=RigVMVariableNode() {variable=Enable Stretch}.Value
    ) {template="DISPATCH_RigVMDispatch_If(in Condition,in True,in False,out Result)"}.Result
   ) {template="Multiply::Execute(in A,in B,out Result)"}
   > RigVMDispatchNode(
    Array=@n1.Array
    Element=/Script/ControlRigSpline.RigUnit_PositionFromControlRigSpline(
     Spline=@n4.Spline
     U=/Script/ControlRigSpline.RigUnit_ParameterAtPercentage(Spline=@n4.Spline, Percentage=@n6.Result) {is=struct RigVMUnitNode}.U
    ) {is=struct RigVMUnitNode}.Position
   ) {template="DISPATCH_RigVMDispatch_ArrayAdd(io Array,in Element,out Index)"}
   RigVMFunctionReturnNode(ExecuteContext=@SequenceExecution.D, Positions=RigVMVariableNode() {variable=Array}.Value)
  ]}
 ContainedGraph {nodes [
   n1 = RigVMFunctionEntryNode()
   SequenceExecution = &RigVMFunction_Sequence(ExecuteContext=@n1.ExecuteContext)
   SequenceExecution_1 = &RigVMFunction_Sequence(ExecuteContext=@SequenceExecution.B)
   SequenceExecution_2 = &RigVMFunction_Sequence(ExecuteContext=@SequenceExecution_1.B)
   RigVMFunctionReturnNode(A=@SequenceExecution.A, B=@SequenceExecution_1.A, C=@SequenceExecution_2.A, D=@SequenceExecution_2.B)
  ]}
 ContainedGraph {nodes [
   n1 = RigVMFunctionEntryNode()
   MathIntEquals = RigVMDispatchNode(A=@n1.Index, B=&RigVMFunction_MathIntSub(B=1, A=@n1.Count) {template="Subtract::Execute(in A,in B,out Result)"}.Result) {template="DISPATCH_RigVMDispatch_CoreEquals(in A,in B,out Result)"}
   RigVMFunctionReturnNode(Result=@MathIntEquals.Result)
  ]}
 ContainedGraph {nodes [
   n1 = RigVMFunctionEntryNode()
   RigVMFunctionReturnNode(
    At=RigVMDispatchNode(Array=@n1.Array, Index=@n1.Index) {template="DISPATCH_RigVMDispatch_ArrayGetAtIndex(in Array,in Index,out Element)"}.Element
    Next=RigVMDispatchNode(
     Array=@n1.Array
     Index=/Script/RigVM.RigVMFunction_MathIntAdd(B=1, A=@n1.Index) {is=struct RigVMUnitNode, template="Add::Execute(in A,in B,out Result)"}.Result
    ) {template="DISPATCH_RigVMDispatch_ArrayGetAtIndex(in Array,in Index,out Element)"}.Element
   )
  ]}
 ContainedGraph {nodes [
   n1 = RigVMFunctionEntryNode()
   n2 = /Script/RigVM.RigVMFunction_MathQuaternionMul(
    A=&RigUnit_GetTransform(Item=@n1.IK Control).Transform.Rotation
    B=/Script/RigVM.RigVMFunction_MathQuaternionFromRotatorV2(Value=(Pitch=90,Yaw=0,Roll=0)) {is=struct RigVMUnitNode, template="Cast::Execute(in Value,out Result)"}.Result
   ) {is=struct RigVMUnitNode, template="Multiply::Execute(in A,in B,out Result)"}
   n3 = &RigUnit_HierarchyGetParent(bDefaultParent=True, Child=@n1.First Bone)
   n4 = &RigUnit_ProjectTransformToNewParent(bChildInitial=True, bOldParentInitial=True, Child=@n1.First Bone, NewParent=@n3.Parent, OldParent=@n3.Parent)
   RigVMFunctionReturnNode(
    Last=@n2.Result
    Interpolate=/Script/RigVM.RigVMFunction_MathQuaternionSlerp(T=@n1.T, B=@n2.Result, A=@n4.Transform.Rotation) {is=struct RigVMUnitNode, template="Interpolate::Execute(in A,in B,in T,out Result)"}.Result
    First=@n4.Transform.Rotation
   )
  ]}
 ContainedGraph {nodes [
   n1 = RigVMFunctionEntryNode()
   SequenceExecution_1 = &RigVMFunction_Sequence(ExecuteContext=@n1.ExecuteContext)
   SequenceExecution_1_1 = &RigVMFunction_Sequence(ExecuteContext=@SequenceExecution_1.B)
   SequenceExecution_1_2 = &RigVMFunction_Sequence(ExecuteContext=@SequenceExecution_1_1.B)
   RigVMFunctionReturnNode(A=@SequenceExecution_1.A, B=@SequenceExecution_1_1.A, C=@SequenceExecution_1_2.A, D=@SequenceExecution_1_2.B)
  ]}
 ContainedGraph {nodes [
   n1 = RigVMFunctionEntryNode()
   SequenceExecution = &RigVMFunction_Sequence(ExecuteContext=@n1.ExecuteContext)
   SequenceExecution_1 = &RigVMFunction_Sequence(ExecuteContext=@SequenceExecution.B)
   SequenceExecution_2 = &RigVMFunction_Sequence(ExecuteContext=@SequenceExecution_1.B)
   RigVMFunctionReturnNode(A=@SequenceExecution.A, B=@SequenceExecution_1.A, C=@SequenceExecution_2.A, D=@SequenceExecution_2.B)
  ]}
 ContainedGraph {nodes [
   BranchNode_RigVMUnitNode = &RigVMFunction_ControlFlowBranch(ExecuteContext=(), Condition=RigVMVariableNode() {variable=debug}.Value)
   /Script/ControlRigSpline.RigUnit_DrawControlRigSpline(
    Color=(R=1,G=0,B=0,A=1)
    Detail=16
    Thickness=0.75
    ExecutePin=@BranchNode_RigVMUnitNode.True
    Spline=RigVMVariableNode() {variable=template_spline}.Value
   ) {is=struct RigVMUnitNode}
   n1 = RigVMFunctionEntryNode()
   n2 = RigVMVariableNode() {variable=Controls}
   BranchNode_4_RigVMUnitNode = &RigVMFunction_ControlFlowBranch(
    ExecuteContext=@n1.ExecuteContext
    Condition=/Script/RigVM.RigVMFunction_MathIntLess(B=4, A=RigVMDispatchNode(Array=@n2.Value) {template="DISPATCH_RigVMDispatch_ArrayGetNum(in Array,out Num)"}.Num) {is=struct RigVMUnitNode, template="Less::Execute(in A,in B,out Result)"}.Result
   )
   RigVMFunctionReturnNode(
    ExecuteContext=@BranchNode_4_RigVMUnitNode.Completed
    Output Spline=RigVMVariableNode() {variable=local_spline}.Value
    Out Template Spline=RigVMVariableNode() {variable=template_spline}.Value
   )
   RigUnit_SequenceExecution_1 = RigVMAggregateNode(ExecuteContext=@BranchNode_4_RigVMUnitNode.False)
   RigVMVariableNode(Value=@n1.Template Spline, ExecuteContext=@RigUnit_SequenceExecution_1.A) {variable=template_spline}
   n3 = /Script/ControlRig.RigUnit_GetTransformItemArray(Space=GlobalSpace, bInitial=False, Items=RigVMVariableNode() {variable=Controls}.Value) {is=struct RigVMUnitNode}
   > RigVMVariableNode(Value=@n3.Transforms) {variable=template_transforms}
   > RigUnit_SequenceExecution = RigVMAggregateNode()
   n4 = RigVMVariableNode() {variable=template_transforms}
   RigVMDispatchNode(
    Num=RigVMDispatchNode(Array=@n4.Value) {template="DISPATCH_RigVMDispatch_ArrayGetNum(in Array,out Num)"}.Num
    ExecuteContext=@RigUnit_SequenceExecution.A
    Array=RigVMVariableNode() {variable=offset_positions}.Value
   ) {template="DISPATCH_RigVMDispatch_ArraySetNum(io Array,in Num)"}
   > n5 = RigVMDispatchNode(Array=@n4.Value) {template="DISPATCH_RigVMDispatch_ArrayIterator(in Array,out Element,out Index,out Count,out Ratio)"}
   > RigVMDispatchNode(
    Index=@n5.Index
    Element=/Script/RigVM.RigVMFunction_MathTransformTransformVector(Transform=@n5.Element, Location=RigVMVariableNode() {variable=Secondary Spline Direction}.Value) {is=struct RigVMUnitNode}.Result
    Array=RigVMVariableNode() {variable=offset_positions}.Value
   ) {template="DISPATCH_RigVMDispatch_ArraySetAtIndex(io Array,in Index,in Element)"}
   RigUnit_SequenceExecution_4 = RigVMAggregateNode(ExecuteContext=@RigUnit_SequenceExecution.C)
   n6 = &RigVMFunction_ControlFlowBranch(ExecuteContext=@RigUnit_SequenceExecution.D, Condition=RigVMVariableNode() {variable=Initial}.Value)
   n7 = RigVMDispatchNode(ExecuteContext=@n6.True, Array=RigVMVariableNode() {variable=Bones}.Value) {template="DISPATCH_RigVMDispatch_ArrayIterator(in Array,out Element,out Index,out Count,out Ratio)"}
   DISPATCH_RigVMDispatch_ArrayGetAtIndex = RigVMDispatchNode(Index=@n7.Index, Array=RigVMVariableNode() {variable=aim_transforms}.Value) {template="DISPATCH_RigVMDispatch_ArrayGetAtIndex(in Array,in Index,out Element)"}
   n8 = &RigUnit_AimBoneMath(
    DebugSettings=(Scale=10,WorldOffset=(Rotation=(X=0,Y=0,Z=0,W=1),Translation=(X=0,Y=0,Z=0),Scale3D=(X=1,Y=1,Z=1)))
    Primary=(Weight=1,Axis=(X=0,Y=1,Z=0),Target=(X=1,Y=0,Z=0),Kind=Location,Space=(Type=Bone))
    Secondary=(Weight=0,Axis=(X=0,Y=0,Z=0),Target=(X=0,Y=0,Z=0),Kind=Location,Space=(Type=Bone))
    Weight=1
    Primary.Target=@DISPATCH_RigVMDispatch_ArrayGetAtIndex.Element
    InputTransform=&RigUnit_GetTransform(Item=@n7.Element).Transform
    Primary.Axis=RigVMVariableNode() {variable=Up Axis}.Value
   )
   MathIntEquals = RigVMDispatchNode(A=@n7.Index, B=&RigVMFunction_MathIntSub(B=1, A=@n7.Count) {template="Subtract::Execute(in A,in B,out Result)"}.Result) {template="DISPATCH_RigVMDispatch_CoreEquals(in A,in B,out Result)"}
   > n9 = &RigUnit_SetTransform(Weight=1, bInitial=False, Value=@n8.Result, Item=@n7.Element, bPropagateToChildren=@MathIntEquals.Result)
   BranchNode_3_RigVMUnitNode = &RigVMFunction_ControlFlowBranch(ExecuteContext=@n9.ExecutePin, Condition=RigVMVariableNode() {variable=debug}.Value)
   /Script/ControlRig.RigUnit_DebugTransformMutableItemSpace(
    Color=(R=1,G=0,B=0,A=1)
    DebugDrawSettings=(DepthPriority=SDPG_Foreground,Lifetime=-1)
    Mode=Axes
    Scale=10
    Space=(Type=Bone)
    Thickness=0.15
    WorldOffset=(Rotation=(X=0,Y=0,Z=0,W=1),Translation=(X=0,Y=0,Z=0),Scale3D=(X=1,Y=1,Z=1))
    bEnabled=True
    Transform=@n8.Result
    ExecutePin=@BranchNode_3_RigVMUnitNode.True
   ) {is=struct RigVMUnitNode}
   n10 = &RigVMFunction_ControlFlowBranch(ExecuteContext=@RigUnit_SequenceExecution_1.B, Condition=RigVMVariableNode() {variable=Initial}.Value)
   Branch_1_RigVMUnitNode = &RigVMFunction_ControlFlowBranch(ExecuteContext=@n10.False, Condition=RigVMVariableNode() {variable=Stretch}.Value)
   n11 = RigVMDispatchNode(Index=-1, Array=RigVMVariableNode() {variable=Controls}.Value) {template="DISPATCH_RigVMDispatch_ArrayGetAtIndex(in Array,in Index,out Element)"}
   n12 = RigVMDispatchNode(Index=-1, Array=RigVMVariableNode() {variable=Bones}.Value) {template="DISPATCH_RigVMDispatch_ArrayGetAtIndex(in Array,in Index,out Element)"}
   &RigUnit_ParentConstraint(
    AdvancedSettings=(RotationOrderForFilter=XZY)
    Filter=(TranslationFilter=(bX=false,bY=false,bZ=false),RotationFilter=(bX=True,bY=True,bZ=True),ScaleFilter=(bX=True,bY=True,bZ=True))
    Parents=((Item=(Type=Bone),Weight=1))
    Weight=1
    Parents.0.Item=@n11.Element
    Child=@n12.Element
    ExecutePin=@Branch_1_RigVMUnitNode.False
   )
   &RigUnit_ParentConstraint(
    AdvancedSettings=(RotationOrderForFilter=XZY)
    Parents=((Item=(Type=Bone),Weight=1))
    Weight=1
    Parents.0.Item=@n11.Element
    Child=@n12.Element
    ExecutePin=@Branch_1_RigVMUnitNode.True
   )
   RigUnit_SequenceExecution_2 = RigVMAggregateNode(ExecuteContext=())
   n13 = RigVMVariableNode(Value=@n1.Local Spline, ExecuteContext=@RigUnit_SequenceExecution_2.A) {variable=local_spline}
   BranchNode_2_RigVMUnitNode = &RigVMFunction_ControlFlowBranch(ExecuteContext=@n13.ExecuteContext, Condition=RigVMVariableNode() {variable=debug}.Value)
   &RigUnit_DrawControlRigSpline(
    Color=(R=1,G=0,B=0,A=1)
    Detail=16
    Thickness=1
    ExecutePin=@BranchNode_2_RigVMUnitNode.True
    Spline=RigVMVariableNode() {variable=local_spline}.Value
   )
   n14 = RigVMVariableNode() {variable=Bones}
   DISPATCH_RigVMDispatch_ArrayGetNum = RigVMDispatchNode(Array=@n14.Value) {template="DISPATCH_RigVMDispatch_ArrayGetNum(in Array,out Num)"}
   DISPATCH_RigVMDispatch_ArraySetNum = RigVMDispatchNode(
    Num=@DISPATCH_RigVMDispatch_ArrayGetNum.Num
    ExecuteContext=@RigUnit_SequenceExecution_2.B
    Array=RigVMVariableNode() {variable=aim_transforms}.Value
   ) {template="DISPATCH_RigVMDispatch_ArraySetNum(io Array,in Num)"}
   > n15 = RigVMDispatchNode(Array=@n14.Value) {template="DISPATCH_RigVMDispatch_ArrayIterator(in Array,out Element,out Index,out Count,out Ratio)"}
   n16 = &RigUnit_GetTransform(Item=@n15.Element, bInitial=RigVMVariableNode() {variable=Initial}.Value)
   n17 = RigVMVariableNode() {variable=local_spline}
   > DISPATCH_RigVMDispatch_ArraySetAtIndex = RigVMDispatchNode(
    Index=@n15.Index
    Element=&RigUnit_PositionFromControlRigSpline(
     U=/Script/ControlRigSpline.RigUnit_ClosestParameterFromControlRigSpline(Position=@n16.Transform.Translation, Spline=@n17.Value) {is=struct RigVMUnitNode}.U
     Spline=@n17.Value
    ).Position
    Array=RigVMVariableNode() {variable=aim_transforms}.Value
   ) {template="DISPATCH_RigVMDispatch_ArraySetAtIndex(io Array,in Index,in Element)"}
   RigVMFunction_ControlFlowBranch = &RigVMFunction_ControlFlowBranch(ExecuteContext=(), Condition=RigVMVariableNode() {variable=Initial}.Value)
   Branch_RigVMUnitNode = &RigVMFunction_ControlFlowBranch(ExecuteContext=@RigVMFunction_ControlFlowBranch.False, Condition=RigVMVariableNode() {variable=Stretch}.Value)
   n18 = RigVMVariableNode() {variable=template_spline}
   n19 = RigVMVariableNode() {variable=Primary Axis}
   n20 = RigVMVariableNode() {variable=Bones}
   /Script/ControlRigSpline.RigUnit_FitChainToSplineCurveItemArray(
    Alignment=Front
    DebugSettings=(Scale=1,CurveColor=(R=1,G=1,B=0,A=1),SegmentsColor=(R=1,G=0,B=0,A=1),WorldOffset=(Rotation=(X=0,Y=0,Z=0,W=1),Translation=(X=0,Y=0,Z=0),Scale3D=(X=1,Y=1,Z=1)))
    Maximum=1
    PoleVectorPosition=(X=0,Y=0,Z=0)
    RotationEaseType=Linear
    Rotations=()
    SamplingPrecision=16
    SecondaryAxis=(X=0,Y=0,Z=0)
    Weight=1
    bPropagateToChildren=True
    ExecutePin=@Branch_RigVMUnitNode.False
    Spline=@n18.Value
    PrimaryAxis=@n19.Value
    Items=@n20.Value
   ) {is=struct RigVMUnitNode}
   &RigUnit_FitChainToSplineCurveItemArray(
    Alignment=Stretched
    DebugSettings=(Scale=1,CurveColor=(R=1,G=1,B=0,A=1),SegmentsColor=(R=1,G=0,B=0,A=1),WorldOffset=(Rotation=(X=0,Y=0,Z=0,W=1),Translation=(X=0,Y=0,Z=0),Scale3D=(X=1,Y=1,Z=1)))
    Maximum=1
    PoleVectorPosition=(X=0,Y=0,Z=0)
    RotationEaseType=Linear
    Rotations=()
    SamplingPrecision=16
    SecondaryAxis=(X=0,Y=0,Z=0)
    Weight=1
    bPropagateToChildren=True
    ExecutePin=@Branch_RigVMUnitNode.True
    Spline=@n18.Value
    PrimaryAxis=@n19.Value
    Items=@n20.Value
   )
   RigVMFunction_MathTransformArrayToSRT = /Script/RigVM.RigVMFunction_MathTransformArrayToSRT(Transforms=@n3.Transforms) {is=struct RigVMUnitNode}
   DISPATCH_RigVMDispatch_CoreEquals = RigVMDispatchNode(B=(), A=RigVMVariableNode() {variable=Template Spline}.Value) {template="DISPATCH_RigVMDispatch_CoreEquals(in A,in B,out Result)"}
   DISPATCH_RigVMDispatch_If = RigVMDispatchNode(
    True=&RigUnit_ControlRigSplineFromPoints(
     bClosed=False
     Points=@RigVMFunction_MathTransformArrayToSRT.Translations
     SplineMode=RigVMVariableNode() {variable=Spline Mode}.Value
     SamplesPerSegment=RigVMVariableNode() {variable=Samples Per Segment}.Value
    ).Spline
    Condition=@DISPATCH_RigVMDispatch_CoreEquals.Result
    False=RigVMVariableNode() {variable=Template Spline}.Value
   ) {template="DISPATCH_RigVMDispatch_If(in Condition,in True,in False,out Result)"}
   RigVMDispatchNode(
    True=&RigUnit_ControlRigSplineFromPoints(
     Stretch=0
     bClosed=False
     SplineMode=RigVMVariableNode() {variable=Spline Mode}.Value
     SamplesPerSegment=RigVMVariableNode() {variable=Samples Per Segment}.Value
     Points=RigVMVariableNode() {variable=offset_positions}.Value
    ).Spline
    Condition=RigVMDispatchNode(B=(), A=RigVMVariableNode() {variable=Local Spline}.Value) {template="DISPATCH_RigVMDispatch_CoreEquals(in A,in B,out Result)"}.Result
    False=RigVMVariableNode() {variable=Local Spline}.Value
   ) {template="DISPATCH_RigVMDispatch_If(in Condition,in True,in False,out Result)"}
  ]}
 ContainedGraph {nodes [
   n1 = RigVMFunctionEntryNode()
   RigUnit_SequenceExecution_2 = &RigVMFunction_Sequence(ExecuteContext=@n1.ExecuteContext)
   RigUnit_SequenceExecution_2_1 = &RigVMFunction_Sequence(ExecuteContext=@RigUnit_SequenceExecution_2.B)
   RigUnit_SequenceExecution_2_2 = &RigVMFunction_Sequence(ExecuteContext=@RigUnit_SequenceExecution_2_1.B)
   RigVMFunctionReturnNode(A=@RigUnit_SequenceExecution_2.A, B=@RigUnit_SequenceExecution_2_1.A, C=@RigUnit_SequenceExecution_2_2.A, D=@RigUnit_SequenceExecution_2_2.B)
  ]}
 ContainedGraph {nodes [
   n1 = RigVMFunctionEntryNode()
   RigUnit_SequenceExecution_4 = &RigVMFunction_Sequence(ExecuteContext=@n1.ExecuteContext)
   RigUnit_SequenceExecution_4_1 = &RigVMFunction_Sequence(ExecuteContext=@RigUnit_SequenceExecution_4.B)
   RigUnit_SequenceExecution_4_2 = &RigVMFunction_Sequence(ExecuteContext=@RigUnit_SequenceExecution_4_1.B)
   RigVMFunctionReturnNode(A=@RigUnit_SequenceExecution_4.A, B=@RigUnit_SequenceExecution_4_1.A, C=@RigUnit_SequenceExecution_4_2.A, D=@RigUnit_SequenceExecution_4_2.B)
  ]}
 ContainedGraph {nodes [
   n1 = RigVMFunctionEntryNode()
   RigUnit_SequenceExecution = &RigVMFunction_Sequence(ExecuteContext=@n1.ExecuteContext)
   RigUnit_SequenceExecution_1 = &RigVMFunction_Sequence(ExecuteContext=@RigUnit_SequenceExecution.B)
   RigUnit_SequenceExecution_2 = &RigVMFunction_Sequence(ExecuteContext=@RigUnit_SequenceExecution_1.B)
   RigVMFunctionReturnNode(A=@RigUnit_SequenceExecution.A, B=@RigUnit_SequenceExecution_1.A, C=@RigUnit_SequenceExecution_2.A, D=@RigUnit_SequenceExecution_2.B)
  ]}
 ContainedGraph {nodes [
   n1 = RigVMFunctionEntryNode()
   RigUnit_SequenceExecution_1 = &RigVMFunction_Sequence(ExecuteContext=@n1.ExecuteContext)
   RigUnit_SequenceExecution_1_1 = &RigVMFunction_Sequence(ExecuteContext=@RigUnit_SequenceExecution_1.B)
   RigUnit_SequenceExecution_1_2 = &RigVMFunction_Sequence(ExecuteContext=@RigUnit_SequenceExecution_1_1.B)
   RigVMFunctionReturnNode(A=@RigUnit_SequenceExecution_1.A, B=@RigUnit_SequenceExecution_1_1.A, C=@RigUnit_SequenceExecution_1_2.A, D=@RigUnit_SequenceExecution_1_2.B)
  ]}
 ContainedGraph {nodes [
   n1 = RigVMFunctionEntryNode()
   RigUnit_SequenceAggregate_1 = &RigVMFunction_Sequence(ExecuteContext=@n1.ExecuteContext)
   RigUnit_SequenceAggregate = RigVMAggregateNode(ExecuteContext=@RigUnit_SequenceAggregate_1.A)
   Greater = /Script/RigVM.RigVMFunction_MathDoubleGreater(B=0, A=@n1.Mirror Plane.X) {is=struct RigVMUnitNode, template="Greater::Execute(in A,in B,out Result)"}
   BranchNode_4 = &RigVMFunction_ControlFlowBranch(Condition=@Greater.Result, ExecuteContext=@RigUnit_SequenceAggregate.A)
   n2 = RigVMVariableNode() {variable=vector pos}
   RigVMVariableNode(
    Value=(X=0,Y=0,Z=0)
    ExecuteContext=@BranchNode_4.True
    Value.X=&RigUnit_GetTransform(bInitial=true, Item=@n1.Item).Transform.Translation.X
    Value.Y=@n2.Value.Y
    Value.Z=@n2.Value.Z
   ) {variable=vector pos}
   n3 = &RigVMFunction_MathDoubleGreater(B=0, A=@n1.Mirror Plane.Y) {template="Greater::Execute(in A,in B,out Result)"}
   BranchNode_1_1 = &RigVMFunction_ControlFlowBranch(Condition=@n3.Result, ExecuteContext=@RigUnit_SequenceAggregate.B)
   n4 = RigVMVariableNode() {variable=vector pos}
   RigVMVariableNode(
    Value=(X=0,Y=0,Z=0)
    ExecuteContext=@BranchNode_1_1.True
    Value.Y=&RigUnit_GetTransform(bInitial=true, Item=@n1.Item).Transform.Translation.Y
    Value.X=@n4.Value.X
    Value.Z=@n4.Value.Z
   ) {variable=vector pos}
   n5 = &RigVMFunction_MathDoubleGreater(B=0, A=@n1.Mirror Plane.Z) {template="Greater::Execute(in A,in B,out Result)"}
   BranchNode_2_1 = &RigVMFunction_ControlFlowBranch(Condition=@n5.Result, ExecuteContext=@RigUnit_SequenceAggregate.C)
   n6 = RigVMVariableNode() {variable=vector pos}
   RigVMVariableNode(
    Value=(X=0,Y=0,Z=0)
    ExecuteContext=@BranchNode_2_1.True
    Value.Z=&RigUnit_GetTransform(bInitial=true, Item=@n1.Item).Transform.Translation.Z
    Value.X=@n6.Value.X
    Value.Y=@n6.Value.Y
   ) {variable=vector pos}
   Cast = &RigVMFunction_MathBoolToFloat(Value=@Greater.Result)
   Dot__ = /Script/RigVM.RigVMFunction_MathVectorDot(
    B=(X=1,Y=1,Z=0)
    B.X=@Cast.Result
    B.Y=&RigVMFunction_MathBoolToFloat(Value=@n3.Result).Result
    B.Z=&RigVMFunction_MathBoolToFloat(Value=@n5.Result).Result
    A=RigVMVariableNode() {variable=vector pos}.Value
   ) {is=struct RigVMUnitNode, template="Dot,|::Execute(in A,in B,out Result)"}
   Less = /Script/RigVM.RigVMFunction_MathDoubleLess(B=0, A=@Dot__.Result) {is=struct RigVMUnitNode, template="Less::Execute(in A,in B,out Result)"}
   RigVMVariableNode(Value=@Less.Result, ExecuteContext=@RigUnit_SequenceAggregate.D) {variable=is on right}
   RigVMFunctionReturnNode(ExecuteContext=@RigUnit_SequenceAggregate_1.B, Right Side=RigVMVariableNode() {variable=is on right}.Value)
  ]}
 ContainedGraph {nodes [
   n1 = RigVMFunctionEntryNode()
   RigUnit_SequenceAggregate = &RigVMFunction_Sequence(ExecuteContext=@n1.ExecuteContext)
   RigUnit_SequenceAggregate_1 = &RigVMFunction_Sequence(ExecuteContext=@RigUnit_SequenceAggregate.B)
   RigUnit_SequenceAggregate_2 = &RigVMFunction_Sequence(ExecuteContext=@RigUnit_SequenceAggregate_1.B)
   RigVMFunctionReturnNode(A=@RigUnit_SequenceAggregate.A, B=@RigUnit_SequenceAggregate_1.A, C=@RigUnit_SequenceAggregate_2.A, D=@RigUnit_SequenceAggregate_2.B)
  ]}
 New Function_ContainedGraph {nodes [
   n1 = RigVMFunctionEntryNode()
   n2 = &RigVMFunction_Sequence(ExecuteContext=@n1.ExecuteContext)
   RigVMFunction_Sequence = RigVMAggregateNode(ExecuteContext=@n2.A)
   n3 = RigVMVariableNode() {variable=Neck Start Connector}
   n4 = RigVMVariableNode() {variable=Head Connector}
   Set_Module_Metadata = RigVMDispatchNode(
    Name=Bones
    NameSpace=Self
    Value=/Script/ControlRig.RigUnit_HierarchyGetChainItemArray(
     bIncludeEnd=True
     bIncludeStart=True
     bReverse=false
     Start=/Script/ControlRig.RigUnit_ResolveConnector(SkipSocket=true, Connector=@n3.Value) {is=struct RigVMUnitNode}.Result
     End=&RigUnit_ResolveConnector(SkipSocket=true, Connector=@n4.Value).Result
    ) {is=struct RigVMUnitNode}.Chain
    ExecuteContext=@RigVMFunction_Sequence.A
   ) {template="DISPATCH_RigDispatch_SetModuleMetadata(in Name,in NameSpace,in Value,out Success)"}
   > n5 = &RigUnit_HierarchyAddNull(
    Name=neck_space
    Space=LocalSpace
    Transform=&RigUnit_GetTransform(bInitial=true, Item=RigVMVariableNode() {variable=Neck Start Connector}.Value).Transform
    Parent=RigVMVariableNode() {variable=Parent Control Connector}.Value
   )
   > /Script/ControlRig.RigUnit_SetMetadataTag(NameSpace=Self, Tag=Neck Space Null, Item=@n5.Item) {is=struct RigVMUnitNode}
   > RigVMVariableNode(Value=@n5.Item) {variable=neck_space}
   Get_Module_Metadata_2 = RigVMDispatchNode(Default=(), Name=Bones, NameSpace=Self) {template="DISPATCH_RigDispatch_GetModuleMetadata(in Name,in NameSpace,in Default,out Value,out Found)"}
   n6 = RigVMDispatchNode(Index=0, Array=@Get_Module_Metadata_2.Value) {template="DISPATCH_RigVMDispatch_ArrayGetAtIndex(in Array,in Index,out Element)"}
   > Create Bone Chain = RigVMFunctionReferenceNode(
    Debug=true
    Debug Scale=15
    Debug Thickness=0.5
    Enable Orient Bones=true
    Kind=Direction
    Optional Suffix=rig
    Primary Axis=(X=1,Y=0,Z=0)
    Secondary Axis=(X=0,Y=1,Z=0)
    Name=@n6.Element.Name
    Transforms=&RigUnit_GetTransformItemArray(Space=GlobalSpace, bInitial=False, Items=@Get_Module_Metadata_2.Value).Transforms
    Parent=&RigUnit_HierarchyGetParent(bDefaultParent=True, Child=@n6.Element).Parent
   )
   > Set_Module_Metadata_1 = RigVMDispatchNode(Name=Rig Bones, NameSpace=Self, Value=@"Create Bone Chain".Bones) {template="DISPATCH_RigDispatch_SetModuleMetadata(in Name,in NameSpace,in Value,out Success)"}
   Get_Module_Metadata = RigVMDispatchNode(Default=(), Name=Rig Bones, NameSpace=Self) {template="DISPATCH_RigDispatch_GetModuleMetadata(in Name,in NameSpace,in Default,out Value,out Found)"}
   DISPATCH_RigVMDispatch_ArrayIterator = RigVMDispatchNode(Array=@Get_Module_Metadata.Value, ExecuteContext=@RigVMFunction_Sequence.B) {template="DISPATCH_RigVMDispatch_ArrayIterator(in Array,out Element,out Index,out Count,out Ratio)"}
   Subtract = &RigVMFunction_MathIntSub(B=1, A=@DISPATCH_RigVMDispatch_ArrayIterator.Count) {template="Subtract::Execute(in A,in B,out Result)"}
   RigVMFunction_MathBoolNot = &RigVMFunction_MathBoolNot(Value=RigVMDispatchNode(A=@DISPATCH_RigVMDispatch_ArrayIterator.Index, B=@Subtract.Result) {template="DISPATCH_RigVMDispatch_CoreEquals(in A,in B,out Result)"}.Result)
   n7 = &RigVMFunction_ControlFlowBranch(ExecuteContext=@DISPATCH_RigVMDispatch_ArrayIterator.ExecuteContext, Condition=@RigVMFunction_MathBoolNot.Result)
   Get_Module_Metadata_1 = RigVMDispatchNode(Default=(), Name=Bones, NameSpace=Self) {template="DISPATCH_RigDispatch_GetModuleMetadata(in Name,in NameSpace,in Default,out Value,out Found)"}
   n8 = RigVMDispatchNode(Index=@DISPATCH_RigVMDispatch_ArrayIterator.Index, Array=@Get_Module_Metadata_1.Value) {template="DISPATCH_RigVMDispatch_ArrayGetAtIndex(in Array,in Index,out Element)"}
   n9 = RigVMDispatchNode(B=0, A=@DISPATCH_RigVMDispatch_ArrayIterator.Index) {template="DISPATCH_RigVMDispatch_CoreEquals(in A,in B,out Result)"}
   DISPATCH_RigVMDispatch_If = RigVMDispatchNode(False=False, True=true, Condition=@n9.Result) {template="DISPATCH_RigVMDispatch_If(in Condition,in True,in False,out Result)"}
   Concat = &RigVMFunction_NameConcat(B=_ctrl, A=&RigUnit_HierarchyGetParent(bDefaultParent=True, Child=@n8.Element).Parent.Name) {template="Concat::Execute(in A,in B,out Result)"}
   Control Stack at Item = RigVMFunctionReferenceNode(
    Add Bottom Null=False
    Add Secondary Control=False
    Control Suffix=_ctrl
    Null Suffix=_null
    Orient to World=true
    Secondary Settings="(DisplayName=\"None\",Shape=(bVisible=True,Name=\"Default\",Color=(R=1,G=0,B=0,A=1),Transform=(Rotation=(X=0,Y=0,Z=0,W=1),Translation=(X=0,Y=0,Z=0),Scale3D=(X=1,Y=1,Z=1))),Proxy=(bIsProxy=False,ShapeVisibility=BasedOnSelection),InitialSpace=LocalSpace,bSelecta ...(truncated)"
    Tag Controls=true
    Zero Controls=true
    Name=@n8.Element.Name
    Item=@DISPATCH_RigVMDispatch_ArrayIterator.Element
    Add Null=@DISPATCH_RigVMDispatch_If.Result
    Parent=RigVMDispatchNode(False=(Type=Control,Name="None"), False.Name=@Concat.Result, Condition=@n9.Result, True=RigVMVariableNode() {variable=neck_space}.Value) {template="DISPATCH_RigVMDispatch_If(in Condition,in True,in False,out Result)"}.Result
    Settings=@n1.Neck FK Settings
    ExecuteContext=@n7.True
   )
   > &RigUnit_SetMetadataTag(NameSpace=Self, Tag=FK Controls, Item=@"Control Stack at Item".Control)
   Control Stack at Item_1 = RigVMFunctionReferenceNode(
    Add Bottom Null=False
    Add Null=true
    Add Secondary Control=False
    Control Suffix=_ctrl
    Name=head
    Null Suffix=_space
    Orient to World=true
    Secondary Settings="(DisplayName=\"None\",Shape=(bVisible=True,Name=\"Default\",Color=(R=1,G=0,B=0,A=1),Transform=(Rotation=(X=0,Y=0,Z=0,W=1),Translation=(X=0,Y=0,Z=0),Scale3D=(X=1,Y=1,Z=1))),Proxy=(bIsProxy=False,ShapeVisibility=BasedOnSelection),InitialSpace=LocalSpace,bSelecta ...(truncated)"
    Tag Controls=true
    Zero Controls=true
    Parent=@"Control Stack at Item".Control
    Item=@DISPATCH_RigVMDispatch_ArrayIterator.Element
    Settings=@n1.Head FK Settings
    ExecuteContext=@n7.False
   )
   > &RigUnit_SetMetadataTag(NameSpace=Self, Tag=FK Controls, Item=@"Control Stack at Item_1".Control)
   > /Script/ControlRig.RigUnit_HierarchyAddAnimationChannelBool(InitialValue=False, LimitsEnabled=(), MaximumValue=True, MinimumValue=False, Name=local, Parent=@"Control Stack at Item_1".Control) {
    is=struct RigVMUnitNode
    template="SpawnAnimationChannel::Execute(in InitialValue,in MinimumValue,in MaximumValue,in LimitsEnabled,in Parent,in Name,out Item)"
   }
   > RigVMDispatchNode(Name=FK Parent, NameSpace=Self, Value=@"Control Stack at Item_1".Control) {template="DISPATCH_RigDispatch_SetModuleMetadata(in Name,in NameSpace,in Value,out Success)"}
   n10 = /Script/ControlRig.RigUnit_GetShapeTransform(Control=@"Control Stack at Item_1".Control.Name) {is=struct RigVMUnitNode}
   n11 = &RigUnit_ResolveConnector(SkipSocket=true, Connector=RigVMVariableNode() {variable=Head Connector}.Value)
   > /Script/ControlRig.RigUnit_SetShapeTransform(
    Transform=(Rotation=(X=0,Y=0,Z=0,W=1),Translation=(X=0,Y=0,Z=0),Scale3D=(X=1,Y=1,Z=1))
    Control=@"Control Stack at Item_1".Control.Name
    Transform.Rotation=@n10.Transform.Rotation
    Transform.Scale3D=@n10.Transform.Scale3D
    Transform.Translation=&RigVMFunction_MathTransformMakeRelative(
     Global=&RigUnit_GetTransform(bInitial=true, Item=@n11.Result).Transform
     Parent=&RigUnit_GetTransform(bInitial=False, Item=@"Control Stack at Item_1".Control).Transform
    ).Local.Translation
   ) {is=struct RigVMUnitNode}
   > Add Null Above = RigVMFunctionReferenceNode(
    New Suffix=_null
    Replace Suffix=_ctrl
    Item=RigVMDispatchNode(Index=-1, Array=/Script/ControlRig.RigUnit_FindItemsWithMetadataTag(NameSpace=Self, Tag=FK Controls) {is=struct RigVMUnitNode}.Items) {template="DISPATCH_RigVMDispatch_ArrayGetAtIndex(in Array,in Index,out Element)"}.Element
   )
   DISPATCH_RigVMDispatch_ArrayGetAtIndex = RigVMDispatchNode(Index=0, Array=&RigUnit_FindItemsWithMetadataTag(NameSpace=Self, Tag=FK Controls).Items) {template="DISPATCH_RigVMDispatch_ArrayGetAtIndex(in Array,in Index,out Element)"}
   Add Null Below = RigVMFunctionReferenceNode(
    New Suffix=_ik_ref
    Replace Suffix=_null
    Transform=(Rotation=(X=0,Y=0,Z=0,W=1),Translation=(X=0,Y=0,Z=0),Scale3D=(X=1,Y=1,Z=1))
    Use Transform=False
    ExecuteContext=@DISPATCH_RigVMDispatch_ArrayIterator.Completed
    Item=&RigUnit_HierarchyGetParent(bDefaultParent=True, Child=@DISPATCH_RigVMDispatch_ArrayGetAtIndex.Element).Parent
   )
   > DISPATCH_RigDispatch_SetMetadata = RigVMDispatchNode(Name=IK Ref Null, NameSpace=Self, Value=@"Add Null Below".Null, Item=@DISPATCH_RigVMDispatch_ArrayGetAtIndex.Element) {template="DISPATCH_RigDispatch_SetMetadata(in Item,in Name,in NameSpace,in Value,out Success)"}
   Format Name = RigVMCollapseNode(name=head, side=None, suffix=_ik, prefix=RigVMVariableNode() {variable=prefix}.Value)
   Control Stack at Item_2 = RigVMFunctionReferenceNode(
    Add Bottom Null=true
    Add Null=true
    Add Secondary Control=False
    Control Suffix=_ctrl
    Null Suffix=_space
    Orient to World=true
    Secondary Settings="(DisplayName=\"None\",Shape=(bVisible=True,Name=\"Default\",Color=(R=1,G=0,B=0,A=1),Transform=(Rotation=(X=0,Y=0,Z=0,W=1),Translation=(X=0,Y=0,Z=0),Scale3D=(X=1,Y=1,Z=1))),Proxy=(bIsProxy=False,ShapeVisibility=BasedOnSelection),InitialSpace=LocalSpace,bUsePref ...(truncated)"
    Tag Controls=true
    Zero Controls=true
    Name=@"Format Name".Result
    Item=&RigUnit_HierarchyGetParent(bDefaultParent=True, Child=RigVMVariableNode() {variable=Head Connector}.Value).Parent
    ExecuteContext=@RigVMFunction_Sequence.C
    Parent=RigVMVariableNode() {variable=neck_space}.Value
    Settings=RigVMVariableNode() {variable=Head IK Settings}.Value
   )
   > &RigUnit_SetMetadataTag(NameSpace=Self, Tag=IK Controls, Item=@"Control Stack at Item_2".Control)
   n12 = &RigUnit_GetShapeTransform(Control=@"Control Stack at Item_2".Control.Name)
   n13 = &RigUnit_ResolveConnector(SkipSocket=true, Connector=RigVMVariableNode() {variable=Head Connector}.Value)
   > &RigUnit_SetShapeTransform(
    Transform=(Rotation=(X=0,Y=0,Z=0,W=1),Translation=(X=0,Y=0,Z=0),Scale3D=(X=1,Y=1,Z=1))
    Control=@"Control Stack at Item_2".Control.Name
    Transform.Rotation=@n12.Transform.Rotation
    Transform.Scale3D=@n12.Transform.Scale3D
    Transform.Translation=&RigVMFunction_MathTransformMakeRelative(
     Global=&RigUnit_GetTransform(bInitial=true, Item=@n13.Result).Transform
     Parent=&RigUnit_GetTransform(bInitial=False, Item=@"Control Stack at Item_2".Control).Transform
    ).Local.Translation
   )
   n14 = RigVMVariableNode() {variable=Head Connector}
   > Set Transform_1 = /Script/ControlRig.RigUnit_SetTranslation(
    Space=GlobalSpace
    Weight=1
    bInitial=true
    bPropagateToChildren=True
    Item=@"Control Stack at Item_2".Bottom Null
    Value=&RigUnit_GetTransform(bInitial=False, Item=&RigUnit_ResolveConnector(SkipSocket=true, Connector=@n14.Value).Result).Transform.Translation
   ) {is=struct RigVMUnitNode, template="Set Transform::Execute(in Item,in Space,in bInitial,in Value,in Weight,in bPropagateToChildren)"}
   > SpawnAnimationChannel = &RigUnit_HierarchyAddAnimationChannelBool(InitialValue=False, LimitsEnabled=(), MaximumValue=True, MinimumValue=False, Name=local, Parent=@"Control Stack at Item_2".Control) {template="SpawnAnimationChannel::Execute(in InitialValue,in MinimumValue,in MaximumValue,in LimitsEnabled,in Parent,in Name,out Item)"}
   Format Name_1 = RigVMCollapseNode(name=neck, side=None, suffix=_fk_ik_switch, prefix=RigVMVariableNode() {variable=prefix}.Value)
   n15 = &RigUnit_ResolveConnector(SkipSocket=true, Connector=RigVMVariableNode() {variable=Settings Control Connector}.Value)
   > n16 = &RigUnit_HierarchyAddAnimationChannelBool(
    LimitsEnabled=()
    MaximumValue=True
    MinimumValue=False
    Name=@"Format Name_1".Result
    Parent=@n15.Result
    InitialValue=RigVMVariableNode() {variable=FK IK Default}.Value
   ) {template="SpawnAnimationChannel::Execute(in InitialValue,in MinimumValue,in MaximumValue,in LimitsEnabled,in Parent,in Name,out Item)"}
   > RigVMDispatchNode(Name=Neck FK IK Switch, NameSpace=Self, Item=@n15.Result, Value=@n16.Item.Name) {template="DISPATCH_RigDispatch_SetMetadata(in Item,in Name,in NameSpace,in Value,out Success)"}
   > DISPATCH_RigDispatch_SetModuleMetadata = RigVMDispatchNode(Name=FK IK Switch, NameSpace=Self, Value=@n16.Item) {template="DISPATCH_RigDispatch_SetModuleMetadata(in Name,in NameSpace,in Value,out Success)"}
   > RigVMDispatchNode(Name=FK IK Switch Parent, NameSpace=Self, Value=@n15.Result) {template="DISPATCH_RigDispatch_SetModuleMetadata(in Name,in NameSpace,in Value,out Success)"}
   Finalize Controls = RigVMCollapseNode(ExecuteContext=@RigVMFunction_Sequence.D, Module Settings=RigVMVariableNode() {variable=Module Settings}.Value)
   Get_Module_Metadata_1_1 = RigVMDispatchNode(Default=(), Name=Bones, NameSpace=Self) {template="DISPATCH_RigDispatch_GetModuleMetadata(in Name,in NameSpace,in Default,out Value,out Found)"}
   Get_Module_Metadata_3_1 = RigVMDispatchNode(Default=(), Name=Rig Bones, NameSpace=Self) {template="DISPATCH_RigDispatch_GetModuleMetadata(in Name,in NameSpace,in Default,out Value,out Found)"}
   Set Closest Bone_1 = RigVMFunctionReferenceNode(Bones=@Get_Module_Metadata_1_1.Value, Items=@Get_Module_Metadata_3_1.Value, ExecuteContext=@RigVMFunction_Sequence.E)
   Get_Module_Metadata_1_1_1 = RigVMDispatchNode(Default=(), Name=Bones, NameSpace=Self) {template="DISPATCH_RigDispatch_GetModuleMetadata(in Name,in NameSpace,in Default,out Value,out Found)"}
   > Set Closest Bone_1_1 = RigVMFunctionReferenceNode(Items=&RigUnit_FindItemsWithMetadataTag(NameSpace=Self, Tag=FK Controls).Items, Bones=@Get_Module_Metadata_1_1_1.Value)
   RigVMFunctionReturnNode(
    ExecuteContext=@n2.B
    Local Spline Out=RigVMVariableNode() {variable=Local Spline}.Value
    Template Spline Out=RigVMVariableNode() {variable=Template Spline}.Value
   )
  ]}
 New Function_ContainedGraph {nodes [
   n1 = RigVMFunctionEntryNode()
   Concat_4 = RigVMAggregateNode(B=@n1.name, A=@n1.prefix, C=@n1.side, D=@n1.suffix)
   RigVMFunctionReturnNode(Result=@Concat_4.Result)
  ]}
 New Function_ContainedGraph {nodes [
   n1 = RigVMFunctionEntryNode()
   Concat_4 = RigVMAggregateNode(B=@n1.name, A=@n1.prefix, C=@n1.side, D=@n1.suffix)
   RigVMFunctionReturnNode(Result=@Concat_4.Result)
  ]}
 New Function_ContainedGraph {nodes [
   n1 = RigVMFunctionEntryNode()
   n2 = &RigUnit_FindItemsWithMetadataTag(NameSpace=Self, Tag=Controls)
   > Scale Module Controls = RigVMFunctionReferenceNode(
    Global Scale=RigVMDispatchNode(Default=1, Name=Control Size, NameSpace=Root) {template="DISPATCH_RigDispatch_GetModuleMetadata(in Name,in NameSpace,in Default,out Value,out Found)"}.Value
    Local Scale=@n1.Module Settings.ControlSize_12_#2
    Controls=@n2.Items
   )
   > Color Module Controls = RigVMFunctionReferenceNode(
    CenterColor=@n1.Module Settings.CenterColor_9_#3
    LeftColor=@n1.Module Settings.LeftColor_3_#4
    Override=@n1.Module Settings.OverrideColor_48_#5
    RightColor=@n1.Module Settings.RightColor_5_#6
    Controls=@n2.Items
   )
   > RigVMFunctionReturnNode()
  ]}
 New Function_ContainedGraph {nodes [
   n1 = RigVMFunctionEntryNode()
   > RigVMFunction_Sequence = RigVMAggregateNode()
   Get_Module_Metadata_2_2 = RigVMDispatchNode(Default=(), Name=Rig Bones, NameSpace=Self) {template="DISPATCH_RigDispatch_GetModuleMetadata(in Name,in NameSpace,in Default,out Value,out Found)"}
   Parent_3 = RigVMFunctionReferenceNode(
    Child=RigVMDispatchNode(Index=0, Array=&RigUnit_FindItemsWithMetadataTag(NameSpace=Self, Tag=Neck Space Null).Items) {template="DISPATCH_RigVMDispatch_ArrayGetAtIndex(in Array,in Index,out Element)"}.Element
    New Parent=&RigUnit_HierarchyGetParent(
     bDefaultParent=True
     Child=RigVMDispatchNode(Index=0, Array=@Get_Module_Metadata_2_2.Value) {template="DISPATCH_RigVMDispatch_ArrayGetAtIndex(in Array,in Index,out Element)"}.Element
    ).Parent
    ExecuteContext=@RigVMFunction_Sequence.B
   )
   RigVMFunction_ControlFlowBranch = &RigVMFunction_ControlFlowBranch(Condition=@n1.FK or IK, ExecuteContext=@RigVMFunction_Sequence.C)
   n2 = RigVMDispatchNode(Array=@n1.FK Controls, ExecuteContext=@RigVMFunction_ControlFlowBranch.False) {template="DISPATCH_RigVMDispatch_ArrayIterator(in Array,out Element,out Index,out Count,out Ratio)"}
   DISPATCH_RigDispatch_GetMetadata = RigVMDispatchNode(Default=(Type=Bone,Name="None"), Name=Item, NameSpace=Self, Item=@n2.Element) {template="DISPATCH_RigDispatch_GetMetadata(in Item,in Name,in NameSpace,in Default,out Value,out Found)"}
   > &RigUnit_ParentConstraint(
    AdvancedSettings=(InterpolationType=Average,RotationOrderForFilter=XZY)
    Parents=((Item=(Type=Bone,Name="None"),Weight=1))
    Weight=1
    Child=@DISPATCH_RigDispatch_GetMetadata.Value
    Parents.0.Item=@n2.Element
   )
   n3 = RigVMDispatchNode(Index=0, Array=@n1.FK Controls) {template="DISPATCH_RigVMDispatch_ArrayGetAtIndex(in Array,in Index,out Element)"}
   Parent = RigVMFunctionReferenceNode(
    ExecuteContext=@n2.Completed
    New Parent=&RigUnit_HierarchyGetParent(
     bDefaultParent=True
     Child=RigVMDispatchNode(Default=(Type=Bone,Name="None"), Name=Item, NameSpace=Self, Item=@n3.Element) {template="DISPATCH_RigDispatch_GetMetadata(in Item,in Name,in NameSpace,in Default,out Value,out Found)"}.Value
    ).Parent
    Child=&RigUnit_HierarchyGetParent(bDefaultParent=True, Child=@n3.Element).Parent
   )
   n4 = RigVMDispatchNode(
    Default=(Type=Bone,Name="None")
    Name=IK Ref Null
    NameSpace=Self
    Item=RigVMDispatchNode(Index=0, Array=RigVMVariableNode() {variable=FK Controls}.Value) {template="DISPATCH_RigVMDispatch_ArrayGetAtIndex(in Array,in Index,out Element)"}.Element
   ) {template="DISPATCH_RigDispatch_GetMetadata(in Item,in Name,in NameSpace,in Default,out Value,out Found)"}
   Get_Module_Metadata_2 = RigVMDispatchNode(Default=(), Name=Rig Bones, NameSpace=Self) {template="DISPATCH_RigDispatch_GetModuleMetadata(in Name,in NameSpace,in Default,out Value,out Found)"}
   n5 = RigVMDispatchNode(Index=0, Array=@Get_Module_Metadata_2.Value) {template="DISPATCH_RigVMDispatch_ArrayGetAtIndex(in Array,in Index,out Element)"}
   SplineIK_0 = RigVMFunctionReferenceNode(
    Bones=((Type=None,Name="None"),(Type=None,Name="None"),(Type=None,Name="None"),(Type=None,Name="None"))
    Controls=((Type=None,Name="None"),(Type=None,Name="None"),(Type=None,Name="None"),(Type=None,Name="None"))
    Primary Axis=(X=1,Y=0,Z=0)
    Samples Per Segment=16
    Secondary Spline Direction=(X=0,Y=5,Z=0)
    Spline Mode=BSpline
    Up Axis=(X=0,Y=-1,Z=0)
    Controls.3=RigVMDispatchNode(Default=(Type=Bone,Name="None"), Name=Bottom Null, NameSpace=Self, Item=@n1.IK Control) {template="DISPATCH_RigDispatch_GetMetadata(in Item,in Name,in NameSpace,in Default,out Value,out Found)"}.Value
    Controls.0=@n4.Value
    Controls.1=@n4.Value
    Bones.0=@n5.Element
    Bones.1=@n5.Element
    Bones.2=RigVMDispatchNode(Index=1, Array=@Get_Module_Metadata_2.Value) {template="DISPATCH_RigVMDispatch_ArrayGetAtIndex(in Array,in Index,out Element)"}.Element
    Bones.3=RigVMDispatchNode(Index=2, Array=@Get_Module_Metadata_2.Value) {template="DISPATCH_RigVMDispatch_ArrayGetAtIndex(in Array,in Index,out Element)"}.Element
    Controls.2=@n1.IK Control
    Stretch=@n1.Stretch
    ExecuteContext=@RigVMFunction_ControlFlowBranch.True
   )
   Get_Module_Metadata_2_1 = RigVMDispatchNode(Default=(), Name=Rig Bones, NameSpace=Self) {template="DISPATCH_RigDispatch_GetModuleMetadata(in Name,in NameSpace,in Default,out Value,out Found)"}
   > Project to New Parents = RigVMFunctionReferenceNode(Propagate to Children=True, Parents=@Get_Module_Metadata_2_1.Value, Children=RigVMVariableNode() {variable=FK Controls}.Value)
   Set IKFK Visibility = RigVMFunctionReferenceNode(
    IK Controls=((Type=Control,Name="head_ik_ctrl"))
    Show Controls=true
    IKFK Switch=@n1.FK or IK
    IK Controls.0=@n1.IK Control
    ExecuteContext=@RigVMFunction_Sequence.D
    FK Controls=RigVMVariableNode() {variable=FK Controls}.Value
   )
   n6 = RigVMDispatchNode(Index=-1, Array=@n1.FK Controls) {template="DISPATCH_RigVMDispatch_ArrayGetAtIndex(in Array,in Index,out Element)"}
   n7 = RigVMDispatchNode(Default=(Type=Bone,Name="None"), Name=Null, NameSpace=Self, Item=@n6.Element) {template="DISPATCH_RigDispatch_GetMetadata(in Item,in Name,in NameSpace,in Default,out Value,out Found)"}
   > Parent_1 = RigVMFunctionReferenceNode(
    Child=@n7.Value
    New Parent=&RigUnit_HierarchyGetParent(
     bDefaultParent=True
     Child=RigVMDispatchNode(Default=(Type=Bone,Name="None"), Name=Item, NameSpace=Self, Item=@n6.Element) {template="DISPATCH_RigDispatch_GetMetadata(in Item,in Name,in NameSpace,in Default,out Value,out Found)"}.Value
    ).Parent
   )
   n8 = &RigVMFunction_ControlFlowBranch(Condition=@n1.Local, ExecuteContext=@Parent_1.ExecuteContext)
   n9 = &RigUnit_HierarchyGetParent(bDefaultParent=True, Child=@n1.IK Control)
   DISPATCH_RigVMDispatch_ArrayGetAtIndex = RigVMDispatchNode(Index=0, Array=@n1.Bones) {template="DISPATCH_RigVMDispatch_ArrayGetAtIndex(in Array,in Index,out Element)"}
   Parent_2 = RigVMFunctionReferenceNode(
    Child=@n9.Parent
    New Parent=&RigUnit_HierarchyGetParent(bDefaultParent=True, Child=@DISPATCH_RigVMDispatch_ArrayGetAtIndex.Element).Parent
    ExecuteContext=@n8.Completed
   )
   n10 = &RigVMFunction_ControlFlowBranch(Condition=@n1.Local, ExecuteContext=@Parent_2.ExecuteContext)
   n11 = &RigUnit_ProjectTransformToNewParent(
    NewParent=(Type=Bone,Name="root")
    OldParent=(Type=Bone,Name="root")
    bChildInitial=True
    bNewParentInitial=False
    bOldParentInitial=True
    Child=@n9.Parent
   )
   Set Transform_1_1 = &RigUnit_SetRotation(Space=GlobalSpace, Weight=1, bInitial=False, Item=@n9.Parent, Value=@n11.Transform.Rotation, ExecutePin=@n10.False)
   n12 = &RigUnit_ProjectTransformToNewParent(
    NewParent=(Type=Bone,Name="root")
    OldParent=(Type=Bone,Name="root")
    bChildInitial=True
    bNewParentInitial=False
    bOldParentInitial=True
    Child=@n7.Value
   )
   Set Transform_1 = &RigUnit_SetRotation(Space=GlobalSpace, Weight=1, bInitial=False, Item=@n7.Value, Value=@n12.Transform.Rotation, ExecutePin=@n8.False)
   n13 = &RigVMFunction_ControlFlowBranch(Condition=@n1.Attach Skeleton, ExecuteContext=@RigVMFunction_Sequence.E)
   Get_Module_Metadata = RigVMDispatchNode(Default=(), Name=Rig Bones, NameSpace=Self) {template="DISPATCH_RigDispatch_GetModuleMetadata(in Name,in NameSpace,in Default,out Value,out Found)"}
   n14 = RigVMDispatchNode(Array=@Get_Module_Metadata.Value, ExecuteContext=@n13.True) {template="DISPATCH_RigVMDispatch_ArrayIterator(in Array,out Element,out Index,out Count,out Ratio)"}
   Get_Module_Metadata_1 = RigVMDispatchNode(Default=(), Name=Bones, NameSpace=Self) {template="DISPATCH_RigDispatch_GetModuleMetadata(in Name,in NameSpace,in Default,out Value,out Found)"}
   > Parent_2_1 = RigVMFunctionReferenceNode(
    Child=RigVMDispatchNode(Index=@n14.Index, Array=@Get_Module_Metadata_1.Value) {template="DISPATCH_RigVMDispatch_ArrayGetAtIndex(in Array,in Index,out Element)"}.Element
    New Parent=@n14.Element
   )
   RigVMFunctionReturnNode(ExecuteContext=@RigVMFunction_Sequence.F)
   RigVMDispatchNode(Default=(Type=Bone,Name="None"), Name=Item, NameSpace=Self, Item=@n1.IK Control) {template="DISPATCH_RigDispatch_GetMetadata(in Item,in Name,in NameSpace,in Default,out Value,out Found)"}
  ]}
 New Function_ContainedGraph {nodes [
   RigVMFunctionEntryNode()
   DISPATCH_RigDispatch_GetModuleMetadata = RigVMDispatchNode(Default=(Type=Bone,Name="None"), Name=FK IK Switch, NameSpace=Self) {template="DISPATCH_RigDispatch_GetModuleMetadata(in Name,in NameSpace,in Default,out Value,out Found)"}
   > SetAnimationChannel = /Script/ControlRig.RigUnit_SetBoolAnimationChannel(
    Value=false
    bInitial=False
    Channel=@DISPATCH_RigDispatch_GetModuleMetadata.Value.Name
    Control=RigVMDispatchNode(Default=(Type=Bone,Name="None"), Name=FK IK Switch Parent, NameSpace=Self) {template="DISPATCH_RigDispatch_GetModuleMetadata(in Name,in NameSpace,in Default,out Value,out Found)"}.Value.Name
   ) {is=struct RigVMUnitNode, template="SetAnimationChannel::Execute(in Value,in Control,in Channel,in bInitial)"}
   > RigVMFunction_Sequence = RigVMAggregateNode()
   DISPATCH_RigVMDispatch_ArrayGetAtIndex = RigVMDispatchNode(Index=0, Array=&RigUnit_FindItemsWithMetadataTag(NameSpace=Self, Tag=IK Controls).Items) {template="DISPATCH_RigVMDispatch_ArrayGetAtIndex(in Array,in Index,out Element)"}
   n1 = &RigUnit_HierarchyGetParent(
    bDefaultParent=True
    Child=RigVMDispatchNode(Default=(Type=Bone,Name="None"), Name=Item, NameSpace=Self, Item=@DISPATCH_RigVMDispatch_ArrayGetAtIndex.Element) {template="DISPATCH_RigDispatch_GetMetadata(in Item,in Name,in NameSpace,in Default,out Value,out Found)"}.Value
   )
   Parent = RigVMFunctionReferenceNode(Child=@DISPATCH_RigVMDispatch_ArrayGetAtIndex.Element, New Parent=@n1.Parent, ExecuteContext=@RigVMFunction_Sequence.A)
   RigVMFunctionReturnNode(ExecuteContext=@RigVMFunction_Sequence.C)
   &RigUnit_SetTransform(
    ExecutePin=()
    Item=(Type=Bone,Name="None")
    Value=(Rotation=(X=0,Y=0,Z=0,W=1),Translation=(X=0,Y=0,Z=0),Scale3D=(X=1,Y=1,Z=1))
    Weight=1
    bInitial=False
   )
   Set Transform_2 = &RigUnit_SetTranslation(
    ExecutePin=()
    Space=GlobalSpace
    Weight=1
    bInitial=False
    bPropagateToChildren=True
    Item=@DISPATCH_RigVMDispatch_ArrayGetAtIndex.Element
    Value=&RigUnit_ProjectTransformToNewParent(
     bChildInitial=True
     bNewParentInitial=False
     bOldParentInitial=True
     Child=@DISPATCH_RigVMDispatch_ArrayGetAtIndex.Element
     NewParent=@n1.Parent
     OldParent=@n1.Parent
    ).Transform.Translation
   ) {template="Set Transform::Execute(in Item,in Space,in bInitial,in Value,in Weight,in bPropagateToChildren)"}
  ]}
 New Function_ContainedGraph {nodes [
   n1 = RigVMFunctionEntryNode()
   > n2 = RigVMDispatchNode(Array=@n1.Items) {template="DISPATCH_RigVMDispatch_ArrayIterator(in Array,out Element,out Index,out Count,out Ratio)"}
   > FindClosestBone = RigVMFunctionReferenceNode(Initial=true, Item=@n2.Element, Bones=@n1.Bones)
   > Set_Metadata = RigVMDispatchNode(Name=Closest Bone, NameSpace=Self, Item=@n2.Element, Value=@FindClosestBone.Value) {template="DISPATCH_RigDispatch_SetMetadata(in Item,in Name,in NameSpace,in Value,out Success)"}
   RigVMFunctionReturnNode(ExecuteContext=@n2.Completed)
  ]}
 RigVMFunctionLibrary {nodes [
   CONSTR Neck = RigVMCollapseNode(
    ExecuteContext=()
    Head Connector=(Type=Connector,Name="Head")
    Head FK Settings="(DisplayName=\"None\",bSelectable=True,InitialSpace=LocalSpace,bUsePreferredRotationOrder=False,PreferredRotationOrder=YZX,Limits=(LimitTranslationX=(bMinimum=False,bMaximum=False),LimitTranslationY=(bMinimum=False,bMaximum=False),LimitTranslationZ=(bMinimum ...(truncated)"
    Head IK Settings="(DisplayName=\"None\",bSelectable=True,InitialSpace=LocalSpace,bUsePreferredRotationOrder=False,PreferredRotationOrder=YZX,Limits=(LimitTranslationX=(bMinimum=False,bMaximum=False),LimitTranslationY=(bMinimum=False,bMaximum=False),LimitTranslationZ=(bMinimum ...(truncated)"
    Neck FK Settings="(DisplayName=\"None\",bSelectable=True,InitialSpace=LocalSpace,bUsePreferredRotationOrder=False,PreferredRotationOrder=YZX,Limits=(LimitTranslationX=(bMinimum=False,bMaximum=False),LimitTranslationY=(bMinimum=False,bMaximum=False),LimitTranslationZ=(bMinimum ...(truncated)"
    Neck Start Connector=(Type=Connector,Name="Neck Start")
    Parent Control Connector=(Type=Connector,Name="Parent Control")
    Settings Control Connector=(Type=Connector,Name="None")
    prefix=None
   )
   FWD Neck = RigVMCollapseNode(
    Attach Skeleton=true
    Bones=()
    Children=()
    ExecuteContext=()
    FK Controls=()
    FK or IK=False
    IK Control=(Type=Control,Name="Neck:neck_01_ctrl")
    Local=false
   )
   Forward Neck = RigVMCollapseNode(
    Bones=()
    ExecuteContext=()
    FK Controls=()
    Head FK Space=()
    Head IK Space=()
    Head Orient Control=()
    Head Orient Reference=()
    IK Control=()
    IKFK Switch Control=()
    Mid IK Null=()
   )
   INV Neck = RigVMCollapseNode(ExecuteContext=())
   Is Right Side = RigVMCollapseNode(ExecuteContext=(), Item=(Type=Bone), Mirror Plane=(X=1,Y=1,Z=0))
   Project to New Parents = RigVMCollapseNode(Children=(), ExecuteContext=(), Parents=(), Propagate to Children=True)
   Set Closest Bone = RigVMCollapseNode(Bones=(), ExecuteContext=(), Items=())
   Set IKFK Keys = RigVMCollapseNode(ExecuteContext=(), FK Controls=(), "IK Controls "=(), Switch Control=())
   Set IKFK Visibility = RigVMCollapseNode(ExecuteContext=(), FK Controls=(), IK Controls=(), IKFK Switch=True)
   SplineIK = RigVMCollapseNode(
    Bones=()
    Controls=()
    ExecuteContext=()
    Local Spline=()
    Primary Axis=(X=1,Y=0,Z=0)
    Samples Per Segment=16
    Secondary Spline Direction=(X=0,Y=5,Z=0)
    Spline Mode=BSpline
    Template Spline=()
    Up Axis=(X=0,Y=1,Z=0)
   )
  ]}
 RigVMFunction_Sequence_1_1_ContainedGraph {nodes [
   n1 = RigVMFunctionEntryNode()
   n2 = &RigVMFunction_Sequence(ExecuteContext=@n1.ExecuteContext)
   n3 = &RigVMFunction_Sequence(ExecuteContext=@n2.B)
   n4 = &RigVMFunction_Sequence(ExecuteContext=@n3.B)
   RigVMFunctionReturnNode(A=@n2.A, B=@n3.A, C=@n4.A, D=@n4.B)
  ]}
 RigVMFunction_Sequence_1_1_ContainedGraph {nodes [
   n1 = RigVMFunctionEntryNode()
   n2 = &RigVMFunction_Sequence(ExecuteContext=@n1.ExecuteContext)
   n3 = &RigVMFunction_Sequence(ExecuteContext=@n2.B)
   n4 = &RigVMFunction_Sequence(ExecuteContext=@n3.B)
   RigVMFunctionReturnNode(A=@n2.A, B=@n3.A, C=@n4.A, D=@n4.B)
  ]}
 RigVMFunction_Sequence_1_ContainedGraph {nodes [
   n1 = RigVMFunctionEntryNode()
   RigVMFunction_Sequence = &RigVMFunction_Sequence(ExecuteContext=@n1.ExecuteContext)
   n2 = &RigVMFunction_Sequence(ExecuteContext=@RigVMFunction_Sequence.B)
   n3 = &RigVMFunction_Sequence(ExecuteContext=@n2.B)
   n4 = &RigVMFunction_Sequence(ExecuteContext=@n3.B)
   n5 = &RigVMFunction_Sequence(ExecuteContext=@n4.B)
   RigVMFunctionReturnNode(A=@RigVMFunction_Sequence.A, B=@n2.A, C=@n3.A, D=@n4.A, E=@n5.A, F=@n5.B)
  ]}
 RigVMFunction_Sequence_1_ContainedGraph {nodes [
   n1 = RigVMFunctionEntryNode()
   RigVMFunction_Sequence = &RigVMFunction_Sequence(ExecuteContext=@n1.ExecuteContext)
   n2 = &RigVMFunction_Sequence(ExecuteContext=@RigVMFunction_Sequence.B)
   n3 = &RigVMFunction_Sequence(ExecuteContext=@n2.B)
   n4 = &RigVMFunction_Sequence(ExecuteContext=@n3.B)
   n5 = &RigVMFunction_Sequence(ExecuteContext=@n4.B)
   RigVMFunctionReturnNode(A=@RigVMFunction_Sequence.A, B=@n2.A, C=@n3.A, D=@n4.A, E=@n5.A, F=@n5.B)
  ]}
 RigVMFunction_Sequence_1_ContainedGraph {nodes [
   n1 = RigVMFunctionEntryNode()
   RigVMFunction_Sequence = &RigVMFunction_Sequence(ExecuteContext=@n1.ExecuteContext)
   n2 = &RigVMFunction_Sequence(ExecuteContext=@RigVMFunction_Sequence.B)
   RigVMFunctionReturnNode(A=@RigVMFunction_Sequence.A, B=@n2.A, C=@n2.B)
  ]}
 RigVMFunction_Sequence_2_ContainedGraph {nodes [
   n1 = RigVMFunctionEntryNode()
   RigVMFunction_Sequence = &RigVMFunction_Sequence(ExecuteContext=@n1.ExecuteContext)
   n2 = &RigVMFunction_Sequence(ExecuteContext=@RigVMFunction_Sequence.B)
   n3 = &RigVMFunction_Sequence(ExecuteContext=@n2.B)
   RigVMFunctionReturnNode(A=@RigVMFunction_Sequence.A, B=@n2.A, C=@n3.A, D=@n3.B)
  ]}
 RigVMModel {nodes [
   n1 = &RigUnit_BeginExecution()
   n2 = &RigUnit_ResolveConnector(Connector=(Type=Connector,Name="Settings Control"), SkipSocket=true)
   GetAnimationChannel = /Script/ControlRig.RigUnit_GetBoolAnimationChannel(
    bInitial=False
    Channel=RigVMDispatchNode(Default=None, Name=Neck FK IK Switch, NameSpace=Self, Item=@n2.Result) {template="DISPATCH_RigDispatch_GetMetadata(in Item,in Name,in NameSpace,in Default,out Value,out Found)"}.Value
    Control=@n2.Result.Name
   ) {is=struct RigVMUnitNode, template="GetAnimationChannel::Execute(out Value,in Control,in Channel,in bInitial)"}
   Get_Module_Metadata_1 = RigVMDispatchNode(Default=(), Name=Bones, NameSpace=Self) {template="DISPATCH_RigDispatch_GetModuleMetadata(in Name,in NameSpace,in Default,out Value,out Found)"}
   DISPATCH_RigVMDispatch_ArrayGetAtIndex = RigVMDispatchNode(Index=0, Array=&RigUnit_FindItemsWithMetadataTag(NameSpace=Self, Tag=IK Controls).Items) {template="DISPATCH_RigVMDispatch_ArrayGetAtIndex(in Array,in Index,out Element)"}
   n3 = &RigUnit_FindItemsWithMetadataTag(NameSpace=Self, Tag=FK Controls)
   RigVMFunction_MathBoolOr = &RigVMFunction_MathBoolOr(
    A=&RigUnit_GetBoolAnimationChannel(Channel=local, bInitial=False, Control=@DISPATCH_RigVMDispatch_ArrayGetAtIndex.Element.Name) {template="GetAnimationChannel::Execute(out Value,in Control,in Channel,in bInitial)"}.Value
    B=&RigUnit_GetBoolAnimationChannel(
     Channel=local
     bInitial=False
     Control=RigVMDispatchNode(Index=-1, Array=@n3.Items) {template="DISPATCH_RigVMDispatch_ArrayGetAtIndex(in Array,in Index,out Element)"}.Element.Name
    ) {template="GetAnimationChannel::Execute(out Value,in Control,in Channel,in bInitial)"}.Value
   )
   FWD Neck = RigVMFunctionReferenceNode(
    Attach Skeleton=true
    Children=()
    Stretch=true
    ExecuteContext=@n1.ExecutePin
    IK Control=RigVMDispatchNode(Index=0, Array=&RigUnit_FindItemsWithMetadataTag(NameSpace=Self, Tag=IK Controls).Items) {template="DISPATCH_RigVMDispatch_ArrayGetAtIndex(in Array,in Index,out Element)"}.Element
    FK Controls=&RigUnit_FindItemsWithMetadataTag(NameSpace=Self, Tag=FK Controls).Items
    FK or IK=@GetAnimationChannel.Value
    Bones=@Get_Module_Metadata_1.Value
    Local=@RigVMFunction_MathBoolOr.Result
   )
  ]}
 RigVMModel Backwards Solve Graph {nodes [
   Get_Module_Metadata_1_1 = RigVMDispatchNode(Default=(), Name=Bones, NameSpace=Self) {template="DISPATCH_RigDispatch_GetModuleMetadata(in Name,in NameSpace,in Default,out Value,out Found)"}
   n1 = RigVMDispatchNode(ExecuteContext=(), Array=@Get_Module_Metadata_1_1.Value) {template="DISPATCH_RigVMDispatch_ArrayIterator(in Array,out Element,out Index,out Count,out Ratio)"}
   Get_Module_Metadata_1 = RigVMDispatchNode(Default=(), Name=Rig Bones, NameSpace=Self) {template="DISPATCH_RigDispatch_GetModuleMetadata(in Name,in NameSpace,in Default,out Value,out Found)"}
   > Parent_2_1 = RigVMFunctionReferenceNode(
    Child=RigVMDispatchNode(Index=@n1.Index, Array=@Get_Module_Metadata_1.Value) {template="DISPATCH_RigVMDispatch_ArrayGetAtIndex(in Array,in Index,out Element)"}.Element
    New Parent=@n1.Element
   )
   Get_Module_Metadata_2_2 = RigVMDispatchNode(Default=(), Name=Rig Bones, NameSpace=Self) {template="DISPATCH_RigDispatch_GetModuleMetadata(in Name,in NameSpace,in Default,out Value,out Found)"}
   Parent_3 = RigVMFunctionReferenceNode(
    Child=RigVMDispatchNode(Index=0, Array=&RigUnit_FindItemsWithMetadataTag(NameSpace=Self, Tag=Neck Space Null).Items) {template="DISPATCH_RigVMDispatch_ArrayGetAtIndex(in Array,in Index,out Element)"}.Element
    ExecuteContext=@n1.Completed
    New Parent=&RigUnit_HierarchyGetParent(
     bDefaultParent=True
     Child=RigVMDispatchNode(Index=0, Array=@Get_Module_Metadata_2_2.Value) {template="DISPATCH_RigVMDispatch_ArrayGetAtIndex(in Array,in Index,out Element)"}.Element
    ).Parent
   )
   > Parent = RigVMFunctionReferenceNode(
    Child=RigVMDispatchNode(Index=0, Array=&RigUnit_FindItemsWithMetadataTag(NameSpace=Self, Tag=IK Controls).Items) {template="DISPATCH_RigVMDispatch_ArrayGetAtIndex(in Array,in Index,out Element)"}.Element
    New Parent=&RigUnit_ResolveConnector(Connector=(Type=Connector,Name="Head Bone"), SkipSocket=true).Result
   )
   > For_Each_1 = RigVMDispatchNode(Array=&RigUnit_FindItemsWithMetadataTag(NameSpace=Self, Tag=FK Controls).Items) {template="DISPATCH_RigVMDispatch_ArrayIterator(in Array,out Element,out Index,out Count,out Ratio)"}
   n2 = &RigUnit_ResolveConnector(Connector=(Type=Connector,Name="Head Bone"), SkipSocket=true)
   n3 = &RigUnit_ResolveConnector(Connector=(Type=Connector,Name="Neck Start Bone"), SkipSocket=False)
   > Parent_1_1 = RigVMFunctionReferenceNode(
    New Parent=RigVMDispatchNode(
     Index=@For_Each_1.Index
     Array=&RigUnit_HierarchyGetChainItemArray(bIncludeEnd=true, bIncludeStart=True, bReverse=False, End=@n2.Result, Start=@n3.Result).Chain
    ) {template="DISPATCH_RigVMDispatch_ArrayGetAtIndex(in Array,in Index,out Element)"}.Element
    Child=@For_Each_1.Element
   )
   n4 = &RigUnit_ResolveConnector(Connector=(Type=Connector,Name="Settings Control"), SkipSocket=true)
   &RigUnit_SetBoolAnimationChannel(
    Value=false
    bInitial=False
    Channel=RigVMDispatchNode(Default=None, Name=Neck FK IK Switch, NameSpace=Self, Item=@n4.Result) {template="DISPATCH_RigDispatch_GetMetadata(in Item,in Name,in NameSpace,in Default,out Value,out Found)"}.Value
    ExecutePin=@For_Each_1.Completed
    Control=@n4.Result.Name
   ) {template="SetAnimationChannel::Execute(in Value,in Control,in Channel,in bInitial)"}
   n5 = /Script/ControlRig.RigUnit_InverseExecution() {is=struct RigVMUnitNode}
   RigVMFunction_Sequence = RigVMAggregateNode(ExecuteContext=@n5.ExecutePin)
   DISPATCH_RigDispatch_GetModuleMetadata = RigVMDispatchNode(Default=(Type=Bone,Name="None"), Name=FK IK Switch, NameSpace=Self) {template="DISPATCH_RigDispatch_GetModuleMetadata(in Name,in NameSpace,in Default,out Value,out Found)"}
   SetAnimationChannel = &RigUnit_SetBoolAnimationChannel(
    Value=false
    bInitial=False
    Channel=@DISPATCH_RigDispatch_GetModuleMetadata.Value.Name
    Control=RigVMDispatchNode(Default=(Type=Bone,Name="None"), Name=FK IK Switch Parent, NameSpace=Self) {template="DISPATCH_RigDispatch_GetModuleMetadata(in Name,in NameSpace,in Default,out Value,out Found)"}.Value.Name
    ExecutePin=@RigVMFunction_Sequence.A
   ) {template="SetAnimationChannel::Execute(in Value,in Control,in Channel,in bInitial)"}
   > &RigUnit_SetBoolAnimationChannel(
    Channel=local
    Value=false
    bInitial=False
    Control=RigVMDispatchNode(Index=-1, Array=&RigUnit_FindItemsWithMetadataTag(NameSpace=Self, Tag=FK Controls).Items) {template="DISPATCH_RigVMDispatch_ArrayGetAtIndex(in Array,in Index,out Element)"}.Element.Name
   ) {template="SetAnimationChannel::Execute(in Value,in Control,in Channel,in bInitial)"}
   Get_Module_Metadata = RigVMDispatchNode(Default=(), Name=Rig Bones, NameSpace=Self) {template="DISPATCH_RigDispatch_GetModuleMetadata(in Name,in NameSpace,in Default,out Value,out Found)"}
   n6 = RigVMDispatchNode(Array=@Get_Module_Metadata.Value, ExecuteContext=@RigVMFunction_Sequence.B) {template="DISPATCH_RigVMDispatch_ArrayIterator(in Array,out Element,out Index,out Count,out Ratio)"}
   Get_Metadata_2 = RigVMDispatchNode(Default=(Type=Bone,Name="None"), Name=Closest Bone, NameSpace=Self, Item=@n6.Element) {template="DISPATCH_RigDispatch_GetMetadata(in Item,in Name,in NameSpace,in Default,out Value,out Found)"}
   > Parent_2 = RigVMFunctionReferenceNode(Child=@n6.Element, New Parent=@Get_Metadata_2.Value)
   RigVMFunction_Sequence_1 = RigVMAggregateNode(ExecuteContext=@RigVMFunction_Sequence.C)
   n7 = RigVMDispatchNode(Index=0, Array=&RigUnit_FindItemsWithMetadataTag(NameSpace=Self, Tag=FK Controls).Items) {template="DISPATCH_RigVMDispatch_ArrayGetAtIndex(in Array,in Index,out Element)"}
   Get_Metadata_1_1 = RigVMDispatchNode(Default=(Type=Bone,Name="None"), Name=Closest Bone, NameSpace=Self, Item=@n7.Element) {template="DISPATCH_RigDispatch_GetMetadata(in Item,in Name,in NameSpace,in Default,out Value,out Found)"}
   n8 = &RigUnit_HierarchyGetParent(bDefaultParent=True, Child=@n7.Element)
   Parent_1_1_2_1 = RigVMFunctionReferenceNode(
    New Parent=@Get_Metadata_1_1.Value
    Child=&RigUnit_HierarchyGetParent(bDefaultParent=True, Child=@n8.Parent).Parent
    ExecuteContext=@RigVMFunction_Sequence_1.A
   )
   > Parent_1_1_1_1_1 = RigVMFunctionReferenceNode(New Parent=@Get_Metadata_1_1.Value, Child=@n8.Parent)
   > Parent_1_1_1_2_1 = RigVMFunctionReferenceNode(Child=@n7.Element, New Parent=@Get_Metadata_1_1.Value)
   n9 = RigVMDispatchNode(Index=-1, Array=&RigUnit_FindItemsWithMetadataTag(NameSpace=Self, Tag=FK Controls).Items) {template="DISPATCH_RigVMDispatch_ArrayGetAtIndex(in Array,in Index,out Element)"}
   Get_Metadata_1 = RigVMDispatchNode(Default=(Type=Bone,Name="None"), Name=Closest Bone, NameSpace=Self, Item=@n9.Element) {template="DISPATCH_RigDispatch_GetMetadata(in Item,in Name,in NameSpace,in Default,out Value,out Found)"}
   n10 = &RigUnit_HierarchyGetParent(bDefaultParent=True, Child=@n9.Element)
   Parent_1_1_1 = RigVMFunctionReferenceNode(
    New Parent=@Get_Metadata_1.Value
    Child=&RigUnit_HierarchyGetParent(bDefaultParent=True, Child=@n10.Parent).Parent
    ExecuteContext=@RigVMFunction_Sequence_1.B
   )
   > Parent_1_1_1_1 = RigVMFunctionReferenceNode(New Parent=@Get_Metadata_1.Value, Child=@n10.Parent)
   > Parent_1_1_1_2 = RigVMFunctionReferenceNode(Child=@n9.Element, New Parent=@Get_Metadata_1.Value)
  ]}
 RigVMModel Connector Event Graph {nodes [
   n1 = /Script/ControlRig.RigUnit_ConnectorExecution() {is=struct RigVMUnitNode}
   RigVMFunction_Sequence_1 = RigVMAggregateNode(ExecuteContext=@n1.ExecuteContext)
   Connect to Module Metadata_1_1_1 = RigVMFunctionReferenceNode(
    Connector=(Type=Connector,Name="Neck Start Bone")
    Module Metadata Name=neck
    Module Metadata NameSpace=Root
    ExecuteContext=@RigVMFunction_Sequence_1.A
   )
   n2 = &RigVMFunction_ControlFlowBranch(ExecuteContext=@"Connect to Module Metadata_1_1_1".ExecuteContext, Condition=@"Connect to Module Metadata_1_1_1".Found)
   n3 = /Script/ControlRig.RigUnit_GetCandidates() {is=struct RigVMUnitNode}
   n4 = &RigVMFunction_ControlFlowBranch(
    Condition=RigVMDispatchNode(B=(Type=Connector,Name="Neck Start Bone"), A=@n3.Connector) {template="DISPATCH_RigVMDispatch_CoreEquals(in A,in B,out Result)"}.Result
    ExecuteContext=@n2.False
   )
   n5 = RigVMDispatchNode(Array=@n3.Candidates, ExecuteContext=@n4.True) {template="DISPATCH_RigVMDispatch_ArrayIterator(in Array,out Element,out Index,out Count,out Ratio)"}
   DISPATCH_RigVMDispatch_ArrayGetAtIndex = RigVMDispatchNode(
    Index=-1
    Array=/Script/ControlRig.RigUnit_HierarchyGetParentsItemArray(
     bDefaultParent=True
     bIncludeChild=true
     bReverse=False
     Child=&RigUnit_ResolveConnector(Connector=(Type=Connector,Name="Neck Primary"), SkipSocket=true).Result
    ) {is=struct RigVMUnitNode}.Parents
   ) {template="DISPATCH_RigVMDispatch_ArrayGetAtIndex(in Array,in Index,out Element)"}
   > Find first in Hierarchy_1 = RigVMFunctionReferenceNode(Children=true, Ignore=("ik"), Search for=("neck"), Type to Search=Bone, Item=@DISPATCH_RigVMDispatch_ArrayGetAtIndex.Element)
   DISPATCH_RigVMDispatch_CoreEquals = RigVMDispatchNode(B=@n5.Element, A=@"Find first in Hierarchy_1".Item Found) {template="DISPATCH_RigVMDispatch_CoreEquals(in A,in B,out Result)"}
   RigVMFunction_ControlFlowBranch = &RigVMFunction_ControlFlowBranch(Condition=@DISPATCH_RigVMDispatch_CoreEquals.Result, ExecuteContext=@"Find first in Hierarchy_1".ExecuteContext)
   /Script/ControlRig.RigUnit_SetDefaultMatch(Default=@n5.Element, ExecutePin=@RigVMFunction_ControlFlowBranch.True) {is=struct RigVMUnitNode, template="Set Default Match::Execute(in Default)"}
   Connect to Module Metadata_1_1_1_1 = RigVMFunctionReferenceNode(
    Connector=(Type=Connector,Name="Head Bone")
    Module Metadata Name=head
    Module Metadata NameSpace=Root
    ExecuteContext=@RigVMFunction_Sequence_1.B
   )
   n6 = &RigVMFunction_ControlFlowBranch(ExecuteContext=@"Connect to Module Metadata_1_1_1_1".ExecuteContext, Condition=@"Connect to Module Metadata_1_1_1_1".Found)
   n7 = &RigUnit_GetCandidates()
   n8 = &RigVMFunction_ControlFlowBranch(
    Condition=RigVMDispatchNode(B=(Type=Connector,Name="Head Bone"), A=@n7.Connector) {template="DISPATCH_RigVMDispatch_CoreEquals(in A,in B,out Result)"}.Result
    ExecuteContext=@n6.False
   )
   n9 = RigVMDispatchNode(Array=@n7.Candidates, ExecuteContext=@n8.True) {template="DISPATCH_RigVMDispatch_ArrayIterator(in Array,out Element,out Index,out Count,out Ratio)"}
   n10 = &RigUnit_ResolveConnector(Connector=(Type=Connector,Name="Neck Primary"), SkipSocket=true)
   > Find first in Hierarchy_1_1 = RigVMFunctionReferenceNode(
    Children=true
    Ignore=("ik")
    Search for=("head")
    Type to Search=Bone
    Item=RigVMDispatchNode(
     Index=-1
     Array=&RigUnit_HierarchyGetParentsItemArray(bDefaultParent=True, bIncludeChild=true, bReverse=False, Child=@n10.Result).Parents
    ) {template="DISPATCH_RigVMDispatch_ArrayGetAtIndex(in Array,in Index,out Element)"}.Element
   )
   n11 = &RigVMFunction_ControlFlowBranch(
    Condition=RigVMDispatchNode(B=@n9.Element, A=@"Find first in Hierarchy_1_1".Item Found) {template="DISPATCH_RigVMDispatch_CoreEquals(in A,in B,out Result)"}.Result
    ExecuteContext=@"Find first in Hierarchy_1_1".ExecuteContext
   )
   &RigUnit_SetDefaultMatch(Default=@n9.Element, ExecutePin=@n11.True) {template="Set Default Match::Execute(in Default)"}
   Connect to Module Metadata = RigVMFunctionReferenceNode(
    Connector=(Type=Connector,Name="Parent Control")
    Module Metadata Name=Parent Control
    Module Metadata NameSpace=Root
    ExecuteContext=@RigVMFunction_Sequence_1.C
   )
   > Connect to Module Metadata_1 = RigVMFunctionReferenceNode(Connector=(Type=Connector,Name="Settings Control"), Module Metadata Name=Settings Control, Module Metadata NameSpace=Root)
  ]}
]
hierarchy [
 Neck Primary {type=Connector, loc=<0 0 0>, rot=<0 0 0 1>}
 Root {type=Socket, loc=<0 0 0>, rot=<0 0 0 1>}
 Parent Control {type=Connector, loc=<0 0 0>, rot=<0 0 0 1>}
 Neck Start Bone {type=Connector, loc=<0 0 0>, rot=<0 0 0 1>}
 Head Bone {type=Connector, loc=<0 0 0>, rot=<0 0 0 1>}
 Settings Control {type=Connector, loc=<0 0 0>, rot=<0 0 0 1>}
]
variables [
 FK IK Default {type=bool, default=False}
 Head FK Settings {
  type=FRigUnit_HierarchyAddControlTransform_Settings
  default="(InitialSpace=LocalSpace,bUsePreferredRotationOrder=False,PreferredRotationOrder=YZX,Limits=(LimitTranslationX=(bMinimum=False,bMaximum=False),LimitTranslationY=(bMinimum=False,bMaximum=False),LimitTranslationZ=(bMinimum=False,bMaximum=False),LimitPitch=(b ...(truncated)"
 }
 Head IK Settings {
  type=FRigUnit_HierarchyAddControlTransform_Settings
  default="(InitialSpace=LocalSpace,bUsePreferredRotationOrder=False,PreferredRotationOrder=YZX,Limits=(LimitTranslationX=(bMinimum=False,bMaximum=False),LimitTranslationY=(bMinimum=False,bMaximum=False),LimitTranslationZ=(bMinimum=False,bMaximum=False),LimitPitch=(b ...(truncated)"
 }
 Local Spline Out {type=FControlRigSpline, default=()}
 Module Settings {
  type=FUserDefinedStruct_#7
  default="(OverrideColor_48_#5=False,LeftColor_3_#4=(R=0,G=0.0231,B=0.4115,A=1),RightColor_5_#6=(R=1,G=0,B=0,A=1),CenterColor_9_#3=(R=1,G=1,B=0,A= ...(truncated)"
 }
 Neck FK Settings {
  type=FRigUnit_HierarchyAddControlTransform_Settings
  default="(InitialSpace=LocalSpace,bUsePreferredRotationOrder=False,PreferredRotationOrder=YZX,Limits=(LimitTranslationX=(bMinimum=False,bMaximum=False),LimitTranslationY=(bMinimum=False,bMaximum=False),LimitTranslationZ=(bMinimum=False,bMaximum=False),LimitPitch=(b ...(truncated)"
 }
 Template Spline Out {type=FControlRigSpline, default=()}
]
