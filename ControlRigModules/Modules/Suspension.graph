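# Control Rig module for a suspension bone, spawning a float slider control with configurable axis, limits and shape at the bone's initial transform.
# Construction tags bone and control when proxy-controlled, publishes the control as module metadata, then scales and colors it.
# Forward solve copies control transform to bone, backward solve reverses it, and the connector event links to the parent module's control.
# Weight channel nodes exist but are unwired.
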ControlRigBlueprint /ControlRigModules/Modules/Suspension
graphs [
 ConstructionGraph {nodes [
   RigUnit_PrepareForExecution = &RigUnit_PrepareForExecution()
   RigVMFunction_Sequence = &RigVMFunction_Sequence(ExecuteContext=@RigUnit_PrepareForExecution.ExecutePin)
   n1 = /Script/ControlRig.RigUnit_ResolveConnector(Connector=(Type=Connector,Name="Suspension"), SkipSocket=true) {is=struct RigVMUnitNode}
   Concat = /Script/RigVM.RigVMFunction_NameConcat(B=_slider_ctrl, A=@n1.Result.Name) {is=struct RigVMUnitNode, template="Concat::Execute(in A,in B,out Result)"}
   n2 = RigVMVariableNode() {variable=PrimaryAxis}
   DISPATCH_RigVMDispatch_CastEnumToInt = RigVMDispatchNode(Value=@n2.Value) {template="DISPATCH_RigVMDispatch_CastEnumToInt(in Value,out Result)"}
   DISPATCH_RigVMDispatch_SelectInt32 = RigVMDispatchNode(Values=((X=0,Y=0,Z=0.7071,W=0.7071),(X=-1,Y=0,Z=0,W=0),(X=-0.7071,Y=0,Z=0,W=0.7071)), Index=@DISPATCH_RigVMDispatch_CastEnumToInt.Result) {template="DISPATCH_RigVMDispatch_SelectInt32(in Index,in Values,out Result)"}
   Multiply = &RigVMFunction_MathTransformMul(
    B=(Rotation=(X=0,Y=0,Z=0,W=1),Translation=(X=0,Y=0,Z=0),Scale3D=(X=1,Y=1,Z=1))
    B.Rotation=@DISPATCH_RigVMDispatch_SelectInt32.Result
    B.Scale3D=RigVMVariableNode() {variable=Scale}.Value
    A=RigVMVariableNode() {variable=ShapeTransform}.Value
   ) {template="Multiply::Execute(in A,in B,out Result)"}
   SpawnControl = /Script/ControlRig.RigUnit_HierarchyAddControlFloat(
    InitialValue=0
    OffsetSpace=GlobalSpace
    Settings="(PrimaryAxis=X,bIsScale=False,Limits=(Limit=(bMinimum=True,bMaximum=True),MinValue=0,MaxValue=100,bDrawLimits=True),Shape=(bVisible=True,Name=\"Arrow_Thick\",Color=(R=1,G=0,B=0,A=1),Transform=(Rotation=(X=0,Y=0,Z=0,W=1),Translation=(X=0,Y=0,Z=0),Scale3D=(X=1 ...(truncated)"
    Name=@Concat.Result
    OffsetTransform=&RigUnit_GetTransform(bInitial=true, Item=@n1.Result).Transform
    Settings.Shape.Transform=@Multiply.Result
    Parent=&RigUnit_ResolveConnector(Connector=(Type=Connector,Name="Parent Control"), SkipSocket=False).Result
    ExecutePin=@RigVMFunction_Sequence.A
    Settings.PrimaryAxis=@n2.Value
    Settings.Shape.Color=RigVMVariableNode() {variable=Color}.Value
    Settings.Limits.bDrawLimits=RigVMVariableNode() {variable=Draw Limits}.Value
    Settings.Shape.Name=RigVMVariableNode() {variable=Control Shape}.Value
    Settings.Limits.Limit.bMinimum=RigVMVariableNode() {variable=LimitMin}.Value
    Settings.Limits.Limit.bMaximum=RigVMVariableNode() {variable=LimitMax}.Value
    Settings.Limits.MinValue=RigVMVariableNode() {variable=Minimum}.Value
    Settings.Limits.MaxValue=RigVMVariableNode() {variable=Maximum}.Value
   ) {
    is=struct RigVMUnitNode
    template="SpawnControl::Execute(in InitialValue,in Settings,in OffsetTransform,in OffsetSpace,in Parent,in Name,out Item)"
   }
   > RigVMVariableNode(Value=@SpawnControl.Item) {variable=Control}
   > DISPATCH_RigDispatch_SetMetadata = RigVMDispatchNode(Name=RequiresBackwardSolve, NameSpace=Root, Value=False, Item=RigVMVariableNode() {variable=Control}.Value) {template="DISPATCH_RigDispatch_SetMetadata(in Item,in Name,in NameSpace,in Value,out Success)"}
   > RigVMDispatchNode(
    Name=SourceBone
    NameSpace=Root
    Value=&RigUnit_ResolveConnector(Connector=(Type=Connector,Name="Suspension"), SkipSocket=true).Result
    Item=@SpawnControl.Item
   ) {template="DISPATCH_RigDispatch_SetMetadata(in Item,in Name,in NameSpace,in Value,out Success)"}
   > Set_Module_Metadata = RigVMDispatchNode(Name=Parent Control, NameSpace=Self, Value=@SpawnControl.Item) {template="DISPATCH_RigDispatch_SetModuleMetadata(in Name,in NameSpace,in Value,out Success)"}
   RigVMFunction_ControlFlowBranch = &RigVMFunction_ControlFlowBranch(ExecuteContext=@Set_Module_Metadata.ExecuteContext, Condition=RigVMVariableNode() {variable=Controlled by Proxy}.Value)
   /Script/ControlRig.RigUnit_SetMetadataTag(
    NameSpace=Root
    Tag=Suspension Bones
    Item=&RigUnit_ResolveConnector(Connector=(Type=Connector,Name="Suspension"), SkipSocket=true).Result
    ExecutePin=@RigVMFunction_ControlFlowBranch.True
   ) {is=struct RigVMUnitNode}
   > &RigUnit_SetMetadataTag(NameSpace=Root, Tag=Run Backwards, Item=RigVMVariableNode() {variable=Control}.Value)
   > &RigUnit_SetMetadataTag(NameSpace=Root, Tag=Suspension Controls, Item=@SpawnControl.Item)
   n3 = RigVMVariableNode() {variable=Module Settings}
   Scale Module Controls = RigVMFunctionReferenceNode(
    Controls=((Type=None,Name="None"))
    Global Scale=RigVMDispatchNode(Default=1, Name=Control Size, NameSpace=Root) {template="DISPATCH_RigDispatch_GetModuleMetadata(in Name,in NameSpace,in Default,out Value,out Found)"}.Value
    ExecuteContext=@RigVMFunction_Sequence.B
    Controls.0=@SpawnControl.Item
    Local Scale=@n3.Value.ControlSize_12_#1
   )
   > Color Module Controls = RigVMFunctionReferenceNode(
    Controls=((Type=None,Name="None"))
    Override=False
    Controls.0=@SpawnControl.Item
    CenterColor=@n3.Value.CenterColor_9_#2
    LeftColor=@n3.Value.LeftColor_3_#3
    RightColor=@n3.Value.RightColor_5_#4
   )
   SpawnAnimationChannel = /Script/ControlRig.RigUnit_HierarchyAddAnimationChannelFloat(
    ExecutePin=()
    LimitsEnabled=(Enabled=(bMinimum=True,bMaximum=True))
    MaximumValue=1
    MinimumValue=0
    Name=Weight
    Parent=@SpawnControl.Item
    InitialValue=RigVMVariableNode() {variable=Weight}.Value
   ) {
    is=struct RigVMUnitNode
    template="SpawnAnimationChannel::Execute(in InitialValue,in MinimumValue,in MaximumValue,in LimitsEnabled,in Parent,in Name,out Item)"
   }
  ]}
 RigVMFunctionLibrary {links [], nodes []}
 RigVMModel {nodes [
   n1 = /Script/ControlRig.RigUnit_InverseExecution() {is=struct RigVMUnitNode}
   &RigUnit_SetTransform(
    Weight=1
    bInitial=False
    Value=&RigUnit_GetTransform(bInitial=False, Item=&RigUnit_ResolveConnector(Connector=(Type=Connector,Name="Suspension"), SkipSocket=true).Result).Transform
    ExecutePin=@n1.ExecutePin
    Item=RigVMVariableNode() {variable=Control}.Value
   )
   > DISPATCH_RigDispatch_SetMetadata = RigVMDispatchNode(Name=RequiresBackwardSolve, NameSpace=Root, Value=False, Item=RigVMVariableNode() {variable=Control}.Value) {template="DISPATCH_RigDispatch_SetMetadata(in Item,in Name,in NameSpace,in Value,out Success)"}
   RigUnit_BeginExecution = &RigUnit_BeginExecution()
   &RigUnit_SetTransform(
    Weight=1
    bInitial=False
    Value=&RigUnit_GetTransform(bInitial=False, Item=RigVMVariableNode() {variable=Control}.Value).Transform
    Item=&RigUnit_ResolveConnector(Connector=(Type=Connector,Name="Suspension"), SkipSocket=true).Result
    ExecutePin=@RigUnit_BeginExecution.ExecutePin
   )
   /Script/ControlRig.RigUnit_GetFloatAnimationChannel(Channel=Weight, bInitial=False, Control=RigVMVariableNode() {variable=Control}.Value.Name) {is=struct RigVMUnitNode, template="GetAnimationChannel::Execute(out Value,in Control,in Channel,in bInitial)"}
   &RigUnit_GetFloatAnimationChannel(Channel=Weight, bInitial=False, Control=RigVMVariableNode() {variable=Control}.Value.Name) {template="GetAnimationChannel::Execute(out Value,in Control,in Channel,in bInitial)"}
  ]}
 RigVMModel Connector Event Graph {nodes [
   n1 = /Script/ControlRig.RigUnit_ConnectorExecution() {is=struct RigVMUnitNode}
   Connect to Module Metadata = RigVMFunctionReferenceNode(
    Connector=(Type=Connector,Name="Parent Control")
    Module Metadata Name=Parent Control
    Module Metadata NameSpace=Parent
    ExecuteContext=@n1.ExecuteContext
   )
  ]}
]
hierarchy [
 Suspension {type=Connector, loc=<0 0 0>, rot=<0 0 0 1>}
 pelvis_socket {type=Socket, loc=<0 0 0>, rot=<0 0 0 1>}
 Parent Control {type=Connector, loc=<0 0 0>, rot=<0 0 0 1>}
]
variables [
 Color {type=FLinearColor, default=(R=1,G=0,B=0,A=1)}
 Control {type=FRigElementKey, default=(Type=None,Name="")}
 Control Shape {type=FName, default=Arrow_Thick}
 Controlled by Proxy {type=bool, default=True}
 Draw Limits {type=bool, default=False}
 LimitMax {type=bool, default=True}
 LimitMin {type=bool, default=True}
 Maximum {type=double, default=100}
 Minimum {type=double, default=0}
 Module Settings {
  type=FUserDefinedStruct_#5
  default="(OverrideColor_48_#6=False,LeftColor_3_#3=(R=0,G=0.0231,B=0.4115,A=1),RightColor_5_#4=(R=1,G=0,B=0,A=1),CenterColor_9_#2=(R=1,G=1,B=0,A= ...(truncated)"
 }
 PrimaryAxis {type=ERigControlAxis, default=Z}
 Scale {type=FVector, default=(X=1,Y=1,Z=1)}
 ShapeTransform {type=FTransform, default=(Rotation=(X=0,Y=0,Z=0,W=1),Translation=(X=0,Y=0,Z=0),Scale3D=(X=1,Y=1,Z=1))}
 Weight {type=float, default=1}
]
