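# Modular Control Rig vehicle module that adds a BodyProxy control for posing the body with automatic wheel response.
# Construction spawns the proxy with a Steering channel and gathers its driven controls.
# Outside interaction the proxy follows the body connector.
# During interaction it drives the hardcoded Body bone, auto-steers tagged bones, adjusts suspension and swing bones by sphere-tracing the ground, then copies SourceBone transforms back onto the driven controls.
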ControlRigBlueprint /ControlRigModules/Modules/VehiclePivotProxy
graphs [
 ConstructionGraph {nodes [
   RigUnit_PrepareForExecution = &RigUnit_PrepareForExecution()
   RigUnit_GetTransform = &RigUnit_GetTransform(
    bInitial=False
    Item=/Script/ControlRig.RigUnit_ResolveConnector(Connector=(Type=Connector,Name="BodyForProxy"), SkipSocket=true) {is=struct RigVMUnitNode}.Result
   )
   SpawnControl = /Script/ControlRig.RigUnit_HierarchyAddControlTransform(
    InitialValue=(Rotation=(X=0,Y=0,Z=0,W=1),Translation=(X=0,Y=0,Z=0),Scale3D=(X=1,Y=1,Z=1))
    Name=BodyProxy
    OffsetSpace=GlobalSpace
    Settings="(InitialSpace=LocalSpace,bUsePreferredRotationOrder=False,PreferredRotationOrder=YZX,Limits=(LimitTranslationX=(bMinimum=False,bMaximum=False),LimitTranslationY=(bMinimum=False,bMaximum=False),LimitTranslationZ=(bMinimum=False,bMaximum=False),LimitPitch=(b ...(truncated)"
    Parent=&RigUnit_ResolveConnector(Connector=(Type=Connector,Name="Parent Control"), SkipSocket=False).Result
    OffsetTransform=@RigUnit_GetTransform.Transform
    ExecutePin=@RigUnit_PrepareForExecution.ExecutePin
    Settings.Proxy.DrivenControls=RigVMVariableNode() {variable=ControlsToRunBackwards}.Value
    Settings.Shape.Name=RigVMVariableNode() {variable=Control Shape}.Value
    Settings.Shape.Transform=RigVMVariableNode() {variable=Control Transform}.Value
   ) {
    is=struct RigVMUnitNode
    template="SpawnControl::Execute(in InitialValue,in Settings,in OffsetTransform,in OffsetSpace,in Parent,in Name,out Item)"
   }
   > RigVMVariableNode(Value=@SpawnControl.Item) {variable=Control}
   > SpawnAnimationChannel = /Script/ControlRig.RigUnit_HierarchyAddAnimationChannelFloat(
    InitialValue=1
    LimitsEnabled=(Enabled=(bMinimum=True,bMaximum=True))
    MaximumValue=1
    MinimumValue=0
    Name=Steering
    Parent=@SpawnControl.Item
   ) {
    is=struct RigVMUnitNode
    template="SpawnAnimationChannel::Execute(in InitialValue,in MinimumValue,in MaximumValue,in LimitsEnabled,in Parent,in Name,out Item)"
   }
   > Scale Module Controls = RigVMFunctionReferenceNode(
    Controls=((Type=None,Name="None"))
    Global Scale=RigVMDispatchNode(Default=1, Name=Control Size, NameSpace=Root) {template="DISPATCH_RigDispatch_GetModuleMetadata(in Name,in NameSpace,in Default,out Value,out Found)"}.Value
    Controls.0=@SpawnControl.Item
    Local Scale=RigVMVariableNode() {variable=Module Settings}.Value.ControlSize_12_#1
   )
   Get_Module_Metadata = RigVMDispatchNode(Default=(Type=Bone,Name="None"), Name=Parent Control, NameSpace=Parent) {template="DISPATCH_RigDispatch_GetModuleMetadata(in Name,in NameSpace,in Default,out Value,out Found)"}
   n1 = RigVMVariableNode() {variable=ControlsToRunBackwards}
   RigVMFunction_ControlFlowBranch = &RigVMFunction_ControlFlowBranch(
    Condition=RigVMDispatchNode(Element=@Get_Module_Metadata.Value, Array=@n1.Value) {template="DISPATCH_RigVMDispatch_ArrayFind(in Array,in Element,out Index,out Success)"}.Success
    ExecuteContext=@"Scale Module Controls".ExecuteContext
   )
   For_Each = RigVMDispatchNode(
    Array=/Script/ControlRig.RigUnit_FindItemsWithMetadataTag(NameSpace=Root, Tag=Run Backwards) {is=struct RigVMUnitNode}.Items
    ExecuteContext=@RigVMFunction_ControlFlowBranch.Completed
   ) {template="DISPATCH_RigVMDispatch_ArrayIterator(in Array,out Element,out Index,out Count,out Ratio)"}
   n2 = RigVMVariableNode() {variable=ControlsToRunBackwards}
   n3 = &RigVMFunction_ControlFlowBranch(
    Condition=RigVMDispatchNode(Element=@For_Each.Element, Array=@n2.Value) {template="DISPATCH_RigVMDispatch_ArrayFind(in Array,in Element,out Index,out Success)"}.Success
    ExecuteContext=@For_Each.ExecuteContext
   )
   RigVMDispatchNode(Element=@For_Each.Element, ExecuteContext=@n3.False, Array=@n2.Value) {template="DISPATCH_RigVMDispatch_ArrayAdd(io Array,in Element,out Index)"}
   RigVMDispatchNode(Element=@Get_Module_Metadata.Value, ExecuteContext=@RigVMFunction_ControlFlowBranch.False, Array=@n1.Value) {template="DISPATCH_RigVMDispatch_ArrayAdd(io Array,in Element,out Index)"}
  ]}
 RigVMFunctionLibrary {links [], nodes []}
 RigVMFunction_Sequence_1_ContainedGraph {nodes [
   n1 = RigVMFunctionEntryNode()
   RigVMFunction_Sequence = &RigVMFunction_Sequence(ExecuteContext=@n1.ExecuteContext)
   n2 = &RigVMFunction_Sequence(ExecuteContext=@RigVMFunction_Sequence.B)
   n3 = &RigVMFunction_Sequence(ExecuteContext=@n2.B)
   n4 = &RigVMFunction_Sequence(ExecuteContext=@n3.B)
   n5 = &RigVMFunction_Sequence(ExecuteContext=@n4.B)
   RigVMFunctionReturnNode(A=@RigVMFunction_Sequence.A, B=@n2.A, C=@n3.A, D=@n4.A, E=@n5.A, F=@n5.B)
  ]}
 RigVMModel {nodes [
   n1 = &RigUnit_BeginExecution()
   RigVMFunction_ControlFlowBranch = &RigVMFunction_ControlFlowBranch(ExecuteContext=@n1.ExecutePin, Condition=/Script/ControlRig.RigUnit_IsInteracting() {is=struct RigVMUnitNode}.bIsInteracting)
   &RigUnit_SetTransform(
    Weight=1
    bInitial=False
    Value=&RigUnit_GetTransform(bInitial=False, Item=&RigUnit_ResolveConnector(Connector=(Type=Connector,Name="BodyForProxy"), SkipSocket=true).Result).Transform
    ExecutePin=@RigVMFunction_ControlFlowBranch.False
    Item=RigVMVariableNode() {variable=Control}.Value
   )
   n2 = /Script/ControlRig.RigUnit_InteractionExecution() {is=struct RigVMUnitNode}
   RigVMFunction_Sequence = RigVMAggregateNode(ExecuteContext=@n2.ExecutePin)
   RigUnit_SetTransform = &RigUnit_SetTransform(
    Item=(Type=Bone,Name="Body")
    Weight=1
    bInitial=False
    Value=&RigUnit_GetTransform(bInitial=False, Item=RigVMVariableNode() {variable=Control}.Value).Transform
    ExecutePin=@RigVMFunction_Sequence.A
   )
   n3 = &RigVMFunction_ControlFlowBranch(ExecuteContext=@RigVMFunction_Sequence.B, Condition=RigVMVariableNode() {variable=AutoSteering}.Value)
   For_Each = RigVMDispatchNode(Array=&RigUnit_FindItemsWithMetadataTag(NameSpace=Root, Tag=Steering Bones).Items, ExecuteContext=@n3.True) {template="DISPATCH_RigVMDispatch_ArrayIterator(in Array,out Element,out Index,out Count,out Ratio)"}
   n4 = &RigVMFunction_ControlFlowBranch(
    Condition=RigVMDispatchNode(Array=&RigUnit_FindItemsWithMetadataTag(NameSpace=Root, Tag=Swing Bones).Items, Element=@For_Each.Element) {template="DISPATCH_RigVMDispatch_ArrayFind(in Array,in Element,out Index,out Success)"}.Success
    ExecuteContext=@For_Each.ExecuteContext
   )
   GetAnimationChannel = /Script/ControlRig.RigUnit_GetFloatAnimationChannel(Channel=Steering, bInitial=False, Control=RigVMVariableNode() {variable=Control}.Value.Name) {is=struct RigVMUnitNode, template="GetAnimationChannel::Execute(out Value,in Control,in Channel,in bInitial)"}
   n5 = &RigUnit_GetTransform(
    Space=LocalSpace
    bInitial=False
    Item=&RigUnit_ResolveConnector(Connector=(Type=Connector,Name="BodyForProxy"), SkipSocket=true).Result
   )
   Inverse = /Script/RigVM.RigVMFunction_MathQuaternionInverse(Value=@n5.Transform.Rotation) {is=struct RigVMUnitNode, template="Inverse::Execute(in Value,out Result)"}
   RigVMFunction_MathQuaternionGetAxis = /Script/RigVM.RigVMFunction_MathQuaternionGetAxis(Axis=Y, Quaternion=@n5.Transform.Rotation) {is=struct RigVMUnitNode, template="Axis::Execute(in Quaternion,in Axis,out Result)"}
   n6 = RigVMVariableNode() {variable=Steering Axis}
   RigVMFunction_MathVectorCross = /Script/RigVM.RigVMFunction_MathVectorCross(
    A=@RigVMFunction_MathQuaternionGetAxis.Result
    B=RigVMDispatchNode(
     Values=((X=1,Y=0,Z=0),(X=0,Y=1,Z=0),((X=0,Y=0,Z=0),(X=0,Y=0,Z=0),X=0,Y=0,Z=1))
     Index=RigVMDispatchNode(Value=@n6.Value) {template="DISPATCH_RigVMDispatch_CastEnumToInt(in Value,out Result)"}.Result
    ) {template="DISPATCH_RigVMDispatch_SelectInt32(in Index,in Values,out Result)"}.Result
   ) {is=struct RigVMUnitNode, template="Cross::Execute(in A,in B,out Result)"}
   Remap = &RigVMFunction_MathDoubleRemap(
    SourceMaximum=1
    SourceMinimum=-1
    TargetMaximum=-90
    TargetMinimum=90
    bClamp=False
    Value=/Script/RigVM.RigVMFunction_MathQuaternionRotateVector(Transform=@Inverse.Result, Vector=@RigVMFunction_MathVectorCross.Result) {is=struct RigVMUnitNode, template="Rotate Vector::Execute(in Transform,in Vector,out Result)"}.Result.Z
   )
   Radians = /Script/RigVM.RigVMFunction_MathDoubleRad(Value=@Remap.Result) {is=struct RigVMUnitNode, template="Radians::Execute(in Value,out Result)"}
   RigVMFunction_MathQuaternionFromAxisAndAngle = /Script/RigVM.RigVMFunction_MathQuaternionFromAxisAndAngle(Axis=(X=0,Y=0,Z=1), Angle=@Radians.Result) {is=struct RigVMUnitNode, template="FromAxisAndAngle::Execute(in Axis,in Angle,out Result)"}
   &RigUnit_OffsetTransformForItem(
    OffsetTransform=(Rotation=(X=0,Y=0,Z=0,W=1),Translation=(X=0,Y=0,Z=0),Scale3D=(X=1,Y=1,Z=1))
    bPropagateToChildren=True
    Item=@For_Each.Element
    Weight=@GetAnimationChannel.Value
    ExecutePin=@n4.False
    OffsetTransform.Rotation=@RigVMFunction_MathQuaternionFromAxisAndAngle.Result
   )
   For_Each_1 = RigVMDispatchNode(
    Array=&RigUnit_FindItemsWithMetadataTag(NameSpace=Root, Tag=Suspension Bones).Items
    ExecuteContext=@RigVMFunction_Sequence.C
   ) {template="DISPATCH_RigVMDispatch_ArrayIterator(in Array,out Element,out Index,out Count,out Ratio)"}
   RigUnit_GetTransform = &RigUnit_GetTransform(bInitial=False, Item=@For_Each_1.Element)
   Add = /Script/RigVM.RigVMFunction_MathVectorAdd(
    A=@RigUnit_GetTransform.Transform.Translation
    B=&RigVMFunction_MathQuaternionRotateVector(Vector=(X=0,Y=0,Z=-1000), Transform=@RigUnit_GetTransform.Transform.Rotation) {template="Rotate Vector::Execute(in Transform,in Vector,out Result)"}.Result
   ) {is=struct RigVMUnitNode, template="Add::Execute(in A,in B,out Result)"}
   n7 = /Script/ControlRig.RigUnit_SphereTraceByTraceChannel(Radius=5, TraceChannel=TraceTypeQuery1, End=@Add.Result, Start=@RigUnit_GetTransform.Transform.Translation) {is=struct RigVMUnitNode}
   RigVMFunction_MathVectorDistance = &RigVMFunction_MathVectorDistance(B=@RigUnit_GetTransform.Transform.Translation, A=@n7.HitLocation)
   n8 = RigVMVariableNode() {variable=Suspension Length}
   Less = /Script/RigVM.RigVMFunction_MathDoubleLess(A=@RigVMFunction_MathVectorDistance.Result, B=@n8.Value) {is=struct RigVMUnitNode, template="Less::Execute(in A,in B,out Result)"}
   RigVMFunction_MathBoolAnd = &RigVMFunction_MathBoolAnd(B=@Less.Result, A=@n7.bHit)
   n9 = &RigVMFunction_ControlFlowBranch(ExecuteContext=@For_Each_1.ExecuteContext, Condition=@RigVMFunction_MathBoolAnd.Result)
   Subtract = /Script/RigVM.RigVMFunction_MathVectorSub(B=@RigUnit_GetTransform.Transform.Translation, A=@n7.HitLocation) {is=struct RigVMUnitNode, template="Subtract::Execute(in A,in B,out Result)"}
   RigVMFunction_MathVectorLength = /Script/RigVM.RigVMFunction_MathVectorLength(Value=@Subtract.Result) {is=struct RigVMUnitNode, template="Length::Execute(in Value,out Result)"}
   Clamp = &RigVMFunction_MathDoubleClamp(
    Maximum=0
    Value=&RigVMFunction_MathDoubleSub(A=@RigVMFunction_MathVectorLength.Result, B=@n8.Value).Result
    Minimum=RigVMVariableNode() {variable=Ground Contact Adjust}.Value
   )
   RigVMFunction_MathVectorSetLength = /Script/RigVM.RigVMFunction_MathVectorSetLength(Length=@Clamp.Result, Value=@Subtract.Result) {is=struct RigVMUnitNode, template="SetLength::Execute(in Value,in Length,out Result)"}
   &RigUnit_SetTransform(
    Weight=1
    bInitial=False
    Item=@For_Each_1.Element
    Value=/Script/RigVM.RigVMFunction_MathTransformMakeAbsolute(
     Parent=(Rotation=(X=0,Y=0,Z=0,W=1),Translation=(X=0,Y=0,Z=0),Scale3D=(X=1,Y=1,Z=1))
     Local=&RigUnit_GetTransform(bInitial=False, Item=@For_Each_1.Element).Transform
     Parent.Translation=@RigVMFunction_MathVectorSetLength.Result
    ) {is=struct RigVMUnitNode, template="Make Absolute::Execute(in Local,in Parent,out Global)"}.Global
    ExecutePin=@n9.True
   )
   For_Each_1_1 = RigVMDispatchNode(Array=&RigUnit_FindItemsWithMetadataTag(NameSpace=Root, Tag=Swing Bones).Items, ExecuteContext=@RigVMFunction_Sequence.D) {template="DISPATCH_RigVMDispatch_ArrayIterator(in Array,out Element,out Index,out Count,out Ratio)"}
   > For_Each_2 = RigVMDispatchNode(Array=&RigUnit_FindItemsWithMetadataTag(NameSpace=Root, Tag=Wheel Bones).Items) {template="DISPATCH_RigVMDispatch_ArrayIterator(in Array,out Element,out Index,out Count,out Ratio)"}
   n10 = /Script/ControlRig.RigUnit_CollectionChildrenArray(TypeToSearch=Bone, bDefaultChildren=True, bIncludeParent=False, bRecursive=false, Parent=@For_Each_1_1.Element) {is=struct RigVMUnitNode}
   n11 = &RigUnit_GetTransform(
    bInitial=False
    Item=RigVMDispatchNode(
     Array=@n10.Items
     Index=RigVMDispatchNode(Array=@n10.Items, Element=@For_Each_2.Element) {template="DISPATCH_RigVMDispatch_ArrayFind(in Array,in Element,out Index,out Success)"}.Index
    ) {template="DISPATCH_RigVMDispatch_ArrayGetAtIndex(in Array,in Index,out Element)"}.Element
   )
   n12 = &RigVMFunction_MathQuaternionGetAxis(
    Axis=Z
    Quaternion=&RigUnit_GetTransform(bInitial=False, Item=&RigUnit_ResolveConnector(Connector=(Type=Connector,Name="BodyForProxy"), SkipSocket=true).Result).Transform.Rotation
   ) {template="Axis::Execute(in Quaternion,in Axis,out Result)"}
   n13 = &RigUnit_SphereTraceByTraceChannel(
    Radius=5
    TraceChannel=TraceTypeQuery1
    End=&RigVMFunction_MathVectorAdd(
     B=@n11.Transform.Translation
     A=/Script/RigVM.RigVMFunction_MathVectorScale(Factor=-1000, Value=@n12.Result) {is=struct RigVMUnitNode, template="Scale::Execute(in Value,in Factor,out Result)"}.Result
    ) {template="Add::Execute(in A,in B,out Result)"}.Result
    Start=@n11.Transform.Translation
   )
   n14 = &RigVMFunction_ControlFlowBranch(ExecuteContext=@For_Each_2.Completed, Condition=@n13.bHit)
   n15 = &RigVMFunction_MathDoubleRemap(
    SourceMaximum=25
    SourceMinimum=10
    TargetMaximum=10
    TargetMinimum=-10
    Value=&RigVMFunction_MathVectorDistance(A=@n11.Transform.Translation, B=@n13.HitLocation).Result
   )
   Set Transform = &RigUnit_SetRotation(
    Weight=1
    bInitial=False
    Item=@For_Each_1_1.Element
    ExecutePin=@n14.True
    Value=&RigVMFunction_MathQuaternionFromAxisAndAngle(
     Axis=(X=1,Y=0,Z=0)
     Angle=&RigVMFunction_MathDoubleRad(Value=@n15.Result) {template="Radians::Execute(in Value,out Result)"}.Result
    ) {template="FromAxisAndAngle::Execute(in Axis,in Angle,out Result)"}.Result
   )
   DISPATCH_RigVMDispatch_ArrayIterator = RigVMDispatchNode(ExecuteContext=@RigVMFunction_Sequence.E, Array=RigVMVariableNode() {variable=ControlsToRunBackwards}.Value) {template="DISPATCH_RigVMDispatch_ArrayIterator(in Array,out Element,out Index,out Count,out Ratio)"}
   DISPATCH_RigDispatch_GetMetadata = RigVMDispatchNode(Default=(Type=Bone,Name="None"), Name=SourceBone, NameSpace=Root, Item=@DISPATCH_RigVMDispatch_ArrayIterator.Element) {template="DISPATCH_RigDispatch_GetMetadata(in Item,in Name,in NameSpace,in Default,out Value,out Found)"}
   n16 = &RigVMFunction_ControlFlowBranch(Condition=@DISPATCH_RigDispatch_GetMetadata.Found, ExecuteContext=@DISPATCH_RigVMDispatch_ArrayIterator.ExecuteContext)
   &RigUnit_SetTransform(
    Weight=1
    bInitial=False
    Item=@DISPATCH_RigVMDispatch_ArrayIterator.Element
    Value=&RigUnit_GetTransform(bInitial=False, Item=@DISPATCH_RigDispatch_GetMetadata.Value).Transform
    ExecutePin=@n16.True
   )
  ]}
 RigVMModel Connector Event Graph {nodes [
   n1 = /Script/ControlRig.RigUnit_ConnectorExecution() {is=struct RigVMUnitNode}
   Connect to Module Metadata = RigVMFunctionReferenceNode(
    Connector=(Type=Connector,Name="Parent Control")
    Module Metadata Name=Root Control
    Module Metadata NameSpace=Root
    ExecuteContext=@n1.ExecuteContext
   )
   > Connect to Module Metadata_1 = RigVMFunctionReferenceNode(Connector=(Type=Connector,Name="BodyForProxy"), Module Metadata Name=Parent Control, Module Metadata NameSpace=Parent)
  ]}
]
hierarchy [
 BodyForProxy {type=Connector, loc=<0 0 0>, rot=<0 0 0 1>}
 Root {type=Socket, loc=<0 0 0>, rot=<0 0 0 1>}
 Parent Control {type=Connector, loc=<0 0 0>, rot=<0 0 0 1>}
]
variables [
 AutoSteering {type=bool, default=True}
 Control {type=FRigElementKey, default=(Type=Bone,Name="")}
 Control Shape {type=FName, default=Circle_Thick}
 Control Transform {type=FTransform, default=(Rotation=(X=0,Y=0,Z=0,W=1),Translation=(X=0,Y=0,Z=0),Scale3D=(X=8,Y=8,Z=8))}
 ControlsToRunBackwards {type="TArray<FRigElementKey>"}
 Ground Contact Adjust {type=double, default=-25}
 Module Settings {
  type=FUserDefinedStruct_#2
  default="(OverrideColor_48_#3=False,LeftColor_3_#4=(R=0,G=0.0231,B=0.4115,A=1),RightColor_5_#5=(R=1,G=0,B=0,A=1),CenterColor_9_#6=(R=1,G=1,B=0,A= ...(truncated)"
 }
 Scale {type=float, default=8}
 Steering Axis {type=ERigControlAxis, default=Y}
 Suspension Length {type=double, default=36}
]
